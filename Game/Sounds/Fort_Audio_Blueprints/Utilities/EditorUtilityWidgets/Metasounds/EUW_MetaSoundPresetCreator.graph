EditorUtilityWidgetBlueprint /Game/Sounds/Fort_Audio_Blueprints/Utilities/EditorUtilityWidgets/Metasounds/EUW_MetaSoundPresetCreator
graphs [EventGraph {nodes [
   "On Pressed (EditorUtilityButton_56)"() {is=title K2Node_ComponentBoundEvent}
   > n1 = Cast To MetaSoundDocumentInterface!(execute, Object=Get SourceReference() -> (SourceReference:Meta Sound Source Object Reference)) -> (then, CastFailed, AsMetaSound Document Interface:Meta Sound Document Interface Interface) {is=title K2Node_DynamicCast}
   > n2 = CreateSourcePresetBuilder!(
    execute
    self:Meta Sound Builder Subsystem Object Reference=Get MetaSoundBuilderSubsystem() -> (ReturnValue:Meta Sound Builder Subsystem Object Reference) {is=title K2Node_GetEngineSubsystem}
    BuilderName:Name=Builder
    ReferencedSourceClass:"Meta Sound Document Interface Interface (by ref)"=@n1
   ) -> (Succeeded:Exec, Failed:Exec, ReturnValue:Meta Sound Source Builder Object Reference)
   > Set MetasoundBuilder(MetasoundBuilder:Meta Sound Source Builder Object Reference=@n2) -> (Output_Get:Meta Sound Source Builder Object Reference)
   > n3 = GetSelectedAssetsOfClass(AssetClass=/Script/Engine.SoundWave)
   n4 = "To String (Text)"(InText=Get Text(self=Get ParamNameInput() -> (ParamNameInput:Editor Utility Editable Text Box Object Reference)))
   > n5 = Create MetaSound Literal From AudioParameter!(
    execute
    self:Meta Sound Builder Subsystem Object Reference=Get MetaSoundBuilderSubsystem()
    Param:"Audio Parameter Structure (by ref)"=ObjectArrayToAudioParameter!(Name:Name=String To Name(InString=@n4), Objects:Array of Object References=@n3) -> (ReturnValue:Audio Parameter Structure)
   ) -> (then, ReturnValue:Meta Sound Literal Structure)
   n6 = Get Text(self=Get ParamNameInput())
   > n7 = SetGraphInputDefault!(
    execute
    self:Meta Sound Builder Base Object Reference=Get MetasoundBuilder() -> (MetasoundBuilder:Meta Sound Source Builder Object Reference)
    InputName:Name=String To Name(InString="To String (Text)"(InText=@n6))
    Literal:"Meta Sound Literal Structure (by ref)"=@n5
    PageName:Name
   ) -> (Succeeded:Exec, Failed:Exec)
   n8 = Get PathInput() -> (PathInput:Editor Utility Editable Text Box Object Reference)
   n9 = Get AssetNameInput() -> (AssetNameInput:Editor Utility Editable Text Box Object Reference)
   BuildToAsset!(
    execute=[@n7.Failed, @n7.Succeeded]
    self:Meta Sound Editor Subsystem Object Reference=Get MetaSoundEditorSubsystem() -> (ReturnValue:Meta Sound Editor Subsystem Object Reference) {is=title K2Node_GetEditorSubsystem}
    InBuilder:Meta Sound Builder Base Object Reference=Get MetasoundBuilder()
    Author:String=SoundBaseAssetAction
    AssetName:String="To String (Text)"(InText=Get Text(self=@n9))
    PackagePath:String="To String (Text)"(InText=Get Text(self=@n8))
    TemplateSoundWave:Sound Wave Object Reference
   ) -> (Succeeded:Exec, Failed:Exec, ReturnValue:Meta Sound Document Interface Interface)
   "On Pressed (EditorUtilityButton)"() {is=title K2Node_ComponentBoundEvent}
   > Set Text(
    Text="To Text (String)"(InString=GetCurrentContentBrowserPath!() -> (OutPath:String, ReturnValue:bool))
    self:Text Box Object Reference=Get PathInput()
   )
   Event PreConstruct()
   > SetObject(self=Get MetaSoundSource() -> (MetaSoundSource:Single Property View Object Reference), NewObject=Self-Reference())
   Event Construct()
   Event Tick()
  ]}]
variables [
 MetasoundBuilder {type=MetaSoundSourceBuilder, default=null}
 SourceReference {type=MetaSoundSource, default=null, class=~, properties=~}
]
widgets {
 info {rootWidgetClass=&CanvasPanel, widgetCount=12}
 widgets [
  CanvasPanel_202 {widgetClassPath=&CanvasPanel, widget=CanvasPanel_202}
  EditorUtilityButton_56 {
   widgetClassPath=&EditorUtilityButton
   parent=CanvasPanel_202
   widget=EditorUtilityButton_56
   slot {
    class=CanvasPanelSlot
    properties {layoutData {alignment=<0 0>, anchors {maximum=<0 0>, minimum=<0 0>}, offsets=<ltrb 16 256 272.6727 51.7718>}}
   }
  }
  UEFN_TextBlock_C_39 {
   widgetClassPath=&UEFN_TextBlock
   parent=EditorUtilityButton_56
   widget=UEFN_TextBlock_C_39
   properties {
    justification=Center
    text=Make Preset
    font {
     bForceMonospaced=false
     bMaterialIsStencil=false
     fontMaterial=null
     fontObject=&BurbankSmall
     letterSpacing=0
     monospacedWidth=1
     size=9
     skewAmount=0
     typefaceFontName=Black
     outlineSettings {
      bApplyOutlineToDropShadows=false
      bMiteredCorners=false
      bSeparateFillAlpha=false
      outlineMaterial=null
      outlineSize=0
      outlineColor=<rgba 0 0 0 1>
     }
    }
   }
   slot {class=ButtonSlot}
  }
  MetaSoundSource {
   widgetClassPath=&SinglePropertyView
   parent=CanvasPanel_202
   widget=MetaSoundSource
   properties {propertyName=SourceReference}
   slot {
    class=CanvasPanelSlot
    properties {layoutData {alignment=<0 0>, anchors {maximum=<0 0>, minimum=<0 0>}, offsets=<ltrb 0 40 506.4064 62.032>}}
   }
  }
  ParamNameInput {
   widgetClassPath=&EditorUtilityEditableTextBox
   bIsVariable=true
   parent=CanvasPanel_202
   widget=ParamNameInput
   slot {
    class=CanvasPanelSlot
    properties {layoutData {alignment=<0 0>, anchors {maximum=<0 0>, minimum=<0 0>}, offsets=<ltrb 156 128 313.964 28.4985>}}
   }
  }
  UEFN_TextBlock_C_41 {
   widgetClassPath=&UEFN_TextBlock
   parent=CanvasPanel_202
   widget=UEFN_TextBlock_C_41
   properties {
    text=Param Name
    font {
     bForceMonospaced=false
     bMaterialIsStencil=false
     fontMaterial=null
     fontObject=&Roboto
     letterSpacing=0
     monospacedWidth=1
     size=10.5
     skewAmount=0
     typefaceFontName=Regular
     outlineSettings {
      bApplyOutlineToDropShadows=false
      bMiteredCorners=false
      bSeparateFillAlpha=false
      outlineMaterial=null
      outlineSize=0
      outlineColor=<rgba 0 0 0 1>
     }
    }
   }
   slot {
    class=CanvasPanelSlot
    properties {layoutData {alignment=<0 0>, anchors {maximum=<0 0>, minimum=<0 0>}, offsets=<ltrb 8 132 101.1862 19.2282>}}
   }
  }
  PathInput {
   widgetClassPath=&EditorUtilityEditableTextBox
   bIsVariable=true
   parent=CanvasPanel_202
   widget=PathInput
   slot {
    class=CanvasPanelSlot
    properties {layoutData {alignment=<0 0>, anchors {maximum=<0 0>, minimum=<0 0>}, offsets=<ltrb 156 168 313.964 28.4985>}}
   }
  }
  UEFN_TextBlock_C {
   widgetClassPath=&UEFN_TextBlock
   parent=CanvasPanel_202
   widget=UEFN_TextBlock_C
   properties {
    text=Path
    font {
     bForceMonospaced=false
     bMaterialIsStencil=false
     fontMaterial=null
     fontObject=&Roboto
     letterSpacing=0
     monospacedWidth=1
     size=10.5
     skewAmount=0
     typefaceFontName=Regular
     outlineSettings {
      bApplyOutlineToDropShadows=false
      bMiteredCorners=false
      bSeparateFillAlpha=false
      outlineMaterial=null
      outlineSize=0
      outlineColor=<rgba 0 0 0 1>
     }
    }
   }
   slot {
    class=CanvasPanelSlot
    properties {layoutData {alignment=<0 0>, anchors {maximum=<0 0>, minimum=<0 0>}, offsets=<ltrb 8 168 101.1862 19.2282>}}
   }
  }
  EditorUtilityButton {
   widgetClassPath=&EditorUtilityButton
   parent=CanvasPanel_202
   widget=EditorUtilityButton
   slot {class=CanvasPanelSlot, properties {layoutData {alignment=<0 0>, anchors {maximum=<0 0>, minimum=<0 0>}, offsets=<ltrb 480 168 100 30>}}}
  }
  UEFN_TextBlock_C_1 {
   widgetClassPath=&UEFN_TextBlock
   parent=EditorUtilityButton
   widget=UEFN_TextBlock_C_1
   properties {
    justification=Center
    text=Get Current Path
    font {
     bForceMonospaced=false
     bMaterialIsStencil=false
     fontMaterial=null
     fontObject=&Roboto
     letterSpacing=0
     monospacedWidth=1
     size=9
     skewAmount=0
     typefaceFontName=Regular
     outlineSettings {
      bApplyOutlineToDropShadows=false
      bMiteredCorners=false
      bSeparateFillAlpha=false
      outlineMaterial=null
      outlineSize=0
      outlineColor=<rgba 0 0 0 1>
     }
    }
   }
   slot {class=ButtonSlot}
  }
  UEFN_TextBlock_C_2 {
   widgetClassPath=&UEFN_TextBlock
   parent=CanvasPanel_202
   widget=UEFN_TextBlock_C_2
   properties {
    text=Asset Name
    font {
     bForceMonospaced=false
     bMaterialIsStencil=false
     fontMaterial=null
     fontObject=&Roboto
     letterSpacing=0
     monospacedWidth=1
     size=10.5
     skewAmount=0
     typefaceFontName=Regular
     outlineSettings {
      bApplyOutlineToDropShadows=false
      bMiteredCorners=false
      bSeparateFillAlpha=false
      outlineMaterial=null
      outlineSize=0
      outlineColor=<rgba 0 0 0 1>
     }
    }
   }
   slot {
    class=CanvasPanelSlot
    properties {layoutData {alignment=<0 0>, anchors {maximum=<0 0>, minimum=<0 0>}, offsets=<ltrb 8 212 101.1862 19.2282>}}
   }
  }
  AssetNameInput {
   widgetClassPath=&EditorUtilityEditableTextBox
   bIsVariable=true
   parent=CanvasPanel_202
   widget=AssetNameInput
   slot {
    class=CanvasPanelSlot
    properties {layoutData {alignment=<0 0>, anchors {maximum=<0 0>, minimum=<0 0>}, offsets=<ltrb 156 208 313.964 28.4985>}}
   }
  }
 ]
}
