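# MetaSound Source preset of MSS_Base_Weapon_Fire_1p that plays first-person pistol gunfire in 5.1.
# It fills the main fire layer with machine pistol close shots, plus pistol silencer waves and a rifle low-ammo cue.
# Mech, sweetener and stop-fire sound arrays are empty.
# Inputs cover weapon state (ammo, aiming, suppressor), per-layer randomization and mix, and a subsynth; the parent holds the logic, and local gunfire concurrency limits voices.
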MetaSoundSource /FortMetaSoundWeapon/Items/Metasounds/Presets/MSS_Pistol_Weapon_Fire_1p_Preset
preset_parent=/FortMetaSoundWeapon/Items/Metasounds/Legacy/MSS_Base_Weapon_Fire_1p
inputs [
 bAutoFire {type=Bool}
 Character.Affiliation {type=Enum:FortAudioAffiliation}
 Character.Health {type=Float}
 Character.UniqueID {type=Int32}
 LowAmmo.ADSMainVolumeMultiply {type=Float}
 LowAmmo.Random.MaxPitch {type=Float}
 LowAmmo.Random.MaxVolume {type=Float}
 LowAmmo.Random.MinPitch {type=Float}
 LowAmmo.Random.MinVolume {type=Float}
 LowAmmo.Sounds {type=WaveAsset:Array}
 LowAmmo.Threshold {type=Int32}
 MainFire.ADSMainVolumeMultiply {type=Float}
 MainFire.LFVolume {type=Float}
 MainFire.Random.MaxPitch {type=Float}
 MainFire.Random.MaxVolume {type=Float}
 MainFire.Random.MinPitch {type=Float}
 MainFire.Random.MinVolume {type=Float}
 MainFire.Sounds {type=WaveAsset:Array}
 MainFire.StopTriggerDelay {type=Time}
 Mech.ADSVolumeMultiply {type=Float}
 Mech.Random.DelayTimeMax {type=Time}
 Mech.Random.DelayTimeMin {type=Time}
 Mech.Random.MaxPitch {type=Float}
 Mech.Random.MaxVolume {type=Float}
 Mech.Random.MinPitch {type=Float}
 Mech.Random.MinVolume {type=Float}
 Mech.Sounds {type=WaveAsset:Array}
 Mixing.LowAmmoVolume {type=Float}
 Mixing.MainFireVolume {type=Float}
 Mixing.MechVolume {type=Float}
 Mixing.OnStopVolume {type=Float}
 Mixing.Suppressor.SuppressorOutputVolume {type=Float}
 Mixing.SweetenerVolume {type=Float}
 Mixing.WeaponOutputVolume {type=Float}
 OnStop {type=Trigger}
 OnStopFire.Random.DelayTimeMax {type=Time}
 OnStopFire.Random.DelayTimeMin {type=Time}
 OnStopFire.Random.MaxPitch {type=Float}
 OnStopFire.Random.MaxVolume {type=Float}
 OnStopFire.Random.MinPitch {type=Float}
 OnStopFire.Random.MinVolume {type=Float}
 OnStopFire.Sounds {type=WaveAsset:Array}
 ReTrigger {type=Trigger}
 Subsynth.ADSVolumeMultiply {type=Float}
 Subsynth.AttackTime {type=Time}
 Subsynth.bEnable {type=Bool}
 Subsynth.DecayTime {type=Time}
 Subsynth.FrontLRVolume {type=Float}
 Subsynth.LFVolume {type=Float}
 Subsynth.Random.DelayTimeMax {type=Time}
 Subsynth.Random.DelayTimeMin {type=Time}
 Subsynth.SubFrequency {type=Float}
 Suppressor.BaseWeaponVolume {type=Float}
 Suppressor.EnvelopeAttackTime {type=Time}
 Suppressor.EnvelopeDecayTime {type=Time}
 Suppressor.Sounds {type=WaveAsset:Array}
 Suppressor.SuppressorVolume {type=Float}
 Sweetener.ADSVolumeMultiply {type=Float}
 Sweetener.LFVolume {type=Float}
 Sweetener.Random.DelayTimeMax {type=Time}
 Sweetener.Random.DelayTimeMin {type=Time}
 Sweetener.Random.MaxPitch {type=Float}
 Sweetener.Random.MaxVolume {type=Float}
 Sweetener.Random.MinPitch {type=Float}
 Sweetener.Random.MinVolume {type=Float}
 Sweetener.Sounds {type=WaveAsset:Array}
 UE.Source.OnPlay {type=Trigger}
 Weapon.AmmoRemaining {type=Int32}
 Weapon.ClipSize {type=Int32}
 Weapon.FireRate {type=Float}
 Weapon.FireStartTime {type=Float}
 Weapon.IsAimingDownSights {type=Bool}
 Weapon.IsSuppressorEnabled {type=Bool}
]
interfaces [Fort.Character:1.0, Fort.Weapon:1.0, UE.OutputFormat.5dot1:1.0, UE.Source.OneShot:1.0, UE.Source:1.1]
outputs [
 bOnAlmostFinished {type=Bool}
 OnGunshotOut {type=Trigger}
 "UE.OutputFormat.5dot1.Audio:0" {type=Audio}
 "UE.OutputFormat.5dot1.Audio:1" {type=Audio}
 "UE.OutputFormat.5dot1.Audio:2" {type=Audio}
 "UE.OutputFormat.5dot1.Audio:3" {type=Audio}
 "UE.OutputFormat.5dot1.Audio:4" {type=Audio}
 "UE.OutputFormat.5dot1.Audio:5" {type=Audio}
 UE.Source.OneShot.OnFinished {type=Trigger}
]
pages [Default {input_defaults {
   Character.Affiliation=0
   Character.Health=0
   Character.UniqueID=0
   LowAmmo.ADSMainVolumeMultiply=1
   LowAmmo.Random.MaxPitch=1.05
   LowAmmo.Random.MaxVolume=1.05
   LowAmmo.Random.MinPitch=0.95
   LowAmmo.Random.MinVolume=0.95
   LowAmmo.Threshold=5
   MainFire.ADSMainVolumeMultiply=1
   MainFire.LFVolume=1
   MainFire.Random.MaxPitch=1.05
   MainFire.Random.MaxVolume=1.05
   MainFire.Random.MinPitch=0.95
   MainFire.Random.MinVolume=0.95
   MainFire.StopTriggerDelay=0.03
   Mech.ADSVolumeMultiply=1
   Mech.Random.DelayTimeMax=0
   Mech.Random.DelayTimeMin=0
   Mech.Random.MaxPitch=1.05
   Mech.Random.MaxVolume=1.05
   Mech.Random.MinPitch=0.95
   Mech.Random.MinVolume=0.95
   Mixing.LowAmmoVolume=1
   Mixing.MainFireVolume=1
   Mixing.MechVolume=1
   Mixing.OnStopVolume=1
   Mixing.Suppressor.SuppressorOutputVolume=1
   Mixing.SweetenerVolume=1
   Mixing.WeaponOutputVolume=1
   OnStop=null
   OnStopFire.Random.DelayTimeMax=0
   OnStopFire.Random.DelayTimeMin=0
   OnStopFire.Random.MaxPitch=1.05
   OnStopFire.Random.MaxVolume=1.05
   OnStopFire.Random.MinPitch=0.95
   OnStopFire.Random.MinVolume=0.95
   ReTrigger=null
   Subsynth.ADSVolumeMultiply=1
   Subsynth.AttackTime=0.01
   Subsynth.DecayTime=0.1
   Subsynth.FrontLRVolume=0
   Subsynth.LFVolume=1
   Subsynth.Random.DelayTimeMax=0
   Subsynth.Random.DelayTimeMin=0
   Subsynth.SubFrequency=35
   Subsynth.bEnable=null
   Suppressor.BaseWeaponVolume=0.9
   Suppressor.EnvelopeAttackTime=0
   Suppressor.EnvelopeDecayTime=0.195
   Suppressor.SuppressorVolume=1
   Sweetener.ADSVolumeMultiply=1
   Sweetener.LFVolume=0
   Sweetener.Random.DelayTimeMax=0
   Sweetener.Random.DelayTimeMin=0
   Sweetener.Random.MaxPitch=1.05
   Sweetener.Random.MaxVolume=1.05
   Sweetener.Random.MinPitch=0.95
   Sweetener.Random.MinVolume=0.95
   UE.Source.OnPlay=null
   Weapon.AmmoRemaining=30
   Weapon.ClipSize=30
   Weapon.FireRate=0.5
   Weapon.FireStartTime=0
   Weapon.IsAimingDownSights=null
   Weapon.IsSuppressorEnabled=null
   bAutoFire=null
   LowAmmo.Sounds [/Game/Sounds/Fort_Weapons/LowAmmo/weapon_low_ammo_hi_rifle_02]
   MainFire.Sounds [
    /Game/Sounds/Fort_Weapons/pistol/MachinePistol/SoundWaves/Machine_Pistol2_Fire_Close_01
    /Game/Sounds/Fort_Weapons/pistol/MachinePistol/SoundWaves/Machine_Pistol2_Fire_Close_02
    /Game/Sounds/Fort_Weapons/pistol/MachinePistol/SoundWaves/Machine_Pistol2_Fire_Close_03
    /Game/Sounds/Fort_Weapons/pistol/MachinePistol/SoundWaves/Machine_Pistol2_Fire_Close_04
    /Game/Sounds/Fort_Weapons/pistol/MachinePistol/SoundWaves/Machine_Pistol2_Fire_Close_05
    /Game/Sounds/Fort_Weapons/pistol/MachinePistol/SoundWaves/Machine_Pistol2_Fire_Close_06
    /Game/Sounds/Fort_Weapons/pistol/MachinePistol/SoundWaves/Machine_Pistol2_Fire_Close_07
    /Game/Sounds/Fort_Weapons/pistol/MachinePistol/SoundWaves/Machine_Pistol2_Fire_Close_08
    /Game/Sounds/Fort_Weapons/pistol/MachinePistol/SoundWaves/Machine_Pistol2_Fire_Close_09
    /Game/Sounds/Fort_Weapons/pistol/MachinePistol/SoundWaves/Machine_Pistol2_Fire_Close_10
    /Game/Sounds/Fort_Weapons/pistol/MachinePistol/SoundWaves/Machine_Pistol2_Fire_Close_11
    /Game/Sounds/Fort_Weapons/pistol/MachinePistol/SoundWaves/Machine_Pistol2_Fire_Close_12
    /Game/Sounds/Fort_Weapons/pistol/MachinePistol/SoundWaves/Machine_Pistol2_Fire_Close_13
    /Game/Sounds/Fort_Weapons/pistol/MachinePistol/SoundWaves/Machine_Pistol2_Fire_Close_14
   ]
   Mech.Sounds []
   OnStopFire.Sounds []
   Suppressor.Sounds [
    /Game/Sounds/Fort_Weapons/Suppressor/Silencer_Pistol_Close_05
    /Game/Sounds/Fort_Weapons/Suppressor/Silencer_Pistol_Close_04
    /Game/Sounds/Fort_Weapons/Suppressor/Silencer_Pistol_Close_03
    /Game/Sounds/Fort_Weapons/Suppressor/Silencer_Pistol_Close_02
    /Game/Sounds/Fort_Weapons/Suppressor/Silencer_Pistol_Close_01
   ]
   Sweetener.Sounds []
  }}]
properties {
 assetImportData=&self:AssetImportData
 numChannels=6
 outputFormat=FiveDotOne
 qualitySetting=Fort_Weapons
 sampleRate=48000
 soundClassObject=&Fort_WeaponRanged_Fire_1P_PSM
 concurrencyOverrides {maxCount=2}
 concurrencySet [&Weapon_Gunfire_Local_Conc]
}
