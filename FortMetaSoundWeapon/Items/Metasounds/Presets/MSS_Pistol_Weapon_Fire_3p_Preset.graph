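# MetaSound Source preset for third-person pistol gunfire, inheriting its graph from the legacy MSS_Base_Weapon_Fire_3p parent.
# It supplies machine pistol close-mech shots, shared pistol medium and distant tails, and pistol silencer layers, alongside a ranged-weapon distance curve bank.
# Exposed inputs tune distance, elevation and target EQ, random pitch, volume and EQ, and suppressor and friendly-fire mixing.
# Stereo output uses projectile attenuation and local and global gunfire concurrency.
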MetaSoundSource /FortMetaSoundWeapon/Items/Metasounds/Presets/MSS_Pistol_Weapon_Fire_3p_Preset
preset_parent=/FortMetaSoundWeapon/Items/Metasounds/Legacy/MSS_Base_Weapon_Fire_3p
inputs [
 bAutoFire {type=Bool}
 Character.Affiliation {type=Enum:FortAudioAffiliation}
 Character.Health {type=Float}
 Character.UniqueID {type=Int32}
 DistanceEQ.bEnable {type=Bool}
 DistanceEQ.FilterMaxGain {type=Float}
 DistanceEQ.FilterMinGain {type=Float}
 DistanceEQ.HighShelfCutoffFrequency {type=Float}
 DistanceEQ.LowShelfCutoffFrequency {type=Float}
 Elevation.AboveGain {type=Float}
 Elevation.BelowGain {type=Float}
 Elevation.CutoffFrequency {type=Float}
 Elevation.LowCutGain {type=Float}
 MainFire.DistanceCurves {type=WaveTableBankAsset}
 MainFire.FalloffDistance {type=Float}
 MainFire.SoundsClose {type=WaveAsset:Array}
 MainFire.SoundsDistant {type=WaveAsset:Array}
 MainFire.SoundsMedium {type=WaveAsset:Array}
 MainFire.StopTriggerDelay {type=Time}
 Mixing.CloseGain {type=Float}
 Mixing.DistantGain {type=Float}
 Mixing.Friendly.FriendlyFireWeaponVolume {type=Float}
 Mixing.MediumGain {type=Float}
 Mixing.Suppressor.SuppressorOutputVolume {type=Float}
 Mixing.WeaponOutputVolume {type=Float}
 OnStop {type=Trigger}
 Random.EQ.A.MaxFreq {type=Float}
 Random.EQ.A.MaxGain {type=Float}
 Random.EQ.A.MinFreq {type=Float}
 Random.EQ.A.MinGain {type=Float}
 Random.EQ.B.MaxFreq {type=Float}
 Random.EQ.B.MaxGain {type=Float}
 Random.EQ.B.MinFreq {type=Float}
 Random.EQ.B.MinGain {type=Float}
 Random.MaxPitch {type=Float}
 Random.MaxVolume {type=Float}
 Random.MinPitch {type=Float}
 Random.MinVolume {type=Float}
 ReTrigger {type=Trigger}
 SharedTails.AttackEnvelopeCurve {type=Float}
 SharedTails.AttackEnvelopeTime {type=Time}
 SharedTails.bEnable {type=Bool}
 SharedTails.DecayEnvelopeCurve {type=Float}
 SharedTails.DecayEnvelopeTime {type=Time}
 SharedTails.TailFadeInTime {type=Float}
 SharedTails.WeaponAttackGain {type=Float}
 Suppressor.BaseWeaponVolumeMultiplier {type=Float}
 Suppressor.EnvelopeAttackTime {type=Time}
 Suppressor.EnvelopeDecayTime {type=Time}
 Suppressor.Sounds.Close {type=WaveAsset:Array}
 Suppressor.Sounds.Medium {type=WaveAsset:Array}
 Suppressor.Volume.Close {type=Float}
 Suppressor.Volume.CloseBase {type=Float}
 Suppressor.Volume.Medium {type=Float}
 Suppressor.Volume.MediumBase {type=Float}
 TargetEQ.NonTargetGain {type=Float}
 UE.Attenuation.Distance {type=Float}
 UE.Source.OnPlay {type=Trigger}
 UE.Source.Orientation.Azimuth {type=Float}
 UE.Source.Orientation.Elevation {type=Float}
 UE.Spatialization.Azimuth {type=Float}
 UE.Spatialization.Elevation {type=Float}
 Weapon.AmmoRemaining {type=Int32}
 Weapon.ClipSize {type=Int32}
 Weapon.FireRate {type=Float}
 Weapon.FireStartTime {type=Float}
 Weapon.IsAimingDownSights {type=Bool}
 Weapon.IsSuppressorEnabled {type=Bool}
]
interfaces [
 Fort.Character:1.0
 Fort.Weapon:1.0
 UE.Attenuation:1.0
 UE.OutputFormat.Stereo:1.0
 UE.Source.OneShot:1.0
 UE.Source.Orientation:1.0
 UE.Source:1.1
 UE.Spatialization:1.0
]
outputs [
 bOnAlmostFinished {type=Bool}
 CloseCurveEvaluation {type=Float}
 MediumCurveEvaluation {type=Float}
 OnGunshotOut {type=Trigger}
 "UE.OutputFormat.Stereo.Audio:0" {type=Audio}
 "UE.OutputFormat.Stereo.Audio:1" {type=Audio}
 UE.Source.OneShot.OnFinished {type=Trigger}
]
pages [Default {input_defaults {
   Character.Affiliation=0
   Character.Health=1
   Character.UniqueID=0
   DistanceEQ.FilterMaxGain=1.5
   DistanceEQ.FilterMinGain=-6
   DistanceEQ.HighShelfCutoffFrequency=11000
   DistanceEQ.LowShelfCutoffFrequency=300
   DistanceEQ.bEnable="<MetaSoundBuilderResult.SUCCEEDED: 0>"
   Elevation.AboveGain=0.4
   Elevation.BelowGain=-1
   Elevation.CutoffFrequency=1500
   Elevation.LowCutGain=-6
   MainFire.DistanceCurves=&WeaponRanged_DistanceCurves
   MainFire.FalloffDistance=27500
   MainFire.StopTriggerDelay=0.03
   Mixing.CloseGain=1
   Mixing.DistantGain=0.6
   Mixing.Friendly.FriendlyFireWeaponVolume=0.7
   Mixing.MediumGain=0.7
   Mixing.Suppressor.SuppressorOutputVolume=1
   Mixing.WeaponOutputVolume=1
   OnStop=null
   Random.EQ.A.MaxFreq=20000
   Random.EQ.A.MaxGain=0
   Random.EQ.A.MinFreq=699.9999
   Random.EQ.A.MinGain=-3
   Random.EQ.B.MaxFreq=8000
   Random.EQ.B.MaxGain=0
   Random.EQ.B.MinFreq=700
   Random.EQ.B.MinGain=-1
   Random.MaxPitch=1.05
   Random.MaxVolume=1.15
   Random.MinPitch=0.95
   Random.MinVolume=0.85
   ReTrigger=null
   SharedTails.AttackEnvelopeCurve=2
   SharedTails.AttackEnvelopeTime=0.01
   SharedTails.DecayEnvelopeCurve=1.5
   SharedTails.DecayEnvelopeTime=0.1
   SharedTails.TailFadeInTime=0.06
   SharedTails.WeaponAttackGain=0.16
   SharedTails.bEnable=null
   Suppressor.BaseWeaponVolumeMultiplier=0.5
   Suppressor.EnvelopeAttackTime=0
   Suppressor.EnvelopeDecayTime=0.195
   Suppressor.Volume.Close=0.5
   Suppressor.Volume.CloseBase=1
   Suppressor.Volume.Medium=0.87
   Suppressor.Volume.MediumBase=0.3
   TargetEQ.NonTargetGain=-2
   UE.Attenuation.Distance=25000
   UE.Source.OnPlay=null
   UE.Source.Orientation.Azimuth=0
   UE.Source.Orientation.Elevation=0
   UE.Spatialization.Azimuth=0
   UE.Spatialization.Elevation=0
   Weapon.AmmoRemaining=15
   Weapon.ClipSize=15
   Weapon.FireRate=0.4
   Weapon.FireStartTime=0
   Weapon.IsAimingDownSights=null
   Weapon.IsSuppressorEnabled=null
   bAutoFire=null
   MainFire.SoundsClose [
    /Game/Sounds/Fort_Weapons/pistol/MachinePistol/SoundWaves/Machine_Pistol2_Fire_CloseMech_01
    /Game/Sounds/Fort_Weapons/pistol/MachinePistol/SoundWaves/Machine_Pistol2_Fire_CloseMech_02
    /Game/Sounds/Fort_Weapons/pistol/MachinePistol/SoundWaves/Machine_Pistol2_Fire_CloseMech_03
    /Game/Sounds/Fort_Weapons/pistol/MachinePistol/SoundWaves/Machine_Pistol2_Fire_CloseMech_04
    /Game/Sounds/Fort_Weapons/pistol/MachinePistol/SoundWaves/Machine_Pistol2_Fire_CloseMech_05
    /Game/Sounds/Fort_Weapons/pistol/MachinePistol/SoundWaves/Machine_Pistol2_Fire_CloseMech_06
    /Game/Sounds/Fort_Weapons/pistol/MachinePistol/SoundWaves/Machine_Pistol2_Fire_CloseMech_07
   ]
   MainFire.SoundsDistant [
    /Game/Sounds/Fort_Weapons/Tails/Tails_Pistols_Distant_A_01
    /Game/Sounds/Fort_Weapons/Tails/Tails_Pistols_Distant_A_02
    /Game/Sounds/Fort_Weapons/Tails/Tails_Pistols_Distant_A_03
    /Game/Sounds/Fort_Weapons/Tails/Tails_Pistols_Distant_A_04
    /Game/Sounds/Fort_Weapons/Tails/Tails_Pistols_Distant_A_05
    /Game/Sounds/Fort_Weapons/Tails/Tails_Pistols_Distant_A_06
    /Game/Sounds/Fort_Weapons/Tails/Tails_Pistols_Distant_A_07
    /Game/Sounds/Fort_Weapons/Tails/Tails_Pistols_Distant_A_08
   ]
   MainFire.SoundsMedium [
    /Game/Sounds/Fort_Weapons/Tails/Tails_Pistols_Medium_A_01
    /Game/Sounds/Fort_Weapons/Tails/Tails_Pistols_Medium_A_02
    /Game/Sounds/Fort_Weapons/Tails/Tails_Pistols_Medium_A_03
    /Game/Sounds/Fort_Weapons/Tails/Tails_Pistols_Medium_A_04
    /Game/Sounds/Fort_Weapons/Tails/Tails_Pistols_Medium_A_05
    /Game/Sounds/Fort_Weapons/Tails/Tails_Pistols_Medium_A_06
    /Game/Sounds/Fort_Weapons/Tails/Tails_Pistols_Medium_A_07
    /Game/Sounds/Fort_Weapons/Tails/Tails_Pistols_Medium_A_08
   ]
   Suppressor.Sounds.Close [
    /Game/Sounds/Fort_Weapons/Suppressor/Silencer_Pistol_Close_01
    /Game/Sounds/Fort_Weapons/Suppressor/Silencer_Pistol_Close_02
    /Game/Sounds/Fort_Weapons/Suppressor/Silencer_Pistol_Close_03
    /Game/Sounds/Fort_Weapons/Suppressor/Silencer_Pistol_Close_04
    /Game/Sounds/Fort_Weapons/Suppressor/Silencer_Pistol_Close_05
   ]
   Suppressor.Sounds.Medium [
    /Game/Sounds/Fort_Weapons/Suppressor/Silencer_Pistol_Medium_01
    /Game/Sounds/Fort_Weapons/Suppressor/Silencer_Pistol_Medium_02
    /Game/Sounds/Fort_Weapons/Suppressor/Silencer_Pistol_Medium_03
    /Game/Sounds/Fort_Weapons/Suppressor/Silencer_Pistol_Medium_04
    /Game/Sounds/Fort_Weapons/Suppressor/Silencer_Pistol_Medium_05
   ]
  }}]
properties {
 assetImportData=&self:AssetImportData
 attenuationSettings=&Fort_Weapon_Projectile_NonAtt_Athena_Att
 maxDistance=27500
 numChannels=2
 outputFormat=Stereo
 qualitySetting=Fort_Weapons
 sampleRate=48000
 soundClassObject=&Fort_WeaponRanged_Fire_3P_PSM
 concurrencySet [&Weapon_Gunfire_Local_Conc, &Weapon_Gunfire_Global_Conc]
}
