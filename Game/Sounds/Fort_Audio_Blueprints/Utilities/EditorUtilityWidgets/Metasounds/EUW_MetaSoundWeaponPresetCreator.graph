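# Editor utility widget that builds 1p or 3p weapon-fire MetaSound source presets from selected SoundWave assets.
# Buttons set base sources, either generic or per weapon class, and a preset builder references the chosen one.
# Waves are sorted by name substring into category arrays, then written as graph input defaults alongside an auto-fire flag.
# BuildToAsset saves the suffixed preset at the entered path, then its quality setting becomes Fort_Weapons.
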EditorUtilityWidgetBlueprint /Game/Sounds/Fort_Audio_Blueprints/Utilities/EditorUtilityWidgets/Metasounds/EUW_MetaSoundWeaponPresetCreator
graphs [
 AssignWaveAssets {nodes [
   n1 = AssignWaveAssets!() -> (then, InputName:String, WaveArray:Array of Object References) {is=title K2Node_FunctionEntry}
   > n2 = Branch(Condition="integer > integer"(A=Length(TargetArray:Array of Object References=@n1.WaveArray)))
   > n3 = Create MetaSound Literal From AudioParameter!(
    execute
    self:Meta Sound Builder Subsystem Object Reference=Get MetaSoundBuilderSubsystem() -> (ReturnValue:Meta Sound Builder Subsystem Object Reference) {is=title K2Node_GetEngineSubsystem}
    Param:"Audio Parameter Structure (by ref)"=ObjectArrayToAudioParameter!(Name:Name=String To Name(InString=@n1), Objects:Array of Object References=@n1.WaveArray) -> (ReturnValue:Audio Parameter Structure)
   ) -> (then, ReturnValue:Meta Sound Literal Structure)
   > n4 = SetGraphInputDefault!(
    execute
    self:Meta Sound Builder Base Object Reference=Get MetasoundBuilder() -> (MetasoundBuilder:Meta Sound Source Builder Object Reference)
    InputName:Name=String To Name(InString=@n1)
    Literal:"Meta Sound Literal Structure (by ref)"=@n3
    PageName:Name
   ) -> (Succeeded:Exec, Failed:Exec)
   Return Node!(execute=@n4.Failed, bSuccess:bool=false)
   > Return Node(bSuccess=true)
   Return Node(execute=@n2.else, bSuccess=false)
  ]}
 ClearArrays {nodes [
   ClearArrays() {is=title K2Node_FunctionEntry}
   > Clear(TargetArray:Array of Object References=Get CloseWaveArray() -> (CloseWaveArray:Array of Object References))
   > Clear(TargetArray=Get MediumWaveArray() -> (MediumWaveArray:Array of Object References))
   > Clear(TargetArray=Get DistantWaveArray() -> (DistantWaveArray:Array of Object References))
   > Clear(TargetArray=Get Close1pWaveArray() -> (Close1pWaveArray:Array of Object References))
   > Clear(TargetArray=Get Mech1pWaveArray() -> (Mech1pWaveArray:Array of Object References))
   > Clear(TargetArray=Get Sweetener1pWaveArray() -> (Sweetener1pWaveArray:Array of Object References))
   > Clear(TargetArray=Get OnStop1pWaveArray() -> (OnStop1pWaveArray:Array of Object References))
   > Clear(TargetArray=Get LowAmmo1pWaveArray() -> (LowAmmo1pWaveArray:Array of Object References))
  ]}
 EventGraph {nodes [
   "On Pressed (EditorUtilityButton_56)"() {is=title K2Node_ComponentBoundEvent}
   > n1 = Set bIs1p(bIs1p:bool=true) -> (Output_Get:bool)
   ClearArrays(execute=[@n1.then, @n20.then]) {is=title K2Node_CallFunction}
   > n2 = Cast To MetaSoundDocumentInterface!(
    execute
    Object=Select(
     Option 0:Meta Sound Source Object Reference=Get SourceReference3p() -> (SourceReference3p:Meta Sound Source Object Reference)
     Option 1:Meta Sound Source Object Reference=Get SourceReference1p() -> (SourceReference1p:Meta Sound Source Object Reference)
     Index=Get bIs1p() -> (bIs1p:bool)
    ) -> (ReturnValue:Meta Sound Source Object Reference)
   ) -> (then, CastFailed, AsMetaSound Document Interface:Meta Sound Document Interface Interface) {is=title K2Node_DynamicCast}
   n3 = Select(Option 0:String=_3p, Option 1:String=_1p, Index=Get bIs1p()) -> (ReturnValue:String)
   > n4 = CreateSourcePresetBuilder!(
    execute
    self:Meta Sound Builder Subsystem Object Reference=Get MetaSoundBuilderSubsystem()
    BuilderName:Name=String To Name(InString=Append(A=Builder, B=@n3))
    ReferencedSourceClass:"Meta Sound Document Interface Interface (by ref)"=@n2
   ) -> (Succeeded:Exec, Failed:Exec, ReturnValue:Meta Sound Source Builder Object Reference)
   > Set MetasoundBuilder(MetasoundBuilder:Meta Sound Source Builder Object Reference=@n4) -> (Output_Get:Meta Sound Source Builder Object Reference)
   > n5 = GetSelectedAssetsOfClass(AssetClass=/Script/Engine.SoundWave)
   > n6 = Branch(Condition=Get bIs1p())
   > n7 = For Each Loop(Array:Array of Object References=@n5) -> (Array Element:Object Reference)
   AssignWaveAssets!(execute=@n7.Completed, self, InputName:String=MainFire.Sounds, WaveArray:Array of Object References=Get Close1pWaveArray()) -> (then, bSuccess:bool) {is=title K2Node_CallFunction}
   > AssignWaveAssets(InputName=LowAmmo.Sounds, WaveArray=Get LowAmmo1pWaveArray())
   > AssignWaveAssets(InputName=Mech.Sounds, WaveArray=Get Mech1pWaveArray())
   > AssignWaveAssets(InputName=Sweetener.Sounds, WaveArray=Get Sweetener1pWaveArray())
   > n8 = AssignWaveAssets(InputName=OnStopFire.Sounds, WaveArray=Get OnStop1pWaveArray())
   SetFireType(execute=[@n8.then, @n15.then])
   n9 = Get PathInput() -> (PathInput:Editor Utility Editable Text Box Object Reference)
   n10 = Get Text(self=Get AssetNameInput() -> (AssetNameInput:Editor Utility Editable Text Box Object Reference))
   n11 = Get bIs1p()
   > n12 = BuildToAsset!(
    execute
    self:Meta Sound Editor Subsystem Object Reference=Get MetaSoundEditorSubsystem() -> (ReturnValue:Meta Sound Editor Subsystem Object Reference) {is=title K2Node_GetEditorSubsystem}
    InBuilder:Meta Sound Builder Base Object Reference=Get MetasoundBuilder()
    Author:String=SoundBaseAssetAction
    AssetName:String=Append(A="To String (Text)"(InText=@n10), B=Select(Option 0=_3p_Preset, Option 1=_1p_Preset, Index=@n11))
    PackagePath:String="To String (Text)"(InText=Get Text(self=@n9))
    TemplateSoundWave:Sound Wave Object Reference
   ) -> (Succeeded:Exec, Failed:Exec, ReturnValue:Meta Sound Document Interface Interface)
   > n13 = Cast To MetaSoundSource!(execute, Object:Meta Sound Document Interface Interface=@n12) -> (then, CastFailed, AsMetaSound Source:Meta Sound Source Object Reference) {is=title K2Node_DynamicCast}
   > Set QualitySetting!(execute, QualitySetting:Name=Fort_Weapons, self:Meta Sound Source Object Reference=@n13) -> (then, Output_Get:Name=None)
   SortWaveAssets!(execute=@n7.LoopBody, self, Wave:Object Reference=@n7, WaveArray:Array of Object References=Get Close1pWaveArray(), Contains:String=close) -> (then)
   > SortWaveAssets(Wave=@n7, WaveArray=Get LowAmmo1pWaveArray(), Contains=lowammo)
   > SortWaveAssets(Wave=@n7, WaveArray=Get Mech1pWaveArray(), Contains=mech)
   > SortWaveAssets(Wave=@n7, WaveArray=Get Sweetener1pWaveArray(), Contains=sweetener)
   > SortWaveAssets(Wave=@n7, WaveArray=Get OnStop1pWaveArray(), Contains=onstop)
   n14 = For Each Loop(Exec=@n6.else, Array=@n5)
   AssignWaveAssets(execute=@n14.Completed, InputName=MainFire.SoundsClose, WaveArray=Get CloseWaveArray())
   > AssignWaveAssets(InputName=MainFire.SoundsMedium, WaveArray=Get MediumWaveArray())
   > n15 = AssignWaveAssets(InputName=MainFire.SoundsDistant, WaveArray=Get DistantWaveArray())
   SortWaveAssets(execute=@n14.LoopBody, Wave=@n14, WaveArray=Get CloseWaveArray(), Contains=close)
   > SortWaveAssets(Wave=@n14, WaveArray=Get MediumWaveArray(), Contains=medium)
   > SortWaveAssets(Wave=@n14, WaveArray=Get DistantWaveArray(), Contains=distant)
   "On Pressed (EditorUtilityButton)"() {is=title K2Node_ComponentBoundEvent}
   > Set Text(
    Text="To Text (String)"(InString=GetCurrentContentBrowserPath!() -> (OutPath:String, ReturnValue:bool))
    self:Text Box Object Reference=Get PathInput()
   )
   "On Pressed (SetBaseShotgun)"() {is=title K2Node_ComponentBoundEvent}
   > n16 = Get!(execute:Exec) -> (then:Exec, else:Exec, Default1pShotgun:Meta Sound Source Object Reference)
   > Set SourceReference1p(SourceReference1p:Meta Sound Source Object Reference=@n16) -> (Output_Get:Meta Sound Source Object Reference)
   > n17 = Get!(execute) -> (then, else, Default3pShotgun:Meta Sound Source Object Reference)
   > Set SourceReference3p(SourceReference3p:Meta Sound Source Object Reference=@n17) -> (Output_Get:Meta Sound Source Object Reference)
   "On Pressed (SetBaseSniper)"() {is=title K2Node_ComponentBoundEvent}
   > n18 = Get!(execute) -> (then, else, Default1pSniper:Meta Sound Source Object Reference)
   > Set SourceReference1p(SourceReference1p=@n18)
   > n19 = Get!(execute) -> (then, else, Source Reference 3p:Meta Sound Source Object Reference)
   > Set SourceReference3p(SourceReference3p=@n19)
   "On Pressed (EditorUtilityButton_1)"() {is=title K2Node_ComponentBoundEvent}
   > n20 = Set bIs1p(bIs1p=false)
   "On Pressed (SetBase1p)"() {is=title K2Node_ComponentBoundEvent}
   > n21 = Get!(execute) -> (then, else, Default1pMSS:Meta Sound Source Object Reference)
   > Set SourceReference1p(SourceReference1p=@n21)
   "On Pressed (SetBase3p)"() {is=title K2Node_ComponentBoundEvent}
   > n22 = Get!(execute) -> (then, else, Default3pMSS:Meta Sound Source Object Reference)
   > Set SourceReference3p(SourceReference3p=@n22)
   n23 = "On Check State Changed (EditorUtilityCheckBox_0)"!() -> (OutputDelegate, then, bIsChecked:bool) {is=title K2Node_ComponentBoundEvent}
   > Set bAutoFire(bAutoFire:bool=@n23.bIsChecked) -> (Output_Get:bool)
   "On Pressed (SetBasePistol)"() {is=title K2Node_ComponentBoundEvent}
   > n24 = Get!(execute) -> (then, else, Default1pPistol:Meta Sound Source Object Reference)
   > Set SourceReference1p(SourceReference1p=@n24)
   > n25 = Get!(execute) -> (then, else, Default3pPistol:Meta Sound Source Object Reference)
   > Set SourceReference3p(SourceReference3p=@n25)
   "On Pressed (SetBaseSMG)"() {is=title K2Node_ComponentBoundEvent}
   > n26 = Get!(execute) -> (then, else, Default1pSMG:Meta Sound Source Object Reference)
   > Set SourceReference1p(SourceReference1p=@n26)
   > n27 = Get!(execute) -> (then, else, Default3pSMG:Meta Sound Source Object Reference)
   > Set SourceReference3p(SourceReference3p=@n27)
   "On Pressed (SetBaseRifle)"() {is=title K2Node_ComponentBoundEvent}
   > n28 = Get!(execute) -> (then, else, Default1pRifle:Meta Sound Source Object Reference)
   > Set SourceReference1p(SourceReference1p=@n28)
   > n29 = Get!(execute) -> (then, else, Default3pRifle:Meta Sound Source Object Reference)
   > Set SourceReference3p(SourceReference3p=@n29)
   Event PreConstruct()
   > SetObject(self=Get MetaSoundSource1p() -> (MetaSoundSource1p:Single Property View Object Reference), NewObject=Self-Reference())
   > SetObject(self=Get MetaSoundSource3p() -> (MetaSoundSource3p:Single Property View Object Reference), NewObject=Self-Reference())
  ]}
 SetFireType {nodes [
   SetFireType() {is=title K2Node_FunctionEntry}
   > n1 = Create MetaSound Bool Literal!(execute, self:Meta Sound Builder Subsystem Object Reference=Get MetaSoundBuilderSubsystem(), Value:bool=Get bAutoFire() -> (bAutoFire:bool)) -> (then, DataType:Name=None, ReturnValue:Meta Sound Literal Structure)
   > SetGraphInputDefault(self=Get MetasoundBuilder(), InputName=bAutoFire, Literal=@n1.ReturnValue)
  ]}
 SortWaveAssets {nodes [
   n1 = SortWaveAssets!() -> (then, Wave:Object Reference, WaveArray:Array of Object References, Contains:String) {is=title K2Node_FunctionEntry}
   > Branch(Condition=Contains(SearchIn=GetObjectName(Object=@n1), Substring=@n1.Contains))
   > Add Unique(TargetArray:Array of Object References=@n1.WaveArray, NewItem:"Object Reference (by ref)"=@n1)
  ]}
]
variables [
 bAutoFire {type=boolean, default=true}
 bIs1p {type=boolean, default=true}
 Close1pWaveArray {type="array<Object>", default []}
 CloseWaveArray {type="array<Object>", default []}
 Default1pMSS {type=MetaSoundSource, default=/FortMetaSoundWeapon/Items/Metasounds/Presets/MSS_Base_Weapon_Fire_1p_Preset, class=~, properties=~}
 Default1pPistol {type=MetaSoundSource, default=/FortMetaSoundWeapon/Items/Metasounds/Presets/MSS_Pistol_Weapon_Fire_1p_Preset, class=~, properties=~}
 Default1pRifle {type=MetaSoundSource, default=/FortMetaSoundWeapon/Items/Metasounds/Presets/MSS_Rifle_Weapon_Fire_1p_Preset, class=~, properties=~}
 Default1pShotgun {type=MetaSoundSource, default=/FortMetaSoundWeapon/Items/Metasounds/Presets/MSS_Shotgun_Weapon_Fire_1p_Preset, class=~, properties=~}
 Default1pSMG {type=MetaSoundSource, default=/FortMetaSoundWeapon/Items/Metasounds/Presets/MSS_SMG_Weapon_Fire_1p_Preset, class=~, properties=~}
 Default1pSniper {type=MetaSoundSource, default=/FortMetaSoundWeapon/Items/Metasounds/Presets/MSS_Sniper_Weapon_Fire_1p_Preset, class=~, properties=~}
 Default3pMSS {type=MetaSoundSource, default=/FortMetaSoundWeapon/Items/Metasounds/Presets/MSS_Base_Weapon_Fire_3p_Preset, class=~, properties=~}
 Default3pPistol {type=MetaSoundSource, default=/FortMetaSoundWeapon/Items/Metasounds/Presets/MSS_Pistol_Weapon_Fire_3p_Preset, class=~, properties=~}
 Default3pRifle {type=MetaSoundSource, default=/FortMetaSoundWeapon/Items/Metasounds/Presets/MSS_Rifle_Weapon_Fire_3p_Preset, class=~, properties=~}
 Default3pShotgun {type=MetaSoundSource, default=/FortMetaSoundWeapon/Items/Metasounds/Presets/MSS_Shotgun_Weapon_Fire_3p_Preset, class=~, properties=~}
 Default3pSMG {type=MetaSoundSource, default=/FortMetaSoundWeapon/Items/Metasounds/Presets/MSS_SMG_Weapon_Fire_3p_Preset, class=~, properties=~}
 DistantWaveArray {type="array<Object>", default []}
 LowAmmo1pWaveArray {type="array<Object>", default []}
 Mech1pWaveArray {type="array<Object>", default []}
 MediumWaveArray {type="array<Object>", default []}
 MetasoundBuilder {type=MetaSoundSourceBuilder, default=null}
 OnStop1pWaveArray {type="array<Object>", default []}
 Source Reference 3p {type=MetaSoundSource, default=/FortMetaSoundWeapon/Items/Metasounds/Presets/MSS_Sniper_Weapon_Fire_3p_Preset, class=~, properties=~}
 SourceReference1p {type=MetaSoundSource, default=null, class=~, properties=~}
 SourceReference3p {type=MetaSoundSource, default=null, class=~, properties=~}
 Sweetener1pWaveArray {type="array<Object>", default []}
]
widgets {
 info {rootWidgetClass=&CanvasPanel, widgetCount=30}
 widgets [
  CanvasPanel_202 {widgetClassPath=&CanvasPanel, widget=CanvasPanel_202}
  EditorUtilityButton_56 {
   widgetClassPath=&EditorUtilityButton
   parent=CanvasPanel_202
   widget=EditorUtilityButton_56
   properties {toolTipText=Once all wave assets are selected make 1p weapon preset}
   slot {
    class=CanvasPanelSlot
    properties {layoutData {alignment=<0 0>, anchors {maximum=<0 0>, minimum=<0 0>}, offsets=<ltrb 24 440 272.6727 51.7718>}}
   }
  }
  UEFN_TextBlock_C_39 {
   widgetClassPath=&UEFN_TextBlock
   parent=EditorUtilityButton_56
   widget=UEFN_TextBlock_C_39
   properties {
    justification=Center
    text=Make 1p Preset
    font {
     bForceMonospaced=false
     bMaterialIsStencil=false
     fontMaterial=null
     fontObject=&BurbankSmall
     letterSpacing=0
     monospacedWidth=1
     size=9
     skewAmount=0
     typefaceFontName=Black
     outlineSettings {
      bApplyOutlineToDropShadows=false
      bMiteredCorners=false
      bSeparateFillAlpha=false
      outlineMaterial=null
      outlineSize=0
      outlineColor=<rgba 0 0 0 1>
     }
    }
   }
   slot {class=ButtonSlot}
  }
  EditorUtilityButton_1 {
   widgetClassPath=&EditorUtilityButton
   parent=CanvasPanel_202
   widget=EditorUtilityButton_1
   properties {toolTipText=Once all wave assets are selected make 3p weapon preset}
   slot {
    class=CanvasPanelSlot
    properties {layoutData {alignment=<0 0>, anchors {maximum=<0 0>, minimum=<0 0>}, offsets=<ltrb 320 440 272.6727 51.7718>}}
   }
  }
  UEFN_TextBlock_C_3 {
   widgetClassPath=&UEFN_TextBlock
   parent=EditorUtilityButton_1
   widget=UEFN_TextBlock_C_3
   properties {
    justification=Center
    text=Make 3p Preset
    font {
     bForceMonospaced=false
     bMaterialIsStencil=false
     fontMaterial=null
     fontObject=&BurbankSmall
     letterSpacing=0
     monospacedWidth=1
     size=9
     skewAmount=0
     typefaceFontName=Black
     outlineSettings {
      bApplyOutlineToDropShadows=false
      bMiteredCorners=false
      bSeparateFillAlpha=false
      outlineMaterial=null
      outlineSize=0
      outlineColor=<rgba 0 0 0 1>
     }
    }
   }
   slot {class=ButtonSlot}
  }
  MetaSoundSource1p {
   widgetClassPath=&SinglePropertyView
   parent=CanvasPanel_202
   widget=MetaSoundSource1p
   properties {propertyName=SourceReference1p}
   slot {
    class=CanvasPanelSlot
    properties {layoutData {alignment=<0 0>, anchors {maximum=<0 0>, minimum=<0 0>}, offsets=<ltrb 0 40 506.4064 62.032>}}
   }
  }
  MetaSoundSource3p {
   widgetClassPath=&SinglePropertyView
   parent=CanvasPanel_202
   widget=MetaSoundSource3p
   properties {propertyName=SourceReference3p}
   slot {
    class=CanvasPanelSlot
    properties {layoutData {alignment=<0 0>, anchors {maximum=<0 0>, minimum=<0 0>}, offsets=<ltrb 0 120 506.4064 62.032>}}
   }
  }
  PathInput {
   widgetClassPath=&EditorUtilityEditableTextBox
   bIsVariable=true
   parent=CanvasPanel_202
   widget=PathInput
   slot {
    class=CanvasPanelSlot
    properties {layoutData {alignment=<0 0>, anchors {maximum=<0 0>, minimum=<0 0>}, offsets=<ltrb 164 284 313.964 28.4985>}}
   }
  }
  UEFN_TextBlock_C {
   widgetClassPath=&UEFN_TextBlock
   parent=CanvasPanel_202
   widget=UEFN_TextBlock_C
   properties {
    text=Path
    font {
     bForceMonospaced=false
     bMaterialIsStencil=false
     fontMaterial=null
     fontObject=&Roboto
     letterSpacing=0
     monospacedWidth=1
     size=10.5
     skewAmount=0
     typefaceFontName=Regular
     outlineSettings {
      bApplyOutlineToDropShadows=false
      bMiteredCorners=false
      bSeparateFillAlpha=false
      outlineMaterial=null
      outlineSize=0
      outlineColor=<rgba 0 0 0 1>
     }
    }
   }
   slot {
    class=CanvasPanelSlot
    properties {layoutData {alignment=<0 0>, anchors {maximum=<0 0>, minimum=<0 0>}, offsets=<ltrb 16 284 101.1862 19.2282>}}
   }
  }
  EditorUtilityButton {
   widgetClassPath=&EditorUtilityButton
   parent=CanvasPanel_202
   widget=EditorUtilityButton
   properties {toolTipText=Set current editor path as the preset build destination}
   slot {class=CanvasPanelSlot, properties {layoutData {alignment=<0 0>, anchors {maximum=<0 0>, minimum=<0 0>}, offsets=<ltrb 492 284 100 30>}}}
  }
  UEFN_TextBlock_C_1 {
   widgetClassPath=&UEFN_TextBlock
   parent=EditorUtilityButton
   widget=UEFN_TextBlock_C_1
   properties {
    justification=Center
    text=Get Current Path
    font {
     bForceMonospaced=false
     bMaterialIsStencil=false
     fontMaterial=null
     fontObject=&Roboto
     letterSpacing=0
     monospacedWidth=1
     size=9
     skewAmount=0
     typefaceFontName=Regular
     outlineSettings {
      bApplyOutlineToDropShadows=false
      bMiteredCorners=false
      bSeparateFillAlpha=false
      outlineMaterial=null
      outlineSize=0
      outlineColor=<rgba 0 0 0 1>
     }
    }
   }
   slot {class=ButtonSlot}
  }
  SetBase1p {
   widgetClassPath=&EditorUtilityButton
   parent=CanvasPanel_202
   widget=SetBase1p
   properties {toolTipText=Auto set default weapon 1p metasound as base for this preset}
   slot {class=CanvasPanelSlot, properties {layoutData {alignment=<0 0>, anchors {maximum=<0 0>, minimum=<0 0>}, offsets=<ltrb 492 44 100 30>}}}
  }
  UEFN_TextBlock_C_4 {
   widgetClassPath=&UEFN_TextBlock
   parent=SetBase1p
   widget=UEFN_TextBlock_C_4
   properties {
    justification=Center
    text=Set Base 1p
    font {
     bForceMonospaced=false
     bMaterialIsStencil=false
     fontMaterial=null
     fontObject=&Roboto
     letterSpacing=0
     monospacedWidth=1
     size=9
     skewAmount=0
     typefaceFontName=Regular
     outlineSettings {
      bApplyOutlineToDropShadows=false
      bMiteredCorners=false
      bSeparateFillAlpha=false
      outlineMaterial=null
      outlineSize=0
      outlineColor=<rgba 0 0 0 1>
     }
    }
   }
   slot {class=ButtonSlot}
  }
  SetBasePistol {
   widgetClassPath=&EditorUtilityButton
   parent=CanvasPanel_202
   widget=SetBasePistol
   properties {toolTipText=Auto set default pistol metasounds as base for this preset}
   slot {class=CanvasPanelSlot, properties {layoutData {alignment=<0 0>, anchors {maximum=<0 0>, minimum=<0 0>}, offsets=<ltrb 108 220 100 30>}}}
  }
  UEFN_TextBlock_C_8 {
   widgetClassPath=&UEFN_TextBlock
   parent=SetBasePistol
   widget=UEFN_TextBlock_C_8
   properties {
    justification=Center
    text=Set Base Pistol
    font {
     bForceMonospaced=false
     bMaterialIsStencil=false
     fontMaterial=null
     fontObject=&Roboto
     letterSpacing=0
     monospacedWidth=1
     size=9
     skewAmount=0
     typefaceFontName=Regular
     outlineSettings {
      bApplyOutlineToDropShadows=false
      bMiteredCorners=false
      bSeparateFillAlpha=false
      outlineMaterial=null
      outlineSize=0
      outlineColor=<rgba 0 0 0 1>
     }
    }
   }
   slot {class=ButtonSlot}
  }
  SetBaseSMG {
   widgetClassPath=&EditorUtilityButton
   parent=CanvasPanel_202
   widget=SetBaseSMG
   properties {toolTipText=Auto set default SMG metasounds as base for this preset}
   slot {class=CanvasPanelSlot, properties {layoutData {alignment=<0 0>, anchors {maximum=<0 0>, minimum=<0 0>}, offsets=<ltrb 220 220 100 30>}}}
  }
  UEFN_TextBlock_C_9 {
   widgetClassPath=&UEFN_TextBlock
   parent=SetBaseSMG
   widget=UEFN_TextBlock_C_9
   properties {
    justification=Center
    text=Set Base SMG
    font {
     bForceMonospaced=false
     bMaterialIsStencil=false
     fontMaterial=null
     fontObject=&Roboto
     letterSpacing=0
     monospacedWidth=1
     size=9
     skewAmount=0
     typefaceFontName=Regular
     outlineSettings {
      bApplyOutlineToDropShadows=false
      bMiteredCorners=false
      bSeparateFillAlpha=false
      outlineMaterial=null
      outlineSize=0
      outlineColor=<rgba 0 0 0 1>
     }
    }
   }
   slot {class=ButtonSlot}
  }
  SetBaseRifle {
   widgetClassPath=&EditorUtilityButton
   parent=CanvasPanel_202
   widget=SetBaseRifle
   properties {toolTipText=Auto set default Rifle metasounds as base for this preset}
   slot {class=CanvasPanelSlot, properties {layoutData {alignment=<0 0>, anchors {maximum=<0 0>, minimum=<0 0>}, offsets=<ltrb 332 220 100 30>}}}
  }
  UEFN_TextBlock_C_10 {
   widgetClassPath=&UEFN_TextBlock
   parent=SetBaseRifle
   widget=UEFN_TextBlock_C_10
   properties {
    justification=Center
    text=Set Base Rifle
    font {
     bForceMonospaced=false
     bMaterialIsStencil=false
     fontMaterial=null
     fontObject=&Roboto
     letterSpacing=0
     monospacedWidth=1
     size=9
     skewAmount=0
     typefaceFontName=Regular
     outlineSettings {
      bApplyOutlineToDropShadows=false
      bMiteredCorners=false
      bSeparateFillAlpha=false
      outlineMaterial=null
      outlineSize=0
      outlineColor=<rgba 0 0 0 1>
     }
    }
   }
   slot {class=ButtonSlot}
  }
  SetBaseShotgun {
   widgetClassPath=&EditorUtilityButton
   parent=CanvasPanel_202
   widget=SetBaseShotgun
   properties {toolTipText=Auto set default Shotgun metasounds as base for this preset}
   slot {class=CanvasPanelSlot, properties {layoutData {alignment=<0 0>, anchors {maximum=<0 0>, minimum=<0 0>}, offsets=<ltrb 448 220 100 30>}}}
  }
  UEFN_TextBlock_C_11 {
   widgetClassPath=&UEFN_TextBlock
   parent=SetBaseShotgun
   widget=UEFN_TextBlock_C_11
   properties {
    justification=Center
    text=Set Base Shotgun
    font {
     bForceMonospaced=false
     bMaterialIsStencil=false
     fontMaterial=null
     fontObject=&Roboto
     letterSpacing=0
     monospacedWidth=1
     size=9
     skewAmount=0
     typefaceFontName=Regular
     outlineSettings {
      bApplyOutlineToDropShadows=false
      bMiteredCorners=false
      bSeparateFillAlpha=false
      outlineMaterial=null
      outlineSize=0
      outlineColor=<rgba 0 0 0 1>
     }
    }
    renderTransform {angle=0, scale=<0.95 0.95>, shear=<0 0>, translation=<0 0>}
   }
   slot {class=ButtonSlot}
  }
  SetBaseSniper {
   widgetClassPath=&EditorUtilityButton
   parent=CanvasPanel_202
   widget=SetBaseSniper
   properties {toolTipText=Auto set default Sniper metasounds as base for this preset}
   slot {class=CanvasPanelSlot, properties {layoutData {alignment=<0 0>, anchors {maximum=<0 0>, minimum=<0 0>}, offsets=<ltrb 560 220 100 30>}}}
  }
  UEFN_TextBlock_C_12 {
   widgetClassPath=&UEFN_TextBlock
   parent=SetBaseSniper
   widget=UEFN_TextBlock_C_12
   properties {
    justification=Center
    text=Set Base Sniper
    font {
     bForceMonospaced=false
     bMaterialIsStencil=false
     fontMaterial=null
     fontObject=&Roboto
     letterSpacing=0
     monospacedWidth=1
     size=9
     skewAmount=0
     typefaceFontName=Regular
     outlineSettings {
      bApplyOutlineToDropShadows=false
      bMiteredCorners=false
      bSeparateFillAlpha=false
      outlineMaterial=null
      outlineSize=0
      outlineColor=<rgba 0 0 0 1>
     }
    }
   }
   slot {class=ButtonSlot}
  }
  SetBase3p {
   widgetClassPath=&EditorUtilityButton
   parent=CanvasPanel_202
   widget=SetBase3p
   properties {toolTipText=Auto set default weapon 3p metasound as base for this preset}
   slot {class=CanvasPanelSlot, properties {layoutData {alignment=<0 0>, anchors {maximum=<0 0>, minimum=<0 0>}, offsets=<ltrb 492 124 100 30>}}}
  }
  UEFN_TextBlock_C_5 {
   widgetClassPath=&UEFN_TextBlock
   parent=SetBase3p
   widget=UEFN_TextBlock_C_5
   properties {
    justification=Center
    text=Set Base 3p
    font {
     bForceMonospaced=false
     bMaterialIsStencil=false
     fontMaterial=null
     fontObject=&Roboto
     letterSpacing=0
     monospacedWidth=1
     size=9
     skewAmount=0
     typefaceFontName=Regular
     outlineSettings {
      bApplyOutlineToDropShadows=false
      bMiteredCorners=false
      bSeparateFillAlpha=false
      outlineMaterial=null
      outlineSize=0
      outlineColor=<rgba 0 0 0 1>
     }
    }
   }
   slot {class=ButtonSlot}
  }
  UEFN_TextBlock_C_2 {
   widgetClassPath=&UEFN_TextBlock
   parent=CanvasPanel_202
   widget=UEFN_TextBlock_C_2
   properties {
    text=Asset Name
    font {
     bForceMonospaced=false
     bMaterialIsStencil=false
     fontMaterial=null
     fontObject=&Roboto
     letterSpacing=0
     monospacedWidth=1
     size=10.5
     skewAmount=0
     typefaceFontName=Regular
     outlineSettings {
      bApplyOutlineToDropShadows=false
      bMiteredCorners=false
      bSeparateFillAlpha=false
      outlineMaterial=null
      outlineSize=0
      outlineColor=<rgba 0 0 0 1>
     }
    }
   }
   slot {
    class=CanvasPanelSlot
    properties {layoutData {alignment=<0 0>, anchors {maximum=<0 0>, minimum=<0 0>}, offsets=<ltrb 16 328 101.1862 19.2282>}}
   }
  }
  UEFN_TextBlock_C_6 {
   widgetClassPath=&UEFN_TextBlock
   parent=CanvasPanel_202
   widget=UEFN_TextBlock_C_6
   properties {
    text=Is Persistent Fire?
    toolTipText="Is this weapon automatic (persistent fire) or single shot?"
    font {
     bForceMonospaced=false
     bMaterialIsStencil=false
     fontMaterial=null
     fontObject=&Roboto
     letterSpacing=0
     monospacedWidth=1
     size=10.5
     skewAmount=0
     typefaceFontName=Regular
     outlineSettings {
      bApplyOutlineToDropShadows=false
      bMiteredCorners=false
      bSeparateFillAlpha=false
      outlineMaterial=null
      outlineSize=0
      outlineColor=<rgba 0 0 0 1>
     }
    }
   }
   slot {
    class=CanvasPanelSlot
    properties {layoutData {alignment=<0 0>, anchors {maximum=<0 0>, minimum=<0 0>}, offsets=<ltrb 16 388 101.1862 19.2282>}}
   }
  }
  AssetNameInput {
   widgetClassPath=&EditorUtilityEditableTextBox
   bIsVariable=true
   parent=CanvasPanel_202
   widget=AssetNameInput
   properties {toolTipText=Metasound preset name. _1p and _3p suffixes are auto added}
   slot {
    class=CanvasPanelSlot
    properties {layoutData {alignment=<0 0>, anchors {maximum=<0 0>, minimum=<0 0>}, offsets=<ltrb 164 324 313.964 28.4985>}}
   }
  }
  EditorUtilityCheckBox_0 {
   widgetClassPath=&EditorUtilityCheckBox
   bIsVariable=true
   parent=CanvasPanel_202
   widget=EditorUtilityCheckBox_0
   properties {checkedState=Checked, toolTipText="Is this weapon automatic (persistent fire) or single shot?"}
   slot {
    class=CanvasPanelSlot
    properties {layoutData {alignment=<0 0>, anchors {maximum=<0 0>, minimum=<0 0>}, offsets=<ltrb 164 376 81.7417 42.8829>}}
   }
  }
  UEFN_TextBlock_C_7 {
   widgetClassPath=&UEFN_TextBlock
   parent=CanvasPanel_202
   widget=UEFN_TextBlock_C_7
   properties {
    text=Optional
    toolTipText=Select if you want to base the preset on a specific weapon class with default sounds
    font {
     bForceMonospaced=false
     bMaterialIsStencil=false
     fontMaterial=null
     fontObject=&Roboto
     letterSpacing=0
     monospacedWidth=1
     size=10.5
     skewAmount=0
     typefaceFontName=Regular
     outlineSettings {
      bApplyOutlineToDropShadows=false
      bMiteredCorners=false
      bSeparateFillAlpha=false
      outlineMaterial=null
      outlineSize=0
      outlineColor=<rgba 0 0 0 1>
     }
    }
   }
   slot {
    class=CanvasPanelSlot
    properties {layoutData {alignment=<0 0>, anchors {maximum=<0 0>, minimum=<0 0>}, offsets=<ltrb 16 224 101.1862 19.2282>}}
   }
  }
 ]
}
