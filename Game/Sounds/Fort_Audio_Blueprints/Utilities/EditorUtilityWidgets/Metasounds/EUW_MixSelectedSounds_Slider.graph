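# Editor utility widget providing a vertical volume slider for one sound asset, with name and value labels.
# InitialiseSlider stores the passed sound, shows its name and loads its current volume into the slider.
# It targets the Volume property for MetaSoundSource assets and VolumeMultiplier for anything else, accessed by name through Get/SetEditorProperty.
# Slider changes write straight to that property and update the value label.
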EditorUtilityWidgetBlueprint /Game/Sounds/Fort_Audio_Blueprints/Utilities/EditorUtilityWidgets/Metasounds/EUW_MixSelectedSounds_Slider
graphs [EventGraph {nodes [
   n1 = "On Value Changed (Slider)"!() -> (OutputDelegate, then, Value:float) {is=title K2Node_ComponentBoundEvent}
   > SetEditorProperty(
    Object=Get SelectedSound() -> (SelectedSound:Object Reference)
    PropertyName=Get VolumeProperty() -> (VolumeProperty:Name=None)
    PropertyValue:float=@n1.Value
   )
   > Set Text(
    Text="To Text (Float)"(Value=@n1.Value, MinimumIntegralDigits=1, MaximumIntegralDigits=324, MinimumFractionalDigits=0, MaximumFractionalDigits=3)
    self=Get Text_Volume() -> (Text_Volume:Text Object Reference)
   )
   n2 = InitialiseSlider!() -> (OutputDelegate, then, SelectedSound:Object Reference) {is=title K2Node_CustomEvent}
   > n3 = Set SelectedSound(SelectedSound:Object Reference=@n2.SelectedSound) -> (Output_Get:Object Reference)
   > Set VolumeProperty(VolumeProperty:Name=SelectName!(
     A:Name=Volume
     B:Name=VolumeMultiplier
     bPickA:bool=Cast To MetaSoundSource!(Object=@n3) -> (AsMetaSound Source:Meta Sound Source Object Reference, bSuccess:bool=~) {is=title K2Node_DynamicCast}.bSuccess
    ) -> (ReturnValue:Name=None)) -> (Output_Get:Name=None)
   > Set Text(
    Text="To Text (Object)"!(InObj:Object Reference=@n3) -> (ReturnValue:Text)
    self=Get Text_SoundName() -> (Text_SoundName:Text Object Reference)
   )
   > n4 = GetEditorProperty!(execute:Exec, Object:Object Reference=Get SelectedSound(), PropertyName:Name=Get VolumeProperty() -> (VolumeProperty=None)) -> (then:Exec, PropertyValue:float, ReturnValue:bool)
   > Set Value!(execute, Value:float=@n4, self:Slider Object Reference=Get Slider() -> (Slider:Editor Utility Slider Object Reference)) -> (then, Output_Get:float)
  ]}]
variables [
 SelectedSound {type=Object, default=null}
 VolumeProperty {type=string, default=None}
 Volumes {type="array<number>", default []}
]
widgets {
 info {rootWidgetClass=&CanvasPanel, widgetCount=5}
 widgets [
  Canvas {widgetClassPath=&CanvasPanel, bIsVariable=true, widget=Canvas}
  SizeBox_118 {
   widgetClassPath=&SizeBox
   parent=Canvas
   widget=SizeBox_118
   slot {
    class=CanvasPanelSlot
    properties {layoutData {alignment=<0.5 0.5>, anchors {maximum=<0.5 0.5>, minimum=<0.5 0.5>}, offsets=<ltrb 0 3 100 194.4>}}
   }
  }
  Slider {
   widgetClassPath=/Script/Blutility.EditorUtilitySlider
   bIsVariable=true
   parent=Canvas
   widget=Slider
   properties {maxValue=5, orientation=Orient_Vertical}
   slot {
    class=CanvasPanelSlot
    properties {layoutData {alignment=<0.5 0.5>, anchors {maximum=<0.5 0.5>, minimum=<0.5 0.5>}, offsets=<ltrb 0 -6.75 13.4129 133.9294>}}
   }
  }
  Text_Volume {
   widgetClassPath=&TextBlock
   bIsVariable=true
   parent=Canvas
   widget=Text_Volume
   properties {text=0}
   slot {
    class=CanvasPanelSlot
    properties {layoutData {alignment=<0.5 0.5>, anchors {maximum=<0.5 0.5>, minimum=<0.5 0.5>}, offsets=<ltrb 0 80 18.6667 30>}}
   }
  }
  Text_SoundName {
   widgetClassPath=&TextBlock
   bIsVariable=true
   parent=Canvas
   widget=Text_SoundName
   properties {
    justification=Center
    text=Sound Name
    font {
     bForceMonospaced=false
     bMaterialIsStencil=false
     fontMaterial=null
     fontObject=&Roboto
     letterSpacing=0
     monospacedWidth=1
     size=10.5
     skewAmount=0
     typefaceFontName=Bold
     outlineSettings {
      bApplyOutlineToDropShadows=false
      bMiteredCorners=false
      bSeparateFillAlpha=false
      outlineMaterial=null
      outlineSize=0
      outlineColor=<rgba 0 0 0 1>
     }
    }
   }
   slot {
    class=CanvasPanelSlot
    properties {bAutoSize=true, layoutData {alignment=<0.5 0.5>, anchors {maximum=<0.5 0.5>, minimum=<0.5 0.5>}, offsets=<ltrb 0 -84 227.6667 16.6667>}}
   }
  }
 ]
}
