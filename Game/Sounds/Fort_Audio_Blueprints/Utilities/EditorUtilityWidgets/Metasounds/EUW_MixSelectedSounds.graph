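# Editor utility widget that builds a mixer panel of sliders for the sounds selected in the editor.
# The Populate button clears the horizontal scroll box and stored asset array, then gathers the selected SoundBase assets.
# For each sound it creates an EUW_MixSelectedSounds_Slider child widget, adds it to the scroll box and passes the sound to InitialiseSlider.
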EditorUtilityWidgetBlueprint /Game/Sounds/Fort_Audio_Blueprints/Utilities/EditorUtilityWidgets/Metasounds/EUW_MixSelectedSounds
graphs [EventGraph {nodes [
   "On Clicked (Button_Populate)"() {is=title K2Node_ComponentBoundEvent}
   > ClearData()
   > PopulateMixer()
   ClearData() {is=title K2Node_CustomEvent}
   > ClearChildren(self=Get ScrollBox() -> (ScrollBox:Editor Utility Scroll Box Object Reference))
   > Clear(TargetArray:Array of Object References=Get SelectedSoundAssets() -> (SelectedSoundAssets:Array of Object References))
   PopulateMixer() {is=title K2Node_CustomEvent}
   > n1 = GetSelectedAssetsOfClass(AssetClass=/Script/Engine.SoundBase)
   > Branch(Condition=Is Not Empty!(TargetArray:Array of Object References=@n1) -> (ReturnValue:bool))
   > n2 = Set SelectedSoundAssets(SelectedSoundAssets:Array of Object References=@n1) -> (Output_Get:Array of Object References)
   > n3 = For Each Loop(Array:Array of Object References=@n2) -> (Array Element:Object Reference)
   > Set CurrentSoundInteration(CurrentSoundInteration:Object Reference=@n3) -> (Output_Get:Object Reference)
   > n4 = Create Widget!(
    execute:Exec
    WorldContextObject:Object Reference
    Class:User Widget Class Reference=&/EUW_MixSelectedSounds_Slider.EUW_MixSelectedSounds_Slider_C
    OwningPlayer:Player Controller Object Reference
   ) -> (then:Exec, ReturnValue:EUW Mix Selected Sounds Slider Object Reference)
   > AddChild(self=Get ScrollBox(), Content=@n4)
   > InitialiseSlider!(
    execute
    self:EUW Mix Selected Sounds Slider Object Reference=@n4
    SelectedSound:Object Reference=Get CurrentSoundInteration() -> (CurrentSoundInteration:Object Reference)
   ) -> (then)
  ]}]
variables [
 CurrentSoundInteration {type=Object, default=null}
 SelectedSoundAssets {type="array<Object>", default []}
]
widgets {
 info {rootWidgetClass=&CanvasPanel, widgetCount=4}
 widgets [
  Canvas {widgetClassPath=&CanvasPanel, bIsVariable=true, widget=Canvas}
  ScrollBox {
   widgetClassPath=&EditorUtilityScrollBox
   bIsVariable=true
   parent=Canvas
   widget=ScrollBox
   properties {orientation=Orient_Horizontal}
   slot {
    class=CanvasPanelSlot
    properties {layoutData {alignment=<0.5 0.5>, anchors {maximum=<0 0.5>, minimum=<0 0.5>}, offsets=<ltrb 919.2592 0 1538.5186 400>}}
   }
  }
  Button_Populate {
   widgetClassPath=&EditorUtilityButton
   parent=Canvas
   widget=Button_Populate
   slot {class=CanvasPanelSlot, properties {layoutData {alignment=<0 0>, anchors {maximum=<0 0.5>, minimum=<0 0.5>}, offsets=<ltrb 24 -28 100 30>}}}
  }
  Text_Populate {
   widgetClassPath=&TextBlock
   parent=Button_Populate
   widget=Text_Populate
   properties {
    text=Populate
    font {
     bForceMonospaced=false
     bMaterialIsStencil=false
     fontMaterial=null
     fontObject=&Roboto
     letterSpacing=0
     monospacedWidth=1
     size=12
     skewAmount=0
     typefaceFontName=Bold
     outlineSettings {
      bApplyOutlineToDropShadows=false
      bMiteredCorners=false
      bSeparateFillAlpha=false
      outlineMaterial=null
      outlineSize=0
      outlineColor=<rgba 0 0 0 1>
     }
    }
   }
   slot {class=ButtonSlot}
  }
 ]
}
