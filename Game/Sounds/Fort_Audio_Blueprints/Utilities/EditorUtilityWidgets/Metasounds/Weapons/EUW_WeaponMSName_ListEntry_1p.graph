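# Editor utility list-entry widget for a first-person weapon MetaSound, with name text and a checkbox choosing whether it gets built.
# When assigned a list item, it stores it as a B_WeaponMSName_Object, and GetMSName returns the name from that object's struct.
# Checkbox changes update Is Checked and broadcast CheckStateChanged.
# The ChangeName event sets SpokenTextBlock's text; list assignment itself leaves the displayed text untouched.
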EditorUtilityWidgetBlueprint /Game/Sounds/Fort_Audio_Blueprints/Utilities/EditorUtilityWidgets/Metasounds/Weapons/EUW_WeaponMSName_ListEntry_1p
graphs [
 CheckStateChanged {nodes [CheckStateChanged() {is=title K2Node_FunctionEntry}]}
 EventGraph {nodes [
   n1 = "On Check State Changed (EditorUtilityCheckBox_205)"!() -> (OutputDelegate, then, bIsChecked:bool) {is=title K2Node_ComponentBoundEvent}
   > Set Is Checked(Is Checked:bool=@n1.bIsChecked) -> (Output_Get:bool)
   > Call Check State Changed!(execute:Exec, self:EUW Weapon MSName List Entry 1p Object Reference) -> (then:Exec) {is=title K2Node_CallDelegate}
   n2 = ChangeName!() -> (OutputDelegate, then, Text:Text) {is=title K2Node_CustomEvent}
   > Set Text(Text=@n2.Text, self=Get SpokenTextBlock() -> (SpokenTextBlock:Text Object Reference))
   n3 = Event OnListItemObjectSet()
   > n4 = Cast To B_WeaponMSName_Object!(execute, Object=@n3.ListItemObject) -> (then, CastFailed, AsB Weapon MSName Object:B Weapon MSName Object Object Reference) {is=title K2Node_DynamicCast}
   > Set WeaponMSName(WeaponMSName:B Weapon MSName Object Object Reference=@n4) -> (Output_Get:B Weapon MSName Object Object Reference)
   Event PreConstruct()
   Event Construct()
   Event Tick()
  ]}
 GetMSName {nodes [
   GetMSName() {is=title K2Node_FunctionEntry}
   n1 = Get WeaponMSName() -> (WeaponMSName:B Weapon MSName Object Object Reference)
   > Return Node!(
    execute
    Name:Name=Break Struct Weapon MSName!(Struct_WeaponMSName:"Struct Weapon MSName Structure (by ref)"=Get Name!(self:B Weapon MSName Object Object Reference=@n1) -> (Name:Struct Weapon MSName Structure)) -> (Name_2_#1:Name) {is=title K2Node_BreakStruct}
   )
  ]}
]
variables [
 CheckStateChanged {default=()}
 Is Checked {type=boolean}
 WeaponMSName {type=B_WeaponMSName_Object_C, default=null}
]
widgets {
 info {rootWidgetClass=&VerticalBox, widgetCount=20}
 widgets [
  VerticalBox_21 {widgetClassPath=&VerticalBox, widget=VerticalBox_21}
  HorizontalBox_35 {widgetClassPath=&HorizontalBox, parent=VerticalBox_21, widget=HorizontalBox_35, slot {class=VerticalBoxSlot}}
  Overlay_0 {
   widgetClassPath=&Overlay
   parent=HorizontalBox_35
   widget=Overlay_0
   slot {class=HorizontalBoxSlot, properties {size {sizeRule=Fill, value=1}}}
  }
  Background {
   widgetClassPath=&Border
   bIsVariable=true
   parent=Overlay_0
   widget=Background
   properties {
    verticalAlignment=VAlign_Center
    background {
     bIsDynamicallyLoaded=false
     drawAs=Box
     imageType=NoImage
     mirroring=NoMirror
     resourceName=None
     resourceObject=null
     tiling=NoTile
     imageSize=<32 32>
     margin=<ltrb 0 0 0 0>
     outlineSettings {
      bUseBrushTransparency=false
      roundingType=FixedRadius
      width=2
      color {colorUseRule=UseColor_Specified, specifiedColor=<rgba 0 0 0 0>}
      cornerRadii=<0 0 0 0>
     }
     tintColor {colorUseRule=UseColor_Specified, specifiedColor=<rgba 1 1 1 1>}
     uVRegion {bIsValid=false, max=<0 0>, min=<0 0>}
    }
    brushColor=<rgba 0.0176 0.0176 0.0176 1>
    padding=<ltrb 6 6 6 4>
   }
   slot {class=OverlaySlot, properties {horizontalAlignment=HAlign_Fill, verticalAlignment=VAlign_Fill}}
  }
  DragEnteredGlow {
   widgetClassPath=&Border
   bIsVariable=true
   parent=Overlay_0
   widget=DragEnteredGlow
   properties {
    visibility=Hidden
    background {
     bIsDynamicallyLoaded=false
     drawAs=Image
     imageType=FullColor
     mirroring=NoMirror
     resourceName=None
     resourceObject=null
     tiling=NoTile
     imageSize=<256 256>
     margin=<ltrb 0 0 0 0>
     outlineSettings {
      bUseBrushTransparency=false
      roundingType=HalfHeightRadius
      width=0
      color {colorUseRule=UseColor_Specified, specifiedColor=<rgba 0 0 0 0>}
      cornerRadii=<0 0 0 0>
     }
     tintColor {colorUseRule=UseColor_Specified, specifiedColor=<rgba 1 1 1 1>}
     uVRegion {bIsValid=false, max=<0 0>, min=<0 0>}
    }
    brushColor=<rgba 0 0.5524 1 0.5>
    padding=<ltrb 0 0 0 0>
   }
   slot {class=OverlaySlot, properties {horizontalAlignment=HAlign_Fill, verticalAlignment=VAlign_Fill}}
  }
  Content {
   widgetClassPath=&HorizontalBox
   parent=Overlay_0
   widget=Content
   slot {class=OverlaySlot, properties {horizontalAlignment=HAlign_Fill, verticalAlignment=VAlign_Center, padding=<ltrb 6 6 6 4>}}
  }
  VerticalBox {widgetClassPath=&VerticalBox, parent=Content, widget=VerticalBox, slot {class=HorizontalBoxSlot}}
  SizeBox_0 {
   widgetClassPath=&SizeBox
   parent=VerticalBox
   widget=SizeBox_0
   properties {bOverride_HeightOverride=true, bOverride_WidthOverride=true, heightOverride=32, widthOverride=32}
   slot {class=VerticalBoxSlot, properties {horizontalAlignment=HAlign_Center, verticalAlignment=VAlign_Center, padding=<ltrb 4 4 4 4>}}
  }
  SoundIcon {
   widgetClassPath=&Image
   bIsVariable=true
   parent=SizeBox_0
   widget=SoundIcon
   properties {
    brush {
     bIsDynamicallyLoaded=false
     drawAs=Image
     imageType=FullColor
     mirroring=NoMirror
     resourceName=None
     resourceObject=/Game/UI/Icons/Icon-Stat/Icon-WeaponSkill-XL
     tiling=NoTile
     imageSize=<104 104>
     margin=<ltrb 0 0 0 0>
     outlineSettings {
      bUseBrushTransparency=false
      roundingType=HalfHeightRadius
      width=0
      color {colorUseRule=UseColor_Specified, specifiedColor=<rgba 0 0 0 0>}
      cornerRadii=<0 0 0 0>
     }
     tintColor {colorUseRule=UseColor_Specified, specifiedColor=<rgba 1 1 1 1>}
     uVRegion {bIsValid=false, max=<0 0>, min=<0 0>}
    }
    colorAndOpacity=<rgba 0.5029 0.5029 0.5029 1>
   }
   slot {class=SizeBoxSlot}
  }
  VerticalBox_23 {
   widgetClassPath=&VerticalBox
   parent=Content
   widget=VerticalBox_23
   slot {class=HorizontalBoxSlot, properties {verticalAlignment=VAlign_Center, padding=<ltrb 6 0 0 0>, size {sizeRule=Fill, value=1}}}
  }
  HorizontalBox_3 {widgetClassPath=&HorizontalBox, parent=VerticalBox_23, widget=HorizontalBox_3, slot {class=VerticalBoxSlot}}
  AssetNameText_1 {
   widgetClassPath=&TextBlock
   bIsVariable=true
   parent=HorizontalBox_3
   widget=AssetNameText_1
   properties {
    text=Asset Name
    colorAndOpacity {colorUseRule=UseColor_Specified, specifiedColor=<rgba 0.5029 0.5029 0.5029 1>}
    font {
     bForceMonospaced=false
     bMaterialIsStencil=false
     fontMaterial=null
     fontObject=&BurbankSmall
     letterSpacing=0
     monospacedWidth=1
     size=9.75
     skewAmount=0
     typefaceFontName=Bold
     outlineSettings {
      bApplyOutlineToDropShadows=false
      bMiteredCorners=false
      bSeparateFillAlpha=false
      outlineMaterial=null
      outlineSize=0
      outlineColor=<rgba 0 0 0 1>
     }
    }
   }
   slot {class=HorizontalBoxSlot, properties {size {sizeRule=Fill, value=1}}}
  }
  HorizontalBox {
   widgetClassPath=&HorizontalBox
   parent=VerticalBox_23
   widget=HorizontalBox
   slot {class=VerticalBoxSlot, properties {size {sizeRule=Fill, value=1}}}
  }
  SpokenTextBlock {
   widgetClassPath=&TextBlock
   bIsVariable=true
   parent=HorizontalBox
   widget=SpokenTextBlock
   properties {
    text=1P Weapon MetaSound
    colorAndOpacity {colorUseRule=UseColor_Specified, specifiedColor=<rgba 0.5029 0.5029 0.5029 1>}
    font {
     bForceMonospaced=false
     bMaterialIsStencil=false
     fontMaterial=null
     fontObject=&Roboto
     letterSpacing=0
     monospacedWidth=1
     size=9
     skewAmount=0
     typefaceFontName=Light
     outlineSettings {
      bApplyOutlineToDropShadows=false
      bMiteredCorners=false
      bSeparateFillAlpha=false
      outlineMaterial=null
      outlineSize=0
      outlineColor=<rgba 0 0 0 1>
     }
    }
   }
   slot {class=HorizontalBoxSlot, properties {horizontalAlignment=HAlign_Left, size {sizeRule=Fill, value=1}}}
  }
  HorizontalBox_2 {
   widgetClassPath=&HorizontalBox
   parent=VerticalBox_23
   widget=HorizontalBox_2
   properties {visibility=Collapsed}
   slot {class=VerticalBoxSlot}
  }
  UnassignedLines {
   widgetClassPath=&TextBlock
   parent=HorizontalBox_2
   widget=UnassignedLines
   properties {
    text=Unassigned Lines
    colorAndOpacity {colorUseRule=UseColor_Specified, specifiedColor=<rgba 0.5029 0.5029 0.5029 1>}
    font {
     bForceMonospaced=false
     bMaterialIsStencil=false
     fontMaterial=null
     fontObject=&Roboto
     letterSpacing=0
     monospacedWidth=1
     size=9
     skewAmount=0
     typefaceFontName=Regular
     outlineSettings {
      bApplyOutlineToDropShadows=false
      bMiteredCorners=false
      bSeparateFillAlpha=false
      outlineMaterial=null
      outlineSize=0
      outlineColor=<rgba 0 0 0 1>
     }
    }
   }
   slot {class=HorizontalBoxSlot}
  }
  VerticalBox_1 {
   widgetClassPath=&VerticalBox
   parent=Content
   widget=VerticalBox_1
   slot {class=HorizontalBoxSlot, properties {horizontalAlignment=HAlign_Right, verticalAlignment=VAlign_Center}}
  }
  HorizontalBox_1 {widgetClassPath=&HorizontalBox, parent=VerticalBox_1, widget=HorizontalBox_1, slot {class=VerticalBoxSlot}}
  SizeBox_1 {widgetClassPath=&SizeBox, parent=HorizontalBox_1, widget=SizeBox_1, slot {class=HorizontalBoxSlot}}
  EditorUtilityCheckBox_205 {
   widgetClassPath=&EditorUtilityCheckBox
   bIsVariable=true
   parent=SizeBox_1
   widget=EditorUtilityCheckBox_205
   properties {checkedState=Checked, toolTipText=Should this MetaSound be built?}
   slot {class=SizeBoxSlot}
  }
 ]
}
