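# Editor utility widget that builds 1P and 3P weapon-fire MetaSound presets for Fortnite audio authoring.
# A looping timer mirrors Content Browser SoundWave selection into lists that sort waves by layer.
# MetaSound builders create presets from the chosen parent, assigning those waves plus weapon-type defaults to graph inputs.
# It then creates or updates a Sound Library, runs weapon item implementation, including bullet whip data, and saves the assets.
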EditorUtilityWidgetBlueprint /Game/Sounds/Fort_Audio_Blueprints/Utilities/EditorUtilityWidgets/Metasounds/EUW_WeaponMetaSoundPresetCreator_v2
graphs [
 AssetSelectionTimer {nodes [
   AssetSelectionTimer() {is=title K2Node_FunctionEntry}
   > n1 = GetSelectedAssetsOfClass(AssetClass=/Script/Engine.SoundWave)
   > n2 = For Each Loop(Array:Array of Object References=@n1) -> (Array Element:Object Reference)
   > n3 = Cast To MetaSoundSource!(execute, Object=@n2) -> (then, CastFailed, AsMetaSound Source:Meta Sound Source Object Reference) {is=title K2Node_DynamicCast}
   n4 = Sequence(execute=@n3.CastFailed)
   > n5 = Branch(Condition=Contains Item(
     TargetArray:Array of Object References=Get SelectedAssets() -> (SelectedAssets:Array of Object References)
     ItemToFind:"Object Reference (by ref)"=@n2
    ))
   n6 = Add Unique(execute=@n5.else, TargetArray:Array of Object References=Get SelectedAssets(), NewItem:"Object Reference (by ref)"=@n2)
   > AddEntryUpdate!(
    execute
    self:EUW Audio Weapons Sound Name List View Object Reference=Get EUW_Audio_Weapons_SoundName_ListView_1P() -> (EUW_Audio_Weapons_SoundName_ListView_1P:EUW Audio Weapons Sound Name List View Object Reference)
    Object:Object Reference=@n2
    AssetIndex:"Integer (by ref)"=@n6
   ) -> (then)
   > AddEntryUpdate(
    self=Get EUW_Audio_Weapons_SoundName_ListView_3P() -> (EUW_Audio_Weapons_SoundName_ListView_3P:EUW Audio Weapons Sound Name List View Object Reference)
    Object=@n2
    AssetIndex=@n6
   )
   n7 = For Each Loop(Exec=@n4.then_1, Array=Get SelectedAssets())
   > n8 = Branch(Condition=Contains Item(TargetArray=@n1, ItemToFind=@n7))
   Remove Item!(execute=@n8.else, TargetArray:Array of Object References=Get SelectedAssets(), Item:"Object Reference (by ref)"=@n7) -> (then, ReturnValue:bool)
   > RemoveEntryUpdate!(
    execute
    self:EUW Audio Weapons Sound Name List View Object Reference=Get EUW_Audio_Weapons_SoundName_ListView_1P()
    Object:Object Reference=@n7
    AssetIndex:"Integer (by ref)"=@n7.Array Index
   ) -> (then)
   > RemoveEntryUpdate(self=Get EUW_Audio_Weapons_SoundName_ListView_3P(), Object=@n7, AssetIndex=@n7.Array Index)
  ]}
 AssignDefaultWaves {nodes [
   AssignDefaultWaves() {is=title K2Node_FunctionEntry}
   > n1 = Branch(Condition=Get bIs1p() -> (bIs1p:bool))
   n2 = Get WeaponSoundData() -> (WeaponSoundData:B Weapon Types Data Object Reference)
   > AssignSelectedWaves!(
    execute
    self
    InputName:Name=Suppressor.Sounds
    WaveArray:Array of Object References=Get Suppressor Sounds!(self:B Weapon Types Data Object Reference=@n2) -> (Suppressor Sounds:Array of Sound Wave Object References)
   ) -> (then, bSuccess:bool)
   > n3 = Branch(Condition=Get bLowAmmoSet() -> (bLowAmmoSet:bool))
   > n4 = Set bLowAmmoSet(bLowAmmoSet:bool=false) -> (Output_Get:bool)
   Return Node(execute=[@n5.then, @n4.then])
   n5 = AssignSelectedWaves(
    execute=@n3.else
    InputName=LowAmmo.Sounds
    WaveArray=Get Low Ammo Sounds!(self:B Weapon Types Data Object Reference=@n2) -> (Low Ammo Sounds:Array of Sound Wave Object References)
   )
   AssignSelectedWaves(
    execute=@n1.else
    InputName=Suppressor.Sounds.Close
    WaveArray=Get Suppressor Close Sounds!(self:B Weapon Types Data Object Reference=@n2) -> (Suppressor Close Sounds:Array of Sound Wave Object References)
   )
   > AssignSelectedWaves(
    InputName=Suppressor.Sounds.Medium
    WaveArray=Get Suppressor Medium Sounds!(self:B Weapon Types Data Object Reference=@n2) -> (Suppressor Medium Sounds:Array of Sound Wave Object References)
   )
   > Return Node()
  ]}
 AssignSelectedWaves {nodes [
   n1 = AssignSelectedWaves!() -> (then, InputName:Name, WaveArray:Array of Object References) {is=title K2Node_FunctionEntry}
   > n2 = Branch(Condition=Is Not Empty!(TargetArray:Array of Object References=@n1.WaveArray) -> (ReturnValue:bool))
   > n3 = Set WaveParameters(WaveParameters:Audio Parameter Structure=ObjectArrayToAudioParameter!(Name:Name=@n1, Objects:Array of Object References=@n1.WaveArray) -> (ReturnValue:Audio Parameter Structure)) -> (Output_Get:Audio Parameter Structure)
   > n4 = Create MetaSound Literal From AudioParameter!(
    execute
    self:Meta Sound Builder Subsystem Object Reference=Get MetaSoundBuilderSubsystem() -> (ReturnValue:Meta Sound Builder Subsystem Object Reference) {is=title K2Node_GetEngineSubsystem}
    Param:"Audio Parameter Structure (by ref)"=@n3
   ) -> (then, ReturnValue:Meta Sound Literal Structure)
   > n5 = Branch(Condition="Equal (Name)"(A=@n1, B=LowAmmo.Sounds))
   > n6 = Set bLowAmmoSet(bLowAmmoSet=true)
   n7 = SetGraphInputDefault!(
    execute=[@n6.then, @n5.else]
    self:Meta Sound Builder Base Object Reference=Get MetaSoundPresetBuilder() -> (MetaSoundPresetBuilder:Meta Sound Source Builder Object Reference)
    InputName:Name=@n1
    Literal:"Meta Sound Literal Structure (by ref)"=@n4
    PageName:Name
   ) -> (Succeeded:Exec, Failed:Exec)
   Return Node!(execute=@n7.Failed, bSuccess:bool=false)
   > Return Node(bSuccess=true)
   Return Node(execute=@n2.else, bSuccess=false)
  ]}
 BulletWhipImplementation {nodes [
   n1 = BulletWhipImplementation!() -> (then, ItemDefinition:Item Definition Base Object Reference) {is=title K2Node_FunctionEntry}
   > n2 = GetItemData!(execute, self:Item Definition Base Object Reference=@n1) -> (Found:Exec, NotFound:Exec, ItemData:"Fortnite Bullet Whip Tracker Data Item Data Structure (by ref)")
   > n3 = RemoveItemData!(
    execute
    self:Item Definition Base Object Reference=@n1
    StructType:Script Struct Object Reference=/Script/FortniteGame.FortItemComponentData_BulletWhipTrackerData
   ) -> (Found:Exec, NotFound:Exec)
   n4 = Select!(
    NewEnumerator0:Bullet Whip Tracker Audio Data Object Reference=/Game/Athena/Sounds/Weapons/BulletWhips/DA_BulletWhipTrackerData_Rifle_NoSound
    NewEnumerator1:Bullet Whip Tracker Audio Data Object Reference=/Game/Athena/Sounds/Weapons/BulletWhips/DA_BulletWhipTrackerData_SMG_NoSound
    NewEnumerator2:Bullet Whip Tracker Audio Data Object Reference=/Game/Athena/Sounds/Weapons/BulletWhips/DA_BulletWhipTrackerData_Pistol_NoSound
    NewEnumerator3:Bullet Whip Tracker Audio Data Object Reference=/Game/Athena/Sounds/Weapons/BulletWhips/DA_BulletWhipTrackerData_Shotgun_NoSound
    NewEnumerator4:Bullet Whip Tracker Audio Data Object Reference=/Game/Athena/Sounds/Weapons/BulletWhips/DA_BulletWhipTrackerData_Sniper_NoSound
    NewEnumerator5:Bullet Whip Tracker Audio Data Object Reference=&DA_BulletWhipTrackerData_SMG_NoSound
    Index:Enum_WeaponMetaSoundClass Enum=Get WeaponType() -> (WeaponType:Enum_WeaponMetaSoundClass Enum=NewEnumerator0)
   ) -> (ReturnValue:Bullet Whip Tracker Audio Data Object Reference)
   n5 = Get ParentMetaSound() -> (ParentMetaSound:Enum_WeaponMetaSoundParents Enum=NewEnumerator4)
   n6 = Select(
    NewEnumerator0=Select!(
     NewEnumerator4=/FortMetaSoundWeapon/Items/Metasounds/Whips/BulletWhipTrackerData_Rifle
     NewEnumerator3=&BulletWhipTrackerData_Rifle
     NewEnumerator5=/FortMetaSoundWeapon/Items/Metasounds/FPS/Whips/BulletWhipTrackerData_FPS_Rifle
     NewEnumerator6:Bullet Whip Tracker Audio Data Object Reference
     Index:Enum_WeaponMetaSoundParents Enum=@n5
    ) -> (ReturnValue)
    NewEnumerator1=Select(
     NewEnumerator4=/FortMetaSoundWeapon/Items/Metasounds/Whips/BulletWhipTrackerData_SMG
     NewEnumerator3=&BulletWhipTrackerData_SMG
     NewEnumerator5=/FortMetaSoundWeapon/Items/Metasounds/FPS/Whips/BulletWhipTrackerData_FPS_SMG
     Index=@n5
    )
    NewEnumerator2=Select(
     NewEnumerator4=/FortMetaSoundWeapon/Items/Metasounds/Whips/BulletWhipTrackerData_Pistol
     NewEnumerator3=&BulletWhipTrackerData_Pistol
     NewEnumerator5=/FortMetaSoundWeapon/Items/Metasounds/FPS/Whips/BulletWhipTrackerData_FPS_Pistol
     Index=@n5
    )
    NewEnumerator3=Select(
     NewEnumerator4=/FortMetaSoundWeapon/Items/Metasounds/Whips/BulletWhipTrackerData_Shotgun
     NewEnumerator3=&BulletWhipTrackerData_Shotgun
     NewEnumerator5=/FortMetaSoundWeapon/Items/Metasounds/FPS/Whips/BulletWhipTrackerData_FPS_Shotgun
     Index=@n5
    )
    NewEnumerator4=Select(
     NewEnumerator4=/FortMetaSoundWeapon/Items/Metasounds/Whips/BulletWhipTrackerData_Sniper
     NewEnumerator3=&BulletWhipTrackerData_Sniper
     NewEnumerator5=/FortMetaSoundWeapon/Items/Metasounds/FPS/Whips/BulletWhipTrackerData_FPS_Sniper
     Index=@n5
    )
    NewEnumerator5=Select(
     NewEnumerator4=/FortMetaSoundWeapon/Items/Metasounds/Whips/BulletWhipTrackerData_Default
     NewEnumerator3=&BulletWhipTrackerData_Default
     NewEnumerator5=&BulletWhipTrackerData_FPS_SMG
     Index=@n5
    )
    Index=Get WeaponType() -> (WeaponType=NewEnumerator0)
   )
   n7 = Get bIsUnvault() -> (bIsUnvault:bool)
   SetItemData!(
    execute=[@n3.Found, @n3.NotFound, @n2.NotFound]
    self:Item Definition Base Object Reference=@n1
    ItemData:Fortnite Bullet Whip Tracker Data Item Data Structure=Make Fortnite Bullet Whip Tracker Data Item Data!(BulletWhipTrackerData:Bullet Whip Tracker Audio Data Object Reference=Select!(Option 0:Bullet Whip Tracker Audio Data Object Reference=@n6, Option 1:Bullet Whip Tracker Audio Data Object Reference=@n4, Index:bool=@n7) -> (ReturnValue)) -> (FortItemComponentData_BulletWhipTrackerData:Fortnite Bullet Whip Tracker Data Item Data Structure) {is=title K2Node_MakeStruct}
   ) -> (then)
   > Return Node!(execute, Item:Item Definition Base Object Reference=@n1)
   Break Fortnite Bullet Whip Tracker Data Item Data!(FortItemComponentData_BulletWhipTrackerData:"Fortnite Bullet Whip Tracker Data Item Data Structure (by ref)"=@n2) -> (BulletWhipTrackerData:Bullet Whip Tracker Audio Data Object Reference) {is=title K2Node_BreakStruct}
  ]}
 ClearAll {nodes [
   ClearAll() {is=title K2Node_FunctionEntry}
   > Set ParentMetaSound(ParentMetaSound:Enum_WeaponMetaSoundParents Enum=NewEnumerator4) -> (Output_Get:Enum_WeaponMetaSoundParents Enum=NewEnumerator4)
   > Set WeaponType(WeaponType:Enum_WeaponMetaSoundClass Enum=NewEnumerator0) -> (Output_Get:Enum_WeaponMetaSoundClass Enum=NewEnumerator0)
   > Clear(TargetArray:Array of Fort Weapon Ranged Item Definition Object References=Get WeaponID() -> (WeaponID:Array of Fort Weapon Ranged Item Definition Object References))
   > Set WeaponSoundLibrary(WeaponSoundLibrary:Sound Library Object Reference) -> (Output_Get:Sound Library Object Reference)
   > Clear(TargetArray:Array of Meta Sound Source Object References=Get BuiltMetaSounds() -> (BuiltMetaSounds:Array of Meta Sound Source Object References))
   > Clear(TargetArray:Array of Object References=Get CurrentEvaluatedWaves() -> (CurrentEvaluatedWaves:Array of Object References))
   > Clear!(
    execute
    TargetMap:Map of Gameplay Tag Structures to Sound Base Soft Object References=Get SoundLibrarySounds() -> (SoundLibrarySounds:Map of Gameplay Tag Structures to Sound Base Soft Object References)
   ) -> (then) {is=title K2Node_CallFunction}
   > ResetTextFields(self:EUW Meta Sound Asset Path Weapon Object Reference=Get EUW_MetaSoundAssetPath_Weapon() -> (EUW_MetaSoundAssetPath_Weapon:EUW Meta Sound Asset Path Weapon Object Reference))
   > n1 = Sequence()
   > n2 = Get!(execute:Exec) -> (then:Exec, else:Exec, 1pWidget:EUW Weapon MSName List Entry 1p Object Reference)
   > Set Text(
    Text="LOCTEXT(\"Asset Name\")"
    self=Get AssetNameText_1!(self:EUW Weapon MSName List Entry 1p Object Reference=@n2) -> (AssetNameText_1:Text Object Reference)
   )
   > Set CheckedState!(
    execute
    CheckedState:ECheckBoxState Enum=Checked
    self:Check Box Object Reference=Get EditorUtilityCheckBox_205!(self:EUW Weapon MSName List Entry 1p Object Reference=@n2) -> (EditorUtilityCheckBox_205:Editor Utility Check Box Object Reference)
   ) -> (then, Output_Get:ECheckBoxState Enum=Unchecked)
   > Set Is Checked!(execute, Is Checked:bool=true, self:EUW Weapon MSName List Entry 1p Object Reference=@n2) -> (then, Output_Get:bool)
   > CheckStateChanged_1p()
   n3 = Get!(execute=@n1.then_1) -> (then, else, 3pWidget:EUW Weapon MSName List Entry 1p Object Reference)
   > Set Text(Text="LOCTEXT(\"Asset Name\")", self=Get AssetNameText_1(self=@n3))
   > Set CheckedState(CheckedState=Checked, self=Get EditorUtilityCheckBox_205(self=@n3)) -> (Output_Get=Unchecked)
   > Set Is Checked(Is Checked=true, self=@n3)
   > CheckStateChanged_3p()
  ]}
 CreateOrAddSoundLibrary {nodes [
   CreateOrAddSoundLibrary() {is=title K2Node_FunctionEntry}
   > n1 = Get!(execute) -> (then, else, WeaponSoundLibrary:Sound Library Object Reference)
   > n2 = Set SoundLibrarySounds(SoundLibrarySounds:Map of Gameplay Tag Structures to Sound Base Soft Object References=Get Sounds!(self:Sound Library Object Reference=@n1) -> (Sounds:Map of Gameplay Tag Structures to Sound Base Soft Object References)) -> (Output_Get:Map of Gameplay Tag Structures to Sound Base Soft Object References)
   n3 = Get BuiltMetaSounds()
   Branch(execute=[@n15.then, @n2.then], Condition=Is Not Empty(TargetArray:Array of Meta Sound Source Object References=@n3))
   > n4 = For Each Loop(Array:Array of Meta Sound Source Object References=@n3) -> (Array Element:Meta Sound Source Object Reference)
   SetEditorProperty(
    execute=@n4.Completed
    Object=Get WeaponSoundLibrary() -> (WeaponSoundLibrary:Sound Library Object Reference)
    PropertyName=Sounds
    PropertyValue:Map of Gameplay Tag Structures to Sound Base Soft Object References=Get SoundLibrarySounds()
   )
   > Clear(TargetMap=Get SoundLibrarySounds())
   n5 = GetObjectName(Object=@n4)
   n6 = Branch(execute=@n4.LoopBody, Condition=Contains(SearchIn=@n5, Substring=1p))
   > Add!(
    execute
    TargetMap:Map of Gameplay Tag Structures to Sound Base Soft Object References=Get SoundLibrarySounds()
    Key:"Gameplay Tag Structure (by ref)"=(TagName="SoundLibrary.Weapon.Fire.1P")
    Value:"Sound Base Soft Object Reference (by ref)"=Make Soft Reference!(Input:Meta Sound Source Object Reference=@n4) -> (Output:Meta Sound Source Soft Object Reference)
   ) -> (then) {is=title K2Node_CallFunction}
   Branch(execute=@n6.else, Condition=Contains(SearchIn=@n5, Substring=3p))
   > Add(TargetMap=Get SoundLibrarySounds(), Key=(TagName="SoundLibrary.Weapon.Fire.3P"), Value=Make Soft Reference(Input=@n4))
   n7 = GetAssetTools(execute=@n1.else)
   n8 = GetAssetName!(self:EUW Meta Sound Asset Path Weapon Object Reference=Get EUW_MetaSoundAssetPath_Weapon()) -> (AssetText:Text, IsEmpty:bool)
   n9 = Replace(SourceString=Replace(SourceString="To String (Text)"(InText=@n8), From=MSS_), From=_Weapon_)
   n10 = GetPath!(self:EUW Meta Sound Asset Path Weapon Object Reference=Get EUW_MetaSoundAssetPath_Weapon()) -> (PathText:Text, bIsEmpty:bool)
   > n11 = CreateAsset(
    self=@n7
    AssetName=Append!(
     A=SL
     B=_
     C:String=Select(Option 0:String=@n9, Option 1:String=TempName, Index=@n8.IsEmpty) -> (ReturnValue:String)
     D:String=_
     E:String=Weapon
    ) -> (ReturnValue)
    PackagePath=Select(
     Option 0="To String (Text)"(InText=@n10)
     Option 1=GetCurrentContentBrowserPath!() -> (OutPath:String, ReturnValue:bool)
     Index=@n10.bIsEmpty
    )
    AssetClass=/Script/SoundLibrary
    CallingContext=Weapon
   )
   > n12 = Cast To SoundLibrary!(execute, Object=@n11) -> (then, CastFailed, AsSound Library:Sound Library Object Reference) {is=title K2Node_DynamicCast}
   > n13 = Set WeaponSoundLibrary(WeaponSoundLibrary=@n12)
   n14 = Get ParentMetaSound() -> (ParentMetaSound=NewEnumerator4)
   SetEditorProperty(
    execute=[@n16.then, @n13.then]
    Object=@n13
    PropertyName=Parent
    PropertyValue:Sound Library Object Reference=Select!(
     NewEnumerator0:Sound Library Object Reference=Select!(
      NewEnumerator4=/FortMetaSoundWeapon/Items/Shared/Common_WeaponRanged_Rifle_Lib
      NewEnumerator3=&Common_WeaponRanged_Rifle_Lib
      NewEnumerator5
      NewEnumerator6:Sound Library Object Reference
      Index:Enum_WeaponMetaSoundParents Enum=@n14
     ) -> (ReturnValue)
     NewEnumerator1:Sound Library Object Reference=Select(
      NewEnumerator4=/FortMetaSoundWeapon/Items/Shared/Common_WeaponRanged_SMG_Lib
      NewEnumerator3=&Common_WeaponRanged_SMG_Lib
      Index=@n14
     )
     NewEnumerator2:Sound Library Object Reference=Select(
      NewEnumerator4=/FortMetaSoundWeapon/Items/Shared/Common_WeaponRanged_Pistol_Lib
      NewEnumerator3=&Common_WeaponRanged_Pistol_Lib
      Index=@n14
     )
     NewEnumerator3:Sound Library Object Reference=Select(
      NewEnumerator4=/FortMetaSoundWeapon/Items/Shared/Common_WeaponRanged_Shotgun_Lib
      NewEnumerator3=&Common_WeaponRanged_Shotgun_Lib
      Index=@n14
     )
     NewEnumerator4:Sound Library Object Reference=Select(
      NewEnumerator4=/FortMetaSoundWeapon/Items/Shared/Common_WeaponRanged_Sniper_Lib
      NewEnumerator3=&Common_WeaponRanged_Sniper_Lib
      Index=@n14
     )
     NewEnumerator5:Sound Library Object Reference
     Index:Enum_WeaponMetaSoundClass Enum=Get WeaponType() -> (WeaponType=NewEnumerator0)
    ) -> (ReturnValue:Sound Library Object Reference)
   )
   > SetPriority!(
    execute
    self:Sound Library Object Reference=@n13
    InPriority:int=Select(
     NewEnumerator4:int=0
     NewEnumerator3:int=0
     NewEnumerator5:int=10
     NewEnumerator6:int=0
     Index=Get ParentMetaSound() -> (ParentMetaSound=NewEnumerator4)
    ) -> (ReturnValue:int)
   ) -> (then)
   > n15 = Set Metadata!(
    execute
    Metadata:Gameplay Tag Container Structure=(GameplayTags=((TagName="SoundLibrary.Weapon.Fire")))
    self:Sound Library Object Reference=@n13
   ) -> (then, Output_Get:Gameplay Tag Container Structure)
   n16 = SetEditorProperty(Object=@n13, PropertyName=bCanActAsParent, PropertyValue:bool=Get bCanActAsParent() -> (bCanActAsParent:bool))
   Set Channel!(execute, Channel:Gameplay Tag Structure=(TagName="SoundLibrary.Weapon"), self:Sound Library Object Reference=@n13) -> (then, Output_Get:Gameplay Tag Structure)
  ]}
 EventGraph {nodes [
   n1 = "On Property Changed (CustomPartDetailView)"!() -> (OutputDelegate, then, PropertyName:Name=None) {is=title K2Node_ComponentBoundEvent}
   > n2 = Branch(Condition="Equal (Name)"(A=@n1.PropertyName, B=ParentMetaSound))
   n3 = Get ParentMetaSound() -> (ParentMetaSound=NewEnumerator4)
   > Set 3pParentMS(3pParentMS:Meta Sound Source Object Reference=Select(
     NewEnumerator4:Meta Sound Source Object Reference=/FortMetaSoundWeapon/Items/Metasounds/MSS_Base_Weapon_Fire_3p_Complete
     NewEnumerator3:Meta Sound Source Object Reference=/FortMetaSoundWeapon/Items/Metasounds/MSS_Base_Weapon_Fire_3p_Simple
     NewEnumerator5:Meta Sound Source Object Reference=/FortMetaSoundWeapon/Items/Metasounds/FPS/MSS_FPS_Weapon_Fire_3p
     NewEnumerator6:Meta Sound Source Object Reference=&MSS_Base_Weapon_Fire_3p_Complete
     Index=@n3
    ) -> (ReturnValue:Meta Sound Source Object Reference)) -> (Output_Get:Meta Sound Source Object Reference)
   > Set 1pParentMS(1pParentMS:Meta Sound Source Object Reference=Select(
     NewEnumerator4=/FortMetaSoundWeapon/Items/Metasounds/MSS_Base_Weapon_Fire_1p_Complete
     NewEnumerator3=/FortMetaSoundWeapon/Items/Metasounds/MSS_Base_Weapon_Fire_1p_Simple
     NewEnumerator5=/FortMetaSoundWeapon/Items/Metasounds/FPS/MSS_FPS_Weapon_Fire_1p
     NewEnumerator6=&MSS_Base_Weapon_Fire_1p_Complete
     Index=@n3
    )) -> (Output_Get:Meta Sound Source Object Reference)
   n4 = Make Array!("[0]":String=Main Fire Close, "[1]":String=Main Fire Medium, "[2]":String=Main Fire Distant) -> (Array:Array of Strings)
   > Set 3pCategories(3pCategories:Array of Strings=Select(
     NewEnumerator4:Array of Strings=@n4
     NewEnumerator3:Array of Strings=@n4
     NewEnumerator5:Array of Strings=@n4
     NewEnumerator6:Array of Strings=Make Array!("[0]"=Main Fire Close, "[1]"=Main Fire Distant) -> (Array)
     Index=Get ParentMetaSound() -> (ParentMetaSound=NewEnumerator4)
    ) -> (ReturnValue:Array of Strings)) -> (Output_Get:Array of Strings)
   n5 = Get EUW_Audio_Weapons_SoundName_ListView_3P()
   > ResetActionWidgets(self:EUW Audio Weapons Sound Name List View Object Reference=@n5)
   n6 = Make Map!(
    Key 0:String=Main Fire
    Value 0:String=1p
    Key 1:String=Sweetener Layer
    Value 1:String=Sweetener
    Key 2:String=Mech Layer
    Value 2:String=Mech
    Key 3:String=On Stop Layer
    Value 3:String=Stop
    Key 4:String=Low Ammo Layer
    Value 4:String=Low
    Key 5:String=Main Fire Close
    Value 5:String=Close
    Key 6:String=Main Fire Medium
    Value 6:String=Medium
    Key 7:String=Main Fire Distant
    Value 7:String=Distant
   ) -> (Map:Map of Strings to Strings) {is=title K2Node_MakeMap}
   > Set TMapStringToStringInput!(
    execute
    TMapStringToStringInput:Map of Strings to Strings=Select(
     NewEnumerator4:Map of Strings to Strings=@n6
     NewEnumerator3:Map of Strings to Strings=@n6
     NewEnumerator5:Map of Strings to Strings=@n6
     NewEnumerator6:Map of Strings to Strings=Make Map!(
      Key 0=Main Fire
      Value 0=1p
      Key 1=Sweetener Layer
      Value 1=Sweetener
      Key 2=Mech Layer
      Value 2=Mech
      Key 3=On Stop Layer
      Value 3=Stop
      Key 4=Low Ammo Layer
      Value 4=Low
      Key 5=Main Fire Close
      Value 5=Close
      Key 6=Main Fire Distant
      Value 6=Distant
     ) -> (Map)
     Index=Get ParentMetaSound() -> (ParentMetaSound=NewEnumerator4)
    ) -> (ReturnValue:Map of Strings to Strings)
    self:EUW Audio Weapons Sound Name List View Object Reference=@n5
   ) -> (then, Output_Get:Map of Strings to Strings)
   n7 = Make Map!(
    Key 0=Main Fire
    Value 0:Name=MainFire.Sounds
    Key 1=Sweetener Layer
    Value 1:Name=Sweetener.Sounds
    Key 2=Mech Layer
    Value 2:Name=Mech.Sounds
    Key 3=On Stop Layer
    Value 3:Name=OnStopFire.Sounds
    Key 4=Low Ammo Layer
    Value 4:Name=LowAmmo.Sounds
    Key 5=Main Fire Close
    Value 5:Name=MainFire.SoundsClose
    Key 6=Main Fire Medium
    Value 6:Name=MainFire.SoundsMedium
    Key 7:String=Main Fire Distant
    Value 7:Name=MainFire.SoundsDistant
   ) -> (Map:Map of Strings to Names)
   > Set TMapStringToNameInput(TMapStringToNameInput:Map of Strings to Names=Select(
     NewEnumerator4:Map of Strings to Names=@n7
     NewEnumerator3:Map of Strings to Names=@n7
     NewEnumerator5:Map of Strings to Names=@n7
     NewEnumerator6:Map of Strings to Names=Make Map!(
      Key 0=Main Fire
      Value 0=MainFire.Sounds
      Key 1=Sweetener Layer
      Value 1=Sweetener.Sounds
      Key 2=Mech Layer
      Value 2=Mech.Sounds
      Key 3=On Stop Layer
      Value 3=OnStopFire.Sounds
      Key 4=Low Ammo Layer
      Value 4=LowAmmo.Sounds
      Key 5=Main Fire Close
      Value 5=MainFire.SoundsClose
      Key 6=Main Fire Distant
      Value 6=MainFire.SoundsMedium
     ) -> (Map)
     Index=Get ParentMetaSound() -> (ParentMetaSound=NewEnumerator4)
    ) -> (ReturnValue:Map of Strings to Names)) -> (Output_Get:Map of Strings to Names)
   > Set3pNameActionWidgets()
   Branch(execute=@n2.else, Condition="Equal (Name)"(A=@n1.PropertyName, B=WeaponType))
   > Set WeaponSoundData(WeaponSoundData:B Weapon Types Data Object Reference=Select!(
     NewEnumerator0:B Weapon Types Data Object Reference=/Game/Sounds/Fort_Audio_Blueprints/Utilities/EditorUtilityWidgets/Metasounds/Weapons/DefaultDataAssets/DA_RifleSounds
     NewEnumerator1:B Weapon Types Data Object Reference=/Game/Sounds/Fort_Audio_Blueprints/Utilities/EditorUtilityWidgets/Metasounds/Weapons/DefaultDataAssets/DA_SMGSounds
     NewEnumerator2:B Weapon Types Data Object Reference=/Game/Sounds/Fort_Audio_Blueprints/Utilities/EditorUtilityWidgets/Metasounds/Weapons/DefaultDataAssets/DA_PistolSounds
     NewEnumerator3:B Weapon Types Data Object Reference=/Game/Sounds/Fort_Audio_Blueprints/Utilities/EditorUtilityWidgets/Metasounds/Weapons/DefaultDataAssets/DA_ShotgunSounds
     NewEnumerator4:B Weapon Types Data Object Reference=/Game/Sounds/Fort_Audio_Blueprints/Utilities/EditorUtilityWidgets/Metasounds/Weapons/DefaultDataAssets/DA_SniperSounds
     NewEnumerator5:B Weapon Types Data Object Reference
     Index:Enum_WeaponMetaSoundClass Enum=Get WeaponType() -> (WeaponType=NewEnumerator0)
    ) -> (ReturnValue:B Weapon Types Data Object Reference)) -> (Output_Get:B Weapon Types Data Object Reference)
   "On Clicked (CreateAssetsButton)"() {is=title K2Node_ComponentBoundEvent}
   > StartMetaSoundBuilder()
   "Event on Asset Path Button (EUW_MetaSoundAssetPath_Weapon)"() {is=title K2Node_ComponentBoundEvent}
   > n8 = SetAssetPath!(
    execute
    self:EUW Meta Sound Asset Path Weapon Object Reference=Get EUW_MetaSoundAssetPath_Weapon()
    Object:Object Reference="Get (a copy)"(Array:Array of Object References=Get SelectedAssets()) -> (Output:Object Reference)
   ) -> (then, bSuccess:bool)
   > n9 = Branch(Condition=@n8)
   WarningDialogue!(execute=@n9.else, self, Message:"Text (by ref)"=MakeLiteralText(Value="LOCTEXT(\"Please select a valid asset in the Content Browser\")")) -> (then)
   "On Clicked (ResetToolButton)"() {is=title K2Node_ComponentBoundEvent}
   > ClearAll() {is=title K2Node_CallFunction}
   "On Inputs Text Updated (EUW_MetaSoundAssetPath_Weapon)"() {is=title K2Node_ComponentBoundEvent}
   > UpdateMetaSoundVisualNames()
   SetStringInput() {is=title K2Node_CustomEvent}
   n10 = Get StringToStringInput() -> (StringToStringInput:Map of Strings to Strings)
   > SetStringToStringMap!(
    execute
    self:EUW Audio Weapons Sound Name List View Object Reference=Get EUW_Audio_Weapons_SoundName_ListView_1P()
    TMapStringToStringInput:Map of Strings to Strings=@n10
   ) -> (then)
   > SetStringToStringMap(self=Get EUW_Audio_Weapons_SoundName_ListView_3P(), TMapStringToStringInput=@n10)
   Set3pNameActionWidgets() {is=title K2Node_CustomEvent}
   > GetSoundNameReferences!(
    execute
    self:EUW Audio Weapons Sound Name List View Object Reference=Get EUW_Audio_Weapons_SoundName_ListView_3P()
    Skip:bool=false
    SourceWaveNames:Array of Strings=Get 3pCategories() -> (3pCategories:Array of Strings)
    SoundGroupType:String=3PWeapon
   ) -> (then)
   Set1pNameActionWidgets() {is=title K2Node_CustomEvent}
   > GetSoundNameReferences(
    self=Get EUW_Audio_Weapons_SoundName_ListView_1P()
    Skip=false
    SourceWaveNames=Get 1pCategories() -> (1pCategories:Array of Strings)
    SoundGroupType=1PWeapon
   )
   n11 = CheckStateChanged_1p() {is=title K2Node_CustomEvent}
   n12 = Get Is Checked!(self:EUW Weapon MSName List Entry 1p Object Reference=Get 1pWidget() -> (1pWidget:EUW Weapon MSName List Entry 1p Object Reference)) -> (Is Checked:bool)
   > SetVisibility(
    self=Get EUW_Audio_Weapons_SoundName_ListView_1P()
    InVisibility=Select!(Option 0:ESlateVisibility Enum=Hidden, Option 1:ESlateVisibility Enum=Visible, Index:bool=@n12) -> (ReturnValue:ESlateVisibility Enum)
   )
   > SetVisibility(self=Get TextBlock_0() -> (TextBlock_0:Text Object Reference), InVisibility=Select(Option 0=Visible, Option 1=Hidden, Index=@n12))
   n13 = CheckStateChanged_3p() {is=title K2Node_CustomEvent}
   n14 = Get Is Checked(self=Get 3pWidget() -> (3pWidget:EUW Weapon MSName List Entry 1p Object Reference))
   > SetVisibility(self=Get EUW_Audio_Weapons_SoundName_ListView_3P(), InVisibility=Select(Option 0=Hidden, Option 1=Visible, Index=@n14))
   > SetVisibility(self=Get TextBlock_2() -> (TextBlock_2:Text Object Reference), InVisibility=Select(Option 0=Visible, Option 1=Hidden, Index=@n14))
   Event PreConstruct()
   > SetObject(self=Get CustomPartDetailView() -> (CustomPartDetailView:Details View Object Reference), NewObject=Self-Reference())
   Event Construct()
   > SetStringInput() {is=title K2Node_CallFunction}
   > n15 = Branch(Condition=Is Valid Timer Handle!(Handle:Timer Handle Structure=Get SelectionTimerHandle() -> (SelectionTimerHandle:Timer Handle Structure)) -> (ReturnValue:bool))
   Set1pNameActionWidgets(execute=[@n15.then, @n23.then]) {is=title K2Node_CallFunction}
   > Set3pNameActionWidgets() {is=title K2Node_CallFunction}
   > SetMetaSoundList()
   > Set 3pParentMS(3pParentMS=&MSS_Base_Weapon_Fire_3p_Complete)
   > Set 1pParentMS(1pParentMS=&MSS_Base_Weapon_Fire_1p_Complete)
   > Set WeaponSoundData(WeaponSoundData=&DA_RifleSounds)
   > Clear(TargetArray:Array of Meta Sound Source Object References=Get BuiltMetaSounds()) {is=title K2Node_CallArrayFunction}
   > Set Text(
    Text="LOCTEXT(\"1P Selected Sounds List\")"
    self=Get SoundListText!(self:EUW Audio Weapons Sound Name List View Object Reference=Get EUW_Audio_Weapons_SoundName_ListView_1P()) -> (SoundListText:Text Object Reference)
   )
   > Set Text(Text="LOCTEXT(\"3P Selected Sounds List\")", self=Get SoundListText(self=Get EUW_Audio_Weapons_SoundName_ListView_3P()))
   > Delay!(execute, WorldContextObject:Object Reference, Duration=1) -> (then)
   n16 = GetDisplayedEntryWidgets!(self:List View Base Object Reference=Get WeaponMetaSoundList() -> (WeaponMetaSoundList:Editor Utility List View Object Reference)) -> (ReturnValue:Array of User Widget Object References)
   > n17 = Cast To EUW_WeaponMSName_ListEntry_1p!(execute, Object="Get (a copy)"(Array:Array of User Widget Object References=@n16) -> (Output:User Widget Object Reference)) -> (then, CastFailed, AsEUW Weapon MSName List Entry 1p:EUW Weapon MSName List Entry 1p Object Reference) {is=title K2Node_DynamicCast}
   > n18 = Set 1pWidget(1pWidget:EUW Weapon MSName List Entry 1p Object Reference=@n17) -> (Output_Get:EUW Weapon MSName List Entry 1p Object Reference)
   > n19 = Assign Check State Changed!(execute:Exec, self:EUW Weapon MSName List Entry 1p Object Reference=@n18, Delegate:"Delegate (by ref)"=@n11) -> (then:Exec) {is=title K2Node_AssignDelegate}
   Branch(
    execute=[@n19.then, @n17.CastFailed]
    Condition=Is Valid Index!(TargetArray:Array of User Widget Object References=@n16, IndexToTest:int=1) -> (ReturnValue:bool)
   )
   > n20 = Cast To EUW_WeaponMSName_ListEntry_1p(Object="Get (a copy)"(Array=@n16, Dimension 1=1))
   > n21 = Set 3pWidget(3pWidget:EUW Weapon MSName List Entry 1p Object Reference=@n20) -> (Output_Get:EUW Weapon MSName List Entry 1p Object Reference)
   > Assign Check State Changed(self=@n21, Delegate=@n13)
   > Set Text(
    Text="LOCTEXT(\"3P Weapon MetaSound\")"
    self=Get SpokenTextBlock!(self:EUW Weapon MSName List Entry 1p Object Reference=@n21) -> (SpokenTextBlock:Text Object Reference)
   )
   n22 = Set Timer by Event(execute=@n15.else, Delegate=Create Event(), Time=0.2, bLooping=true)
   > n23 = Set SelectionTimerHandle(SelectionTimerHandle:Timer Handle Structure=@n22) -> (Output_Get:Timer Handle Structure)
   Event Destruct()
   n24 = Get SelectionTimerHandle()
   > Branch(Condition=Is Valid Timer Handle(Handle=@n24))
   > Clear and Invalidate Timer by Handle!(execute, WorldContextObject:Object Reference, Handle=@n24) -> (then)
  ]}
 SaveAssets {nodes [
   SaveAssets() {is=title K2Node_FunctionEntry}
   n1 = Get BuiltMetaSounds()
   > n2 = Branch(Condition=Is Not Empty(TargetArray=@n1))
   > n3 = SaveLoadedAssets(AssetsToSave=@n1, bOnlyIfIsDirty=false)
   n4 = Get(execute=@n2.else)
   n5 = SaveLoadedAsset(execute=[@n3.then, @n4.then], AssetToSave=@n4, bOnlyIfIsDirty=false)
   n6 = Get WeaponID()
   Branch(execute=[@n5.then, @n4.else], Condition=Is Not Empty(TargetArray:Array of Fort Weapon Ranged Item Definition Object References=@n6))
   > SaveLoadedAssets(AssetsToSave=@n6, bOnlyIfIsDirty=false)
  ]}
 SetMetaSoundList {nodes [
   SetMetaSoundList() {is=title K2Node_FunctionEntry}
   n1 = Get WeaponMetaSoundList()
   > RegenerateAllEntries(self:List View Base Object Reference=@n1)
   > n2 = For Each Loop(Array:Array of Object References=Get MetaSoundTypesList() -> (MetaSoundTypesList:Array of Object References)) -> (Array Element:Object Reference)
   > AddItem!(execute, self:List View Object Reference=@n1, Item:Object Reference=@n2) -> (then)
  ]}
 SetToBuildMetaSounds {nodes [
   SetToBuildMetaSounds() {is=title K2Node_FunctionEntry}
   > n1 = Sequence()
   > n2 = Switch on ESlateVisibility!(
    execute
    Selection:ESlateVisibility Enum=GetVisibility!(self:Widget Object Reference=Get EUW_Audio_Weapons_SoundName_ListView_1P()) -> (ReturnValue:ESlateVisibility Enum=Visible)
   ) -> (Visible:Exec, Collapsed:Exec, Hidden:Exec, HitTestInvisible:Exec, SelfHitTestInvisible:Exec) {is=title K2Node_SwitchEnum}
   n3 = Get ToBuildMetaSounds() -> (ToBuildMetaSounds:Array of Names)
   Remove Item(execute=@n2.Hidden, TargetArray:Array of Names=@n3, Item:"Name (by ref)"=1p)
   Add Unique(execute=@n2.Visible, TargetArray:Array of Names=@n3, NewItem:"Name (by ref)"=1p)
   n4 = Switch on ESlateVisibility(execute=@n1.then_1, Selection=GetVisibility(self=Get EUW_Audio_Weapons_SoundName_ListView_3P()) -> (ReturnValue=Visible))
   n5 = Get ToBuildMetaSounds()
   Remove Item(execute=@n4.Hidden, TargetArray=@n5, Item=3p)
   Add Unique(execute=@n4.Visible, TargetArray=@n5, NewItem=3p)
  ]}
 SortSelectedWaves {nodes [
   n1 = SortSelectedWaves!() -> (then, ActionNameWidget:EUW Audio Weapons Action Name Widget Object Reference) {is=title K2Node_FunctionEntry}
   > Clear(TargetArray:Array of Object References=Get CurrentEvaluatedWaves())
   > n2 = For Each Loop(Array:Array of Widget Object References=GetAllChildren(self=Get ScrollBox!(self:EUW Audio Weapons Action Name Widget Object Reference=@n1) -> (ScrollBox:Editor Utility Scroll Box Object Reference))) -> (Array Element:Widget Object Reference)
   Return Node!(execute=@n2.Completed, SortedWaves:Array of Object References=Get CurrentEvaluatedWaves())
   > n3 = Cast To EUW_Audio_Weapons_SoundNameObject!(execute, Object=@n2) -> (then, CastFailed, AsEUW Audio Weapons Sound Name Object:EUW Audio Weapons Sound Name Object Object Reference) {is=title K2Node_DynamicCast}
   > Add Unique(
    TargetArray:Array of Object References=Get CurrentEvaluatedWaves()
    NewItem:"Object Reference (by ref)"=Get SoundObjectRef!(self:EUW Audio Weapons Sound Name Object Object Reference=@n3) -> (SoundObjectRef:Object Reference)
   )
  ]}
 StartMetaSoundBuilder {nodes [
   StartMetaSoundBuilder() {is=title K2Node_FunctionEntry}
   > n1 = Branch(Condition=GetPath(self=Get EUW_MetaSoundAssetPath_Weapon()).bIsEmpty)
   > WarningDialogue(Message=MakeLiteralText(Value="LOCTEXT(\"Build or Asset path are empty. Please select a valid directory and try again.\")"))
   SetToBuildMetaSounds(execute=@n1.else) {is=title K2Node_CallFunction}
   n2 = Get ToBuildMetaSounds()
   > n3 = Branch(Condition=Is Not Empty(TargetArray:Array of Names=@n2))
   > Clear(TargetArray:Array of Meta Sound Source Object References=Get BuiltMetaSounds())
   > n4 = For Each Loop(Array:Array of Names=@n2) -> (Array Element:Name)
   CreateOrAddSoundLibrary(execute=@n4.Completed) {is=title K2Node_CallFunction}
   > WeaponItemImplementation()
   > SaveAssets() {is=title K2Node_CallFunction}
   Set bIs1p(execute=@n4.LoopBody, bIs1p:bool="Equal (Name)"(A=@n4, B=1p)) -> (Output_Get:bool)
   > n5 = Cast To MetaSoundDocumentInterface!(
    execute
    Object=Select(
     Option 0:Meta Sound Source Object Reference=Get 3pParentMS() -> (3pParentMS:Meta Sound Source Object Reference)
     Option 1:Meta Sound Source Object Reference=Get 1pParentMS() -> (1pParentMS:Meta Sound Source Object Reference)
     Index=Get bIs1p()
    ) -> (ReturnValue:Meta Sound Source Object Reference)
   ) -> (then, CastFailed, AsMetaSound Document Interface:Meta Sound Document Interface Interface) {is=title K2Node_DynamicCast}
   n6 = Select(Option 0:String=_3p, Option 1:String=_1p, Index=Get bIs1p()) -> (ReturnValue:String)
   > n7 = CreateSourcePresetBuilder!(
    execute
    self:Meta Sound Builder Subsystem Object Reference=Get MetaSoundBuilderSubsystem()
    BuilderName:Name=String To Name(InString=Append!(A=WeaponBuilder, B=@n6) -> (ReturnValue))
    ReferencedSourceClass:"Meta Sound Document Interface Interface (by ref)"=@n5
   ) -> (Succeeded:Exec, Failed:Exec, ReturnValue:Meta Sound Source Builder Object Reference)
   > Set MetaSoundPresetBuilder(MetaSoundPresetBuilder:Meta Sound Source Builder Object Reference=@n7) -> (Output_Get:Meta Sound Source Builder Object Reference)
   n8 = Get EUW_Audio_Weapons_SoundName_ListView_3P()
   n9 = Get EUW_Audio_Weapons_SoundName_ListView_1P()
   > n10 = For Each Loop(Array:Array of EUW Audio Weapons Action Name Widget Object References=Select(
     Option 0:Array of EUW Audio Weapons Action Name Widget Object References=Get WeaponActionNameWidgets!(self:EUW Audio Weapons Sound Name List View Object Reference=@n8) -> (WeaponActionNameWidgets:Array of EUW Audio Weapons Action Name Widget Object References)
     Option 1:Array of EUW Audio Weapons Action Name Widget Object References=Get WeaponActionNameWidgets(self=@n9)
     Index=Get bIs1p()
    ) -> (ReturnValue:Array of EUW Audio Weapons Action Name Widget Object References)) -> (Array Element:EUW Audio Weapons Action Name Widget Object Reference)
   AssignDefaultWaves(execute=@n10.Completed) {is=title K2Node_CallFunction}
   > n11 = Sequence()
   > Branch(Condition="Equal (Enum)"(
     A:Enum_WeaponMetaSoundParents Enum=Get ParentMetaSound() -> (ParentMetaSound=NewEnumerator4)
     B:Enum_WeaponMetaSoundParents Enum=NewEnumerator6
    ))
   > n12 = Get!(execute) -> (then, else, bIs1p:bool)
   n13 = Create MetaSound Literal From AudioParameter(
    execute=@n12.else
    self=Get MetaSoundBuilderSubsystem()
    Param=ObjectToAudioParameter!(
     Name:Name=MainFire.DistanceCurves
     Object:Object Reference=/FortMetaSoundWeapon/Items/Shared/Settings/Attenuation/WTB_Weapon_Ranged_DistanceCurves
    ) -> (ReturnValue:Audio Parameter Structure)
   )
   > SetGraphInputDefault(self=Get MetaSoundPresetBuilder(), InputName=MainFire.DistanceCurves, Literal=@n13)
   n14 = Get EUW_MetaSoundAssetPath_Weapon()
   n15 = Get EUW_MetaSoundAssetPath_Weapon()
   n16 = Append(
    A="To String (Text)"(InText=GetAssetName(self=@n15))
    B=Select(Option 0:String=_3p_Preset, Option 1:String=_1p_Preset, Index=Get bIs1p()) -> (ReturnValue:String)
   )
   n17 = BuildToAsset!(
    execute=@n11.then_1
    self:Meta Sound Editor Subsystem Object Reference=Get MetaSoundEditorSubsystem() -> (ReturnValue:Meta Sound Editor Subsystem Object Reference) {is=title K2Node_GetEditorSubsystem}
    InBuilder:Meta Sound Builder Base Object Reference=Get MetaSoundPresetBuilder()
    Author:String=SoundBaseAssetAction
    AssetName:String=@n16
    PackagePath:String="To String (Text)"(InText=GetPath(self=@n14))
    TemplateSoundWave:Sound Wave Object Reference
   ) -> (Succeeded:Exec, Failed:Exec, ReturnValue:Meta Sound Document Interface Interface)
   WarningDialogue(
    execute=@n17.Failed
    Message=Format Text!(Format="LOCTEXT(\"Failed to build {string}, please close and restart the tool.\")", string:String=@n16) -> (Result)
   )
   > Return Node!(execute)
   n18 = Cast To MetaSoundSource(execute=@n17.Succeeded, Object:Meta Sound Document Interface Interface=@n17)
   > Add Unique(TargetArray:Array of Meta Sound Source Object References=Get BuiltMetaSounds(), NewItem:"Meta Sound Source Object Reference (by ref)"=@n18)
   > Set QualitySetting!(execute, QualitySetting:Name=Fort_Weapons, self:Meta Sound Source Object Reference=@n18) -> (then, Output_Get:Name=None)
   > Branch(Condition="Equal (Enum)"(A=Get ParentMetaSound() -> (ParentMetaSound=NewEnumerator4), B=NewEnumerator6))
   > n19 = Get()
   SetEditorProperty(
    execute=@n19.else
    Object=@n18
    PropertyName=AttenuationSettings
    PropertyValue:Sound Attenuation Object Reference=Get 3pAttenuationSettings() -> (3pAttenuationSettings:Sound Attenuation Object Reference)
   )
   n20 = SortSelectedWaves!(execute=@n10.LoopBody, self, ActionNameWidget:EUW Audio Weapons Action Name Widget Object Reference=@n10) -> (then, SortedWaves:Array of Object References) {is=title K2Node_CallFunction}
   n21 = "GetText (Text)"!(self:Text Object Reference=Get PresetText!(self:EUW Audio Weapons Action Name Widget Object Reference=@n10) -> (PresetText:Text Object Reference)) -> (ReturnValue:Text)
   > AssignSelectedWaves(
    InputName=Find(
     TargetMap:Map of Strings to Names=Get TMapStringToNameInput() -> (TMapStringToNameInput:Map of Strings to Names)
     Key="To String (Text)"(InText=@n21)
    ) -> (Value:Name)
    WaveArray=@n20
   ) {is=title K2Node_CallFunction}
   WarningDialogue(execute=@n3.else, Message=MakeLiteralText(Value="LOCTEXT(\"Internal Error, please close and restart the tool.\")"))
  ]}
 UpdateMetaSoundVisualNames {nodes [
   UpdateMetaSoundVisualNames() {is=title K2Node_FunctionEntry}
   > n1 = Sequence()
   n2 = GetDisplayedEntryWidgets(self=Get WeaponMetaSoundList())
   > n3 = Cast To EUW_WeaponMSName_ListEntry_1p(Object="Get (a copy)"(Array=@n2))
   > Set Text(Text=GetAssetName(self=Get EUW_MetaSoundAssetPath_Weapon()), self=Get AssetNameText_1(self=@n3))
   n4 = Cast To EUW_WeaponMSName_ListEntry_1p(execute=@n1.then_1, Object="Get (a copy)"(Array=@n2, Dimension 1=1))
   > Set Text(Text=GetAssetName(self=Get EUW_MetaSoundAssetPath_Weapon()), self=Get AssetNameText_1(self=@n4))
  ]}
 WarningDialogue {nodes [
   n1 = WarningDialogue!() -> (then, Message:"Text (by ref)") {is=title K2Node_FunctionEntry}
   > Show Message Dialog(
    Title=MakeLiteralText(Value="LOCTEXT(\"Weapon MS Creator Tool Warning!\")")
    Message=@n1
    MessageType=Ok
    DefaultValue=No
    MessageCategory=Warning
   )
  ]}
 WeaponItemImplementation {nodes [
   WeaponItemImplementation() {is=title K2Node_FunctionEntry}
   n1 = Get WeaponID()
   > Branch(Condition=Is Not Empty(TargetArray:Array of Fort Weapon Ranged Item Definition Object References=@n1))
   > n2 = For Each Loop(Array:Array of Fort Weapon Ranged Item Definition Object References=@n1) -> (Array Element:Fort Weapon Ranged Item Definition Object Reference)
   > n3 = GetItemData(self=@n2) -> (ItemData:"Fortnite Weapon Sound Libraries Item Data Structure (by ref)")
   > n4 = RemoveItemData(self=@n2, StructType=/Script/FortniteGame.FortItemComponentData_WeaponSoundLibraries)
   n5 = Get!(execute=[@n4.Found, @n3.NotFound]) -> (then, else, WeaponSoundLibrary:Sound Library Object Reference)
   n6 = Make Array!("[0]":Sound Library Object Reference=@n5) -> (Array:Array of Sound Library Object References)
   n7 = Get bIsUnvault()
   n8 = Get WeaponCollectionTag() -> (WeaponCollectionTag:Gameplay Tag Container Structure)
   > n9 = SetItemData(
    self=@n2
    ItemData:Fortnite Weapon Sound Libraries Item Data Structure=Make Fortnite Weapon Sound Libraries Item Data!(
     WeaponSoundLibraries:Array of Sound Library Object References=Select(Option 0:Array of Sound Library Object References=@n6, Option 1:Array of Sound Library Object References, Index=@n7) -> (ReturnValue:Array of Sound Library Object References)
     WeaponOwnerSoundLibraries:Array of Sound Library Object References
     WeaponCollectionTag:Gameplay Tag Container Structure=Select(Option 0:Gameplay Tag Container Structure, Option 1:Gameplay Tag Container Structure=@n8, Index=@n7) -> (ReturnValue:Gameplay Tag Container Structure)
    ) -> (FortItemComponentData_WeaponSoundLibraries:Fortnite Weapon Sound Libraries Item Data Structure) {is=title K2Node_MakeStruct}
   )
   BulletWhipImplementation!(execute=[@n9.then, @n5.else], self, ItemDefinition:Item Definition Base Object Reference=@n2) -> (then, Item:Item Definition Base Object Reference) {is=title K2Node_CallFunction}
   > Branch(Condition=Get bIsUnvault())
   > OpenEditorForAssets!(
    execute
    self:Asset Editor Subsystem Object Reference=Get AssetEditorSubsystem()
    Assets:Array of Object References=Make Array("[0]":Object Reference=/WeaponsUnvaulted/Assets/Audio/SLC_UnvaultedWeapons) -> (Array:Array of Object References)
    OpenedMethod:EAssetTypeActivationOpenedMethod Enum=Edit
   ) -> (then, ReturnValue:bool)
   Break Fortnite Weapon Sound Libraries Item Data!(FortItemComponentData_WeaponSoundLibraries:"Fortnite Weapon Sound Libraries Item Data Structure (by ref)"=@n3) -> (
    WeaponSoundLibraries:Array of Sound Library Object References
    WeaponOwnerSoundLibraries:Array of Sound Library Object References
    WeaponCollectionTag:Gameplay Tag Container Structure
   ) {is=title K2Node_BreakStruct}
  ]}
]
variables [
 1pCategories {type="array<string>", default [Main Fire, Mech Layer, Sweetener Layer, Low Ammo Layer, On Stop Layer]}
 1pParentMS {type=MetaSoundSource, default=null, class=~, properties=~}
 1pWidget {type=EUW_WeaponMSName_ListEntry_1p_C, default=null}
 3pAttenuationSettings {type=SoundAttenuation, default=/FortMetaSoundWeapon/Items/Shared/Settings/Attenuation/ATT_Weapon_Ranged_3P}
 3pCategories {type="array<string>", default [Main Fire Close, Main Fire Medium, Main Fire Distant]}
 3pParentMS {type=MetaSoundSource, default=null, class=~, properties=~}
 3pWidget {type=EUW_WeaponMSName_ListEntry_1p_C, default=null}
 bCanActAsParent {type=boolean, default=true}
 bIs1p {type=boolean}
 bIsUnvault {type=boolean}
 bLowAmmoSet {type=boolean}
 BuiltMetaSounds {type="array<MetaSoundSource>", default []}
 CurrentEvaluatedWaves {type="array<Object>", default []}
 MetaSoundPresetBuilder {type=MetaSoundSourceBuilder, default=null}
 MetaSoundTypesList {
  type="array<Object>"
  default [
   /Game/Sounds/Fort_Audio_Blueprints/Utilities/EditorUtilityWidgets/Metasounds/Weapons/EUW_WeaponMSName_ListEntry_1p
   /Game/Sounds/Fort_Audio_Blueprints/Utilities/EditorUtilityWidgets/Metasounds/Weapons/EUW_WeaponMSName_ListEntry_3p
  ]
 }
 ParentMetaSound {type=Enum_WeaponMetaSoundParents, default=Complete}
 SelectedAssets {type="array<Object>", default []}
 SelectionTimerHandle {type=TimerHandle}
 SoundLibrarySounds {default {}}
 StringToStringInput {
  type="map<string>"
  default {
   Low Ammo Layer=Low
   Main Fire=1p
   Main Fire Close=Close
   Main Fire Distant=Distant
   Main Fire Medium=Medium
   Mech Layer=Mech
   On Stop Layer=Stop
   Sweetener Layer=Sweetener
  }
 }
 TMapStringToNameInput {
  type="map<string>"
  default {
   Low Ammo Layer=LowAmmo.Sounds
   Main Fire=MainFire.Sounds
   Main Fire Close=MainFire.SoundsClose
   Main Fire Distant=MainFire.SoundsDistant
   Main Fire Medium=MainFire.SoundsMedium
   Mech Layer=Mech.Sounds
   On Stop Layer=OnStopFire.Sounds
   Sweetener Layer=Sweetener.Sounds
  }
 }
 ToBuildMetaSounds {type="array<string>", default [1p,3p]}
 WaveParameters {
  type=AudioParameter
  default {
   boolParam=false
   floatParam=0
   intParam=0
   objectParam=null
   paramName=None
   paramType=None
   stringParam=""
   typeName=None
   arrayBoolParam []
   arrayFloatParam []
   arrayIntParam []
   arrayObjectParam []
   arrayStringParam []
  }
 }
 WeaponCollectionTag {type=GameplayTagContainer, default {gameplayTags []}}
 WeaponID {type="array<FortWeaponRangedItemDefinition>", default []}
 WeaponSoundData {type=B_WeaponTypesData_C, default=null}
 WeaponSoundLibrary {type=SoundLibrary, default=null}
 WeaponType {type=Enum_WeaponMetaSoundClass, default=Rifle}
]
widgets {
 info {namedSlotCount=2, rootWidgetClass=&CanvasPanel, widgetCount=22}
 widgets [
  CanvasPanel_167 {widgetClassPath=&CanvasPanel, widget=CanvasPanel_167}
  VerticalBox_29 {
   widgetClassPath=&VerticalBox
   parent=CanvasPanel_167
   widget=VerticalBox_29
   slot {
    class=CanvasPanelSlot
    properties {layoutData {alignment=<0 0>, anchors {maximum=<0 0>, minimum=<0 0>}, offsets=<ltrb 0 159.1115 481.0128 399.2706>}}
   }
  }
  SizeBox_72 {
   widgetClassPath=&SizeBox
   parent=VerticalBox_29
   widget=SizeBox_72
   slot {class=VerticalBoxSlot, properties {size {sizeRule=Fill, value=1}}}
  }
  CustomPartDetailView {
   widgetClassPath=&DetailsView
   parent=SizeBox_72
   widget=CustomPartDetailView
   properties {columnWidth=0.5, categoriesToShow [WID Implementation, Weapon MetaSound Creator, Core Game Unvault Options, Sound Library Set Up]}
   slot {class=SizeBoxSlot}
  }
  WeaponMetaSoundsBox {
   widgetClassPath=&HorizontalBox
   parent=CanvasPanel_167
   widget=WeaponMetaSoundsBox
   slot {
    class=CanvasPanelSlot
    properties {layoutData {alignment=<0 0>, anchors {maximum=<0 0>, minimum=<0 0>}, offsets=<ltrb 0 567.4431 479.7994 30>}}
   }
  }
  WeaponMetaSoundExpandable {
   widgetClassPath=/Script/Blutility.EditorUtilityExpandableArea
   parent=WeaponMetaSoundsBox
   widget=WeaponMetaSoundExpandable
   properties {
    bIsExpanded=true
    borderBrush {
     bIsDynamicallyLoaded=false
     drawAs=RoundedBox
     imageType=NoImage
     mirroring=NoMirror
     resourceName=None
     resourceObject=null
     tiling=NoTile
     imageSize=<0 0>
     margin=<ltrb 0 0 0 0>
     outlineSettings {
      bUseBrushTransparency=false
      roundingType=FixedRadius
      width=1
      color {colorUseRule=UseColor_Specified, specifiedColor=<rgba 0 0 0 0>}
      cornerRadii=<4 4 4 4>
     }
     tintColor {colorUseRule=UseColor_Specified, specifiedColor=<rgba 0.0284 0.0284 0.0284 1>}
     uVRegion {bIsValid=false, max=<0 0>, min=<0 0>}
    }
    headerPadding=<ltrb 4 4 4 4>
   }
   slot {class=HorizontalBoxSlot, properties {size {sizeRule=Fill, value=1}}}
  }
  TextBlock_335 {
   widgetClassPath=&TextBlock
   namedSlotHost=WeaponMetaSoundExpandable
   widget=TextBlock_335
   properties {
    text=Weapon MetaSounds
    colorAndOpacity {colorUseRule=UseColor_Specified, specifiedColor=<rgba 0.5029 0.5029 0.5029 1>}
    font {
     bForceMonospaced=false
     bMaterialIsStencil=false
     fontMaterial=null
     fontObject=&Roboto
     letterSpacing=0
     monospacedWidth=1
     size=9.75
     skewAmount=0
     typefaceFontName=Bold
     outlineSettings {
      bApplyOutlineToDropShadows=false
      bMiteredCorners=false
      bSeparateFillAlpha=false
      outlineMaterial=null
      outlineSize=0
      outlineColor=<rgba 0 0 0 1>
     }
    }
    margin=<ltrb 0 2 0 0>
   }
  }
  VerticalBox_32 {widgetClassPath=&VerticalBox, namedSlotHost=WeaponMetaSoundExpandable, widget=VerticalBox_32}
  WeaponMetaSoundList {
   widgetClassPath=/Script/Blutility.EditorUtilityListView
   bIsVariable=true
   parent=VerticalBox_32
   widget=WeaponMetaSoundList
   properties {
    bAllowDragging=false
    bClearSelectionOnClick=true
    bEnableScrollAnimation=true
    entryWidgetClass=/Game/Sounds/Fort_Audio_Blueprints/Utilities/EditorUtilityWidgets/Metasounds/Weapons/EUW_WeaponMSName_ListEntry_1p.EUW_WeaponMSName_ListEntry_1p_C
    numDesignerPreviewEntries=2
   }
   slot {class=VerticalBoxSlot}
  }
  CreateAssetsButtonBox {
   widgetClassPath=&HorizontalBox
   parent=CanvasPanel_167
   widget=CreateAssetsButtonBox
   slot {
    class=CanvasPanelSlot
    properties {layoutData {alignment=<0 0>, anchors {maximum=<0 0>, minimum=<0 0>}, offsets=<ltrb 138.6813 714.927 216.48 30>}}
   }
  }
  WrapBox_2 {
   widgetClassPath=&WrapBox
   parent=CreateAssetsButtonBox
   widget=WrapBox_2
   properties {horizontalAlignment=HAlign_Fill, innerSlotPadding=<4 4>}
   slot {class=HorizontalBoxSlot, properties {size {sizeRule=Fill, value=1}}}
  }
  CreateAssetsButton {
   widgetClassPath=&EditorUtilityButton
   parent=WrapBox_2
   widget=CreateAssetsButton
   properties {widgetStyle {
     clickedSlateSound {resourceObject=null}
     disabled {
      bIsDynamicallyLoaded=false
      drawAs=RoundedBox
      imageType=NoImage
      mirroring=NoMirror
      resourceName=None
      resourceObject=null
      tiling=NoTile
      imageSize=<0 0>
      margin=<ltrb 0 0 0 0>
      outlineSettings {
       bUseBrushTransparency=false
       roundingType=FixedRadius
       width=1
       color {colorUseRule=UseColor_ColorTable, specifiedColor=<rgba 0 0 0 0>}
       cornerRadii=<4 4 4 4>
      }
      tintColor {colorUseRule=UseColor_Specified, specifiedColor=<rgba 0.0395 0.0395 0.0395 1>}
      uVRegion {bIsValid=false, max=<0 0>, min=<0 0>}
     }
     disabledForeground {colorUseRule=UseColor_ColorTable, specifiedColor=<rgba 0 0 0 0>}
     hovered {
      bIsDynamicallyLoaded=false
      drawAs=RoundedBox
      imageType=NoImage
      mirroring=NoMirror
      resourceName=None
      resourceObject=null
      tiling=NoTile
      imageSize=<32 32>
      margin=<ltrb 0 0 0 0>
      outlineSettings {
       bUseBrushTransparency=false
       roundingType=FixedRadius
       width=1
       color {colorUseRule=UseColor_ColorTable, specifiedColor=<rgba 0 0 0 0>}
       cornerRadii=<4 4 4 4>
      }
      tintColor {colorUseRule=UseColor_Specified, specifiedColor=<rgba 0.0044 0.2384 1 1>}
      uVRegion {bIsValid=false, max=<0 0>, min=<0 0>}
     }
     hoveredForeground {colorUseRule=UseColor_ColorTable, specifiedColor=<rgba 0 0 0 0>}
     hoveredSlateSound {resourceObject=null}
     normal {
      bIsDynamicallyLoaded=false
      drawAs=RoundedBox
      imageType=NoImage
      mirroring=NoMirror
      resourceName=None
      resourceObject=null
      tiling=NoTile
      imageSize=<32 32>
      margin=<ltrb 0 0 0 0>
      outlineSettings {
       bUseBrushTransparency=false
       roundingType=FixedRadius
       width=1
       color {colorUseRule=UseColor_ColorTable, specifiedColor=<rgba 0 0 0 0>}
       cornerRadii=<4 4 4 4>
      }
      tintColor {colorUseRule=UseColor_Specified, specifiedColor=<rgba 0 0.162 0.7454 1>}
      uVRegion {bIsValid=false, max=<0 0>, min=<0 0>}
     }
     normalForeground {colorUseRule=UseColor_ColorTable, specifiedColor=<rgba 0 0 0 0>}
     normalPadding=<ltrb 12 1.5 12 1.5>
     pressed {
      bIsDynamicallyLoaded=false
      drawAs=RoundedBox
      imageType=NoImage
      mirroring=NoMirror
      resourceName=None
      resourceObject=null
      tiling=NoTile
      imageSize=<32 32>
      margin=<ltrb 0 0 0 0>
      outlineSettings {
       bUseBrushTransparency=false
       roundingType=FixedRadius
       width=1
       color {colorUseRule=UseColor_ColorTable, specifiedColor=<rgba 0 0 0 0>}
       cornerRadii=<4 4 4 4>
      }
      tintColor {colorUseRule=UseColor_Specified, specifiedColor=<rgba 0 0.0802 0.3515 1>}
      uVRegion {bIsValid=false, max=<0 0>, min=<0 0>}
     }
     pressedForeground {colorUseRule=UseColor_ColorTable, specifiedColor=<rgba 0 0 0 0>}
     pressedPadding=<ltrb 12 2.5 12 0.5>
     pressedSlateSound {resourceObject=null}
    }}
   slot {class=WrapBoxSlot}
  }
  TextBlock_295 {
   widgetClassPath=&TextBlock
   parent=CreateAssetsButton
   widget=TextBlock_295
   properties {
    text=Create Weapon MetaSounds
    font {
     bForceMonospaced=false
     bMaterialIsStencil=false
     fontMaterial=null
     fontObject=&Burbank
     letterSpacing=0
     monospacedWidth=1
     size=12
     skewAmount=0
     typefaceFontName=Burbank-Black
     outlineSettings {
      bApplyOutlineToDropShadows=false
      bMiteredCorners=false
      bSeparateFillAlpha=false
      outlineMaterial=null
      outlineSize=0
      outlineColor=<rgba 0 0 0 1>
     }
    }
    margin=<ltrb 0 1 0 0>
   }
   slot {class=ButtonSlot}
  }
  ResetToolBox {
   widgetClassPath=&HorizontalBox
   parent=CanvasPanel_167
   widget=ResetToolBox
   slot {
    class=CanvasPanelSlot
    properties {layoutData {alignment=<0 0>, anchors {maximum=<0 0>, minimum=<0 0>}, offsets=<ltrb 138.6529 765.3105 216.48 30>}}
   }
  }
  WrapBox {
   widgetClassPath=&WrapBox
   parent=ResetToolBox
   widget=WrapBox
   properties {horizontalAlignment=HAlign_Fill, innerSlotPadding=<4 4>}
   slot {class=HorizontalBoxSlot, properties {size {sizeRule=Fill, value=1}}}
  }
  ResetToolButton {
   widgetClassPath=&EditorUtilityButton
   parent=WrapBox
   widget=ResetToolButton
   properties {widgetStyle {
     clickedSlateSound {resourceObject=null}
     disabled {
      bIsDynamicallyLoaded=false
      drawAs=RoundedBox
      imageType=NoImage
      mirroring=NoMirror
      resourceName=None
      resourceObject=null
      tiling=NoTile
      imageSize=<0 0>
      margin=<ltrb 0 0 0 0>
      outlineSettings {
       bUseBrushTransparency=false
       roundingType=FixedRadius
       width=1
       color {colorUseRule=UseColor_ColorTable, specifiedColor=<rgba 0 0 0 0>}
       cornerRadii=<4 4 4 4>
      }
      tintColor {colorUseRule=UseColor_Specified, specifiedColor=<rgba 0.0395 0.0395 0.0395 1>}
      uVRegion {bIsValid=false, max=<0 0>, min=<0 0>}
     }
     disabledForeground {colorUseRule=UseColor_ColorTable, specifiedColor=<rgba 0 0 0 0>}
     hovered {
      bIsDynamicallyLoaded=false
      drawAs=RoundedBox
      imageType=NoImage
      mirroring=NoMirror
      resourceName=None
      resourceObject=null
      tiling=NoTile
      imageSize=<32 32>
      margin=<ltrb 0 0 0 0>
      outlineSettings {
       bUseBrushTransparency=false
       roundingType=FixedRadius
       width=1
       color {colorUseRule=UseColor_ColorTable, specifiedColor=<rgba 0 0 0 0>}
       cornerRadii=<4 4 4 4>
      }
      tintColor {colorUseRule=UseColor_Specified, specifiedColor=<rgba 0.1565 0.1681 0.2229 1>}
      uVRegion {bIsValid=false, max=<0 0>, min=<0 0>}
     }
     hoveredForeground {colorUseRule=UseColor_ColorTable, specifiedColor=<rgba 0 0 0 0>}
     hoveredSlateSound {resourceObject=null}
     normal {
      bIsDynamicallyLoaded=false
      drawAs=RoundedBox
      imageType=NoImage
      mirroring=NoMirror
      resourceName=None
      resourceObject=null
      tiling=NoTile
      imageSize=<32 32>
      margin=<ltrb 0 0 0 0>
      outlineSettings {
       bUseBrushTransparency=false
       roundingType=FixedRadius
       width=1
       color {colorUseRule=UseColor_ColorTable, specifiedColor=<rgba 0 0 0 0>}
       cornerRadii=<4 4 4 4>
      }
      tintColor {colorUseRule=UseColor_Specified, specifiedColor=<rgba 0.0741 0.0853 0.1312 1>}
      uVRegion {bIsValid=false, max=<0 0>, min=<0 0>}
     }
     normalForeground {colorUseRule=UseColor_ColorTable, specifiedColor=<rgba 0 0 0 0>}
     normalPadding=<ltrb 12 1.5 12 1.5>
     pressed {
      bIsDynamicallyLoaded=false
      drawAs=RoundedBox
      imageType=NoImage
      mirroring=NoMirror
      resourceName=None
      resourceObject=null
      tiling=NoTile
      imageSize=<32 32>
      margin=<ltrb 0 0 0 0>
      outlineSettings {
       bUseBrushTransparency=false
       roundingType=FixedRadius
       width=1
       color {colorUseRule=UseColor_ColorTable, specifiedColor=<rgba 0 0 0 0>}
       cornerRadii=<4 4 4 4>
      }
      tintColor {colorUseRule=UseColor_Specified, specifiedColor=<rgba 0 0.0802 0.3515 1>}
      uVRegion {bIsValid=false, max=<0 0>, min=<0 0>}
     }
     pressedForeground {colorUseRule=UseColor_ColorTable, specifiedColor=<rgba 0 0 0 0>}
     pressedPadding=<ltrb 12 2.5 12 0.5>
     pressedSlateSound {resourceObject=null}
    }}
   slot {class=WrapBoxSlot}
  }
  TextBlock {
   widgetClassPath=&TextBlock
   parent=ResetToolButton
   widget=TextBlock
   properties {
    text=Reset Tool
    font {
     bForceMonospaced=false
     bMaterialIsStencil=false
     fontMaterial=null
     fontObject=&Burbank
     letterSpacing=0
     monospacedWidth=1
     size=10.5
     skewAmount=0
     typefaceFontName=Burbank-Black
     outlineSettings {
      bApplyOutlineToDropShadows=false
      bMiteredCorners=false
      bSeparateFillAlpha=false
      outlineMaterial=null
      outlineSize=0
      outlineColor=<rgba 0 0 0 1>
     }
    }
    margin=<ltrb 0 1 0 0>
   }
   slot {class=ButtonSlot}
  }
  EUW_Audio_Weapons_SoundName_ListView_1P {
   widgetClassPath=/Game/Sounds/Fort_Audio_Blueprints/Utilities/EditorUtilityWidgets/Metasounds/Weapons/EUW_Audio_Weapons_SoundName_ListView
   bIsVariable=true
   parent=CanvasPanel_167
   widget=EUW_Audio_Weapons_SoundName_ListView_1P
   properties {renderTransform {angle=0, scale=<1 1>, shear=<0 0>, translation=<-80 -42>}}
   slot {
    class=CanvasPanelSlot
    properties {layoutData {alignment=<0 0>, anchors {maximum=<0 0>, minimum=<0 0>}, offsets=<ltrb 620 80 603.7037 472.0741>}}
   }
  }
  EUW_Audio_Weapons_SoundName_ListView_3P {
   widgetClassPath=&EUW_Audio_Weapons_SoundName_ListView
   bIsVariable=true
   parent=CanvasPanel_167
   widget=EUW_Audio_Weapons_SoundName_ListView_3P
   properties {renderTransform {angle=0, scale=<1 1>, shear=<0 0>, translation=<101 -42>}}
   slot {class=CanvasPanelSlot, properties {layoutData {alignment=<0 0>, anchors {maximum=<0 0>, minimum=<0 0>}, offsets=<ltrb 1100 80 100 30>}}}
  }
  TextBlock_0 {
   widgetClassPath=&TextBlock
   bIsVariable=true
   parent=CanvasPanel_167
   widget=TextBlock_0
   properties {
    text=1p Weapon MetaSound is disabled from builder
    visibility=Hidden
    font {
     bForceMonospaced=false
     bMaterialIsStencil=false
     fontMaterial=null
     fontObject=&BurbankSmall
     letterSpacing=0
     monospacedWidth=1
     size=12
     skewAmount=0
     typefaceFontName=Bold
     outlineSettings {
      bApplyOutlineToDropShadows=false
      bMiteredCorners=false
      bSeparateFillAlpha=false
      outlineMaterial=null
      outlineSize=0
      outlineColor=<rgba 0 0 0 1>
     }
    }
   }
   slot {
    class=CanvasPanelSlot
    properties {layoutData {alignment=<0 0>, anchors {maximum=<0 0>, minimum=<0 0>}, offsets=<ltrb 646.3757 276 384.5641 40>}}
   }
  }
  TextBlock_2 {
   widgetClassPath=&TextBlock
   bIsVariable=true
   parent=CanvasPanel_167
   widget=TextBlock_2
   properties {
    text=3p Weapon MetaSound is disabled from builder
    visibility=Hidden
    font {
     bForceMonospaced=false
     bMaterialIsStencil=false
     fontMaterial=null
     fontObject=&BurbankSmall
     letterSpacing=0
     monospacedWidth=1
     size=12
     skewAmount=0
     typefaceFontName=Bold
     outlineSettings {
      bApplyOutlineToDropShadows=false
      bMiteredCorners=false
      bSeparateFillAlpha=false
      outlineMaterial=null
      outlineSize=0
      outlineColor=<rgba 0 0 0 1>
     }
    }
   }
   slot {
    class=CanvasPanelSlot
    properties {layoutData {alignment=<0 0>, anchors {maximum=<0 0>, minimum=<0 0>}, offsets=<ltrb 1330.2047 276 354.1746 40>}}
   }
  }
  EUW_MetaSoundAssetPath_Weapon {
   widgetClassPath=/Game/Sounds/Fort_Audio_Blueprints/Utilities/EditorUtilityWidgets/Metasounds/Weapons/EUW_MetaSoundAssetPath_Weapon
   bIsVariable=true
   parent=CanvasPanel_167
   widget=EUW_MetaSoundAssetPath_Weapon
   slot {class=CanvasPanelSlot, properties {layoutData {alignment=<0 0>, anchors {maximum=<0 0>, minimum=<0 0>}, offsets=<ltrb 4 4 477 131>}}}
  }
 ]
}
