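# Editor Utility Widget that serves as the in-editor User Guide panel for the MetaSound Preset Migrator tool.
# A scrollable static layout explains four steps (select assets, link inputs, optionally save links, migrate preset) and lists owner contact details.
# The header close button fires the OnCloseButtonClicked dispatcher, with no other logic in the graph.
# A collapsed Step0 row exists only as a naming template for duplicated step widgets.
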EditorUtilityWidgetBlueprint /Game/Sounds/Fort_Audio_Blueprints/Utilities/EditorUtilityWidgets/Metasounds/Tools/MetaSoundPresetMigratorData/MetaSoundPresetMigrator_HelpWidget
graphs [
 EventGraph {nodes [
   "On Clicked (EditorUtilityButton_192)"() {is=title K2Node_ComponentBoundEvent}
   > Call On Close Button Clicked!(execute:Exec, self:Meta Sound Preset Migrator Help Widget Object Reference) -> (then:Exec) {is=title K2Node_CallDelegate}
   Event PreConstruct()
   Event Construct()
   Event Tick()
  ]}
 OnCloseButtonClicked {nodes [OnCloseButtonClicked() {is=title K2Node_FunctionEntry}]}
]
variables [OnCloseButtonClicked {default=()}]
widgets {
 info {rootWidgetClass=&VerticalBox, widgetCount=46}
 widgets [
  VerticalBox_19 {widgetClassPath=&VerticalBox, widget=VerticalBox_19}
  Border_0 {
   widgetClassPath=&Border
   parent=VerticalBox_19
   widget=Border_0
   properties {
    background {
     bIsDynamicallyLoaded=false
     drawAs=RoundedBox
     imageType=NoImage
     mirroring=NoMirror
     resourceName=None
     resourceObject=null
     tiling=NoTile
     imageSize=<32 32>
     margin=<ltrb 0 0 0 0>
     outlineSettings {
      bUseBrushTransparency=false
      roundingType=FixedRadius
      width=0
      color {colorUseRule=UseColor_Foreground, specifiedColor=<rgba 0 0 0 0>}
      cornerRadii=<16 16 16 16>
     }
     tintColor {colorUseRule=UseColor_Specified, specifiedColor=<rgba 1 1 1 1>}
     uVRegion {bIsValid=false, max=<0 0>, min=<0 0>}
    }
    brushColor=<rgba 0 0 0 0.75>
    padding=<ltrb 0 0 0 0>
   }
   slot {class=VerticalBoxSlot, properties {padding=<ltrb 4 4 4 4>, size {sizeRule=Fill, value=1}}}
  }
  EditorUtilityScrollBox_798 {
   widgetClassPath=&EditorUtilityScrollBox
   parent=Border_0
   widget=EditorUtilityScrollBox_798
   slot {class=BorderSlot, properties {padding=<ltrb 0 0 0 0>}}
  }
  VerticalBox_21 {
   widgetClassPath=&VerticalBox
   parent=EditorUtilityScrollBox_798
   widget=VerticalBox_21
   slot {class=ScrollBoxSlot, properties {padding=<ltrb 4 2 4 2>, size {sizeRule=Fill, value=1}}}
  }
  Header {widgetClassPath=&HorizontalBox, parent=VerticalBox_21, widget=Header, slot {class=VerticalBoxSlot}}
  VerticalBox_26 {
   widgetClassPath=&VerticalBox
   parent=Header
   widget=VerticalBox_26
   slot {class=HorizontalBoxSlot, properties {padding=<ltrb 16 16 16 16>, size {sizeRule=Fill, value=1}}}
  }
  TextBlock_31 {
   widgetClassPath=&TextBlock
   parent=VerticalBox_26
   widget=TextBlock_31
   properties {
    text=User Guide
    font {
     bForceMonospaced=false
     bMaterialIsStencil=false
     fontMaterial=null
     fontObject=&Roboto
     letterSpacing=0
     monospacedWidth=1
     size=12
     skewAmount=0
     typefaceFontName=Bold
     outlineSettings {
      bApplyOutlineToDropShadows=false
      bMiteredCorners=false
      bSeparateFillAlpha=false
      outlineMaterial=null
      outlineSize=0
      outlineColor=<rgba 0 0 0 1>
     }
    }
   }
   slot {class=VerticalBoxSlot}
  }
  VerticalBox_187 {
   widgetClassPath=&VerticalBox
   parent=Header
   widget=VerticalBox_187
   slot {class=HorizontalBoxSlot, properties {padding=<ltrb 16 16 16 16>}}
  }
  EditorUtilityButton_192 {
   widgetClassPath=&EditorUtilityButton
   parent=VerticalBox_187
   widget=EditorUtilityButton_192
   properties {widgetStyle {
     clickedSlateSound {resourceObject=null}
     disabled {
      bIsDynamicallyLoaded=false
      drawAs=RoundedBox
      imageType=NoImage
      mirroring=NoMirror
      resourceName=None
      resourceObject=null
      tiling=NoTile
      imageSize=<0 0>
      margin=<ltrb 0 0 0 0>
      outlineSettings {
       bUseBrushTransparency=false
       roundingType=FixedRadius
       width=1
       color {colorUseRule=UseColor_ColorTable, specifiedColor=<rgba 0 0 0 0>}
       cornerRadii=<4 4 4 4>
      }
      tintColor {colorUseRule=UseColor_Specified, specifiedColor=<rgba 0.0395 0.0395 0.0395 1>}
      uVRegion {bIsValid=false, max=<0 0>, min=<0 0>}
     }
     disabledForeground {colorUseRule=UseColor_ColorTable, specifiedColor=<rgba 0 0 0 0>}
     hovered {
      bIsDynamicallyLoaded=false
      drawAs=RoundedBox
      imageType=NoImage
      mirroring=NoMirror
      resourceName=None
      resourceObject=null
      tiling=NoTile
      imageSize=<32 32>
      margin=<ltrb 0 0 0 0>
      outlineSettings {
       bUseBrushTransparency=false
       roundingType=FixedRadius
       width=1
       color {colorUseRule=UseColor_ColorTable, specifiedColor=<rgba 0 0 0 0>}
       cornerRadii=<4 4 4 4>
      }
      tintColor {colorUseRule=UseColor_Specified, specifiedColor=<rgba 0.0953 0.0953 0.0953 1>}
      uVRegion {bIsValid=false, max=<0 0>, min=<0 0>}
     }
     hoveredForeground {colorUseRule=UseColor_ColorTable, specifiedColor=<rgba 0 0 0 0>}
     hoveredSlateSound {resourceObject=null}
     normal {
      bIsDynamicallyLoaded=false
      drawAs=RoundedBox
      imageType=NoImage
      mirroring=NoMirror
      resourceName=None
      resourceObject=null
      tiling=NoTile
      imageSize=<32 32>
      margin=<ltrb 0 0 0 0>
      outlineSettings {
       bUseBrushTransparency=false
       roundingType=FixedRadius
       width=1
       color {colorUseRule=UseColor_ColorTable, specifiedColor=<rgba 0 0 0 0>}
       cornerRadii=<4 4 4 4>
      }
      tintColor {colorUseRule=UseColor_Specified, specifiedColor=<rgba 0.0395 0.0395 0.0395 1>}
      uVRegion {bIsValid=false, max=<0 0>, min=<0 0>}
     }
     normalForeground {colorUseRule=UseColor_ColorTable, specifiedColor=<rgba 0 0 0 0>}
     normalPadding=<ltrb 2 2 2 2>
     pressed {
      bIsDynamicallyLoaded=false
      drawAs=RoundedBox
      imageType=NoImage
      mirroring=NoMirror
      resourceName=None
      resourceObject=null
      tiling=NoTile
      imageSize=<32 32>
      margin=<ltrb 0 0 0 0>
      outlineSettings {
       bUseBrushTransparency=false
       roundingType=FixedRadius
       width=1
       color {colorUseRule=UseColor_ColorTable, specifiedColor=<rgba 0 0 0 0>}
       cornerRadii=<4 4 4 4>
      }
      tintColor {colorUseRule=UseColor_Specified, specifiedColor=<rgba 0.0284 0.0284 0.0284 1>}
      uVRegion {bIsValid=false, max=<0 0>, min=<0 0>}
     }
     pressedForeground {colorUseRule=UseColor_ColorTable, specifiedColor=<rgba 0 0 0 0>}
     pressedPadding=<ltrb 0 0 0 0>
     pressedSlateSound {resourceObject=null}
    }}
   slot {class=VerticalBoxSlot, properties {verticalAlignment=VAlign_Center, size {sizeRule=Fill, value=1}}}
  }
  Image_158 {
   widgetClassPath=&Image
   bIsVariable=true
   parent=EditorUtilityButton_192
   widget=Image_158
   properties {brush {
     bIsDynamicallyLoaded=false
     drawAs=Image
     imageType=FullColor
     mirroring=NoMirror
     resourceName=None
     resourceObject=/Game/UI/Icons/Icon_Seach_Close
     tiling=NoTile
     imageSize=<16 16>
     margin=<ltrb 0 0 0 0>
     outlineSettings {
      bUseBrushTransparency=false
      roundingType=HalfHeightRadius
      width=0
      color {colorUseRule=UseColor_Specified, specifiedColor=<rgba 0 0 0 0>}
      cornerRadii=<0 0 0 0>
     }
     tintColor {colorUseRule=UseColor_Specified, specifiedColor=<rgba 1 1 1 1>}
     uVRegion {bIsValid=false, max=<0 0>, min=<0 0>}
    }}
   slot {class=ButtonSlot, properties {padding=<ltrb 0 0 0 0>}}
  }
  Body {
   widgetClassPath=&VerticalBox
   parent=VerticalBox_21
   widget=Body
   slot {class=VerticalBoxSlot, properties {padding=<ltrb 16 4 4 4>}}
  }
  Step0 {
   widgetClassPath=&HorizontalBox
   parent=Body
   widget=Step0
   properties {bIsEnabled=false, visibility=Collapsed}
   slot {class=VerticalBoxSlot}
  }
  VerticalBox_2 {widgetClassPath=&VerticalBox, parent=Step0, widget=VerticalBox_2, slot {class=HorizontalBoxSlot}}
  StepNumber {
   widgetClassPath=&TextBlock
   parent=VerticalBox_2
   widget=StepNumber
   properties {
    text=0.
    font {
     bForceMonospaced=false
     bMaterialIsStencil=false
     fontMaterial=null
     fontObject=&Roboto
     letterSpacing=0
     monospacedWidth=1
     size=10.5
     skewAmount=0
     typefaceFontName=Bold
     outlineSettings {
      bApplyOutlineToDropShadows=false
      bMiteredCorners=false
      bSeparateFillAlpha=false
      outlineMaterial=null
      outlineSize=0
      outlineColor=<rgba 0 0 0 1>
     }
    }
   }
   slot {class=VerticalBoxSlot}
  }
  VerticalBox_3 {
   widgetClassPath=&VerticalBox
   parent=Step0
   widget=VerticalBox_3
   slot {class=HorizontalBoxSlot, properties {padding=<ltrb 4 0 0 0>, size {sizeRule=Fill, value=1}}}
  }
  StepHeader {
   widgetClassPath=&TextBlock
   parent=VerticalBox_3
   widget=StepHeader
   properties {
    autoWrapText=true
    text=This one is hidden because it makes it easier to label my widgets on the left hierarchy.
    font {
     bForceMonospaced=false
     bMaterialIsStencil=false
     fontMaterial=null
     fontObject=&Roboto
     letterSpacing=0
     monospacedWidth=1
     size=10.5
     skewAmount=0
     typefaceFontName=Bold
     outlineSettings {
      bApplyOutlineToDropShadows=false
      bMiteredCorners=false
      bSeparateFillAlpha=false
      outlineMaterial=null
      outlineSize=0
      outlineColor=<rgba 0 0 0 1>
     }
    }
   }
   slot {class=VerticalBoxSlot}
  }
  StepBody {
   widgetClassPath=&TextBlock
   parent=VerticalBox_3
   widget=StepBody
   properties {
    autoWrapText=true
    text="Really that's the only reason this is here, so that I don't have to enter new names for my widgets in the hierarchy and it automatically increments the names with each step that I add here."
    font {
     bForceMonospaced=false
     bMaterialIsStencil=false
     fontMaterial=null
     fontObject=&Roboto
     letterSpacing=0
     monospacedWidth=1
     size=9
     skewAmount=0
     typefaceFontName=Regular
     outlineSettings {
      bApplyOutlineToDropShadows=false
      bMiteredCorners=false
      bSeparateFillAlpha=false
      outlineMaterial=null
      outlineSize=0
      outlineColor=<rgba 0 0 0 1>
     }
    }
   }
   slot {class=VerticalBoxSlot, properties {padding=<ltrb 16 8 0 0>}}
  }
  Step1 {widgetClassPath=&HorizontalBox, parent=Body, widget=Step1, slot {class=VerticalBoxSlot, properties {padding=<ltrb 0 0 0 16>}}}
  VerticalBox_23 {widgetClassPath=&VerticalBox, parent=Step1, widget=VerticalBox_23, slot {class=HorizontalBoxSlot}}
  StepNumber_1 {
   widgetClassPath=&TextBlock
   parent=VerticalBox_23
   widget=StepNumber_1
   properties {
    text=1.
    font {
     bForceMonospaced=false
     bMaterialIsStencil=false
     fontMaterial=null
     fontObject=&Roboto
     letterSpacing=0
     monospacedWidth=1
     size=10.5
     skewAmount=0
     typefaceFontName=Bold
     outlineSettings {
      bApplyOutlineToDropShadows=false
      bMiteredCorners=false
      bSeparateFillAlpha=false
      outlineMaterial=null
      outlineSize=0
      outlineColor=<rgba 0 0 0 1>
     }
    }
   }
   slot {class=VerticalBoxSlot}
  }
  VerticalBox_24 {
   widgetClassPath=&VerticalBox
   parent=Step1
   widget=VerticalBox_24
   slot {class=HorizontalBoxSlot, properties {padding=<ltrb 4 0 0 0>, size {sizeRule=Fill, value=1}}}
  }
  StepHeader_1 {
   widgetClassPath=&TextBlock
   parent=VerticalBox_24
   widget=StepHeader_1
   properties {
    autoWrapText=true
    text=Select MetaSound Assets
    font {
     bForceMonospaced=false
     bMaterialIsStencil=false
     fontMaterial=null
     fontObject=&Roboto
     letterSpacing=0
     monospacedWidth=1
     size=10.5
     skewAmount=0
     typefaceFontName=Bold
     outlineSettings {
      bApplyOutlineToDropShadows=false
      bMiteredCorners=false
      bSeparateFillAlpha=false
      outlineMaterial=null
      outlineSize=0
      outlineColor=<rgba 0 0 0 1>
     }
    }
   }
   slot {class=VerticalBoxSlot, properties {padding=<ltrb 0 0 16 0>}}
  }
  StepBody_1 {
   widgetClassPath=&TextBlock
   parent=VerticalBox_24
   widget=StepBody_1
   properties {
    autoWrapText=true
    lineHeightPercentage=1.25
    text="Start by selecting a MetaSound Source Preset asset that you wish to migrate and a MetaSound Source asset that you want to migrate the preset to. \r\n\r\nCurrently, this tool does not allow presets to become a preset of another preset. That is a rare case that may be supported in the future."
    colorAndOpacity {colorUseRule=UseColor_Specified, specifiedColor=<rgba 0.5029 0.5029 0.5029 1>}
    font {
     bForceMonospaced=false
     bMaterialIsStencil=false
     fontMaterial=null
     fontObject=&Roboto
     letterSpacing=0
     monospacedWidth=1
     size=9
     skewAmount=0
     typefaceFontName=Regular
     outlineSettings {
      bApplyOutlineToDropShadows=false
      bMiteredCorners=false
      bSeparateFillAlpha=false
      outlineMaterial=null
      outlineSize=0
      outlineColor=<rgba 0 0 0 1>
     }
    }
   }
   slot {class=VerticalBoxSlot, properties {padding=<ltrb 0 8 16 0>}}
  }
  Step2 {widgetClassPath=&HorizontalBox, parent=Body, widget=Step2, slot {class=VerticalBoxSlot, properties {padding=<ltrb 0 0 0 16>}}}
  VerticalBox {widgetClassPath=&VerticalBox, parent=Step2, widget=VerticalBox, slot {class=HorizontalBoxSlot}}
  StepNumber_2 {
   widgetClassPath=&TextBlock
   parent=VerticalBox
   widget=StepNumber_2
   properties {
    text=2.
    font {
     bForceMonospaced=false
     bMaterialIsStencil=false
     fontMaterial=null
     fontObject=&Roboto
     letterSpacing=0
     monospacedWidth=1
     size=10.5
     skewAmount=0
     typefaceFontName=Bold
     outlineSettings {
      bApplyOutlineToDropShadows=false
      bMiteredCorners=false
      bSeparateFillAlpha=false
      outlineMaterial=null
      outlineSize=0
      outlineColor=<rgba 0 0 0 1>
     }
    }
   }
   slot {class=VerticalBoxSlot}
  }
  VerticalBox_1 {
   widgetClassPath=&VerticalBox
   parent=Step2
   widget=VerticalBox_1
   slot {class=HorizontalBoxSlot, properties {padding=<ltrb 4 0 0 0>, size {sizeRule=Fill, value=1}}}
  }
  StepHeader_2 {
   widgetClassPath=&TextBlock
   parent=VerticalBox_1
   widget=StepHeader_2
   properties {
    autoWrapText=true
    text=Link MetaSound Inputs
    font {
     bForceMonospaced=false
     bMaterialIsStencil=false
     fontMaterial=null
     fontObject=&Roboto
     letterSpacing=0
     monospacedWidth=1
     size=10.5
     skewAmount=0
     typefaceFontName=Bold
     outlineSettings {
      bApplyOutlineToDropShadows=false
      bMiteredCorners=false
      bSeparateFillAlpha=false
      outlineMaterial=null
      outlineSize=0
      outlineColor=<rgba 0 0 0 1>
     }
    }
   }
   slot {class=VerticalBoxSlot}
  }
  StepBody_2 {
   widgetClassPath=&TextBlock
   parent=VerticalBox_1
   widget=StepBody_2
   properties {
    autoWrapText=true
    lineHeightPercentage=1.25
    text="Once you've selected both assets, you will see a list of MetaSound inputs for each. Since the inputs are probably not the same for both assets, linking is required in order to migrate the input values to the destination asset.\r\n\r\nTo link an input, drag and drop an input from one list to another input on the other asset. Inputs can only be linked with another input of the same data type. An input cannot be linked to more than one other input.\r\n\r\nTo unlink inputs, Shift+Click on one of the linked inputs. \r\n\r\nUpon selecting both assets, the tool will automatically link inputs that have matching names and data types, unless the links were recalled from a save."
    colorAndOpacity {colorUseRule=UseColor_Specified, specifiedColor=<rgba 0.5029 0.5029 0.5029 1>}
    font {
     bForceMonospaced=false
     bMaterialIsStencil=false
     fontMaterial=null
     fontObject=&Roboto
     letterSpacing=0
     monospacedWidth=1
     size=9
     skewAmount=0
     typefaceFontName=Regular
     outlineSettings {
      bApplyOutlineToDropShadows=false
      bMiteredCorners=false
      bSeparateFillAlpha=false
      outlineMaterial=null
      outlineSize=0
      outlineColor=<rgba 0 0 0 1>
     }
    }
   }
   slot {class=VerticalBoxSlot, properties {padding=<ltrb 0 8 16 0>}}
  }
  Step3 {widgetClassPath=&HorizontalBox, parent=Body, widget=Step3, slot {class=VerticalBoxSlot, properties {padding=<ltrb 0 0 0 16>}}}
  VerticalBox_4 {widgetClassPath=&VerticalBox, parent=Step3, widget=VerticalBox_4, slot {class=HorizontalBoxSlot}}
  StepNumber_3 {
   widgetClassPath=&TextBlock
   parent=VerticalBox_4
   widget=StepNumber_3
   properties {
    text=3.
    font {
     bForceMonospaced=false
     bMaterialIsStencil=false
     fontMaterial=null
     fontObject=&Roboto
     letterSpacing=0
     monospacedWidth=1
     size=10.5
     skewAmount=0
     typefaceFontName=Bold
     outlineSettings {
      bApplyOutlineToDropShadows=false
      bMiteredCorners=false
      bSeparateFillAlpha=false
      outlineMaterial=null
      outlineSize=0
      outlineColor=<rgba 0 0 0 1>
     }
    }
   }
   slot {class=VerticalBoxSlot}
  }
  VerticalBox_5 {
   widgetClassPath=&VerticalBox
   parent=Step3
   widget=VerticalBox_5
   slot {class=HorizontalBoxSlot, properties {padding=<ltrb 4 0 0 0>, size {sizeRule=Fill, value=1}}}
  }
  StepHeader_3 {
   widgetClassPath=&TextBlock
   parent=VerticalBox_5
   widget=StepHeader_3
   properties {
    autoWrapText=true
    text="Save Linked Inputs (optional)"
    font {
     bForceMonospaced=false
     bMaterialIsStencil=false
     fontMaterial=null
     fontObject=&Roboto
     letterSpacing=0
     monospacedWidth=1
     size=10.5
     skewAmount=0
     typefaceFontName=Bold
     outlineSettings {
      bApplyOutlineToDropShadows=false
      bMiteredCorners=false
      bSeparateFillAlpha=false
      outlineMaterial=null
      outlineSize=0
      outlineColor=<rgba 0 0 0 1>
     }
    }
   }
   slot {class=VerticalBoxSlot}
  }
  StepBody_3 {
   widgetClassPath=&TextBlock
   parent=VerticalBox_5
   widget=StepBody_3
   properties {
    autoWrapText=true
    lineHeightPercentage=1.25
    text="To speed up the process of migrating large numbers of presets, linked inputs can be saved. Saved links are automatically recalled when selecting a pair of MetaSound Source assets that correspond to a previously-saved set of linked inputs.\r\n\r\nSaved links are stored in a data asset that exists separately from this tool. The data asset will be checked out of version control upon saving. Be sure to submit any changes in order for them to persist and be shared with others."
    colorAndOpacity {colorUseRule=UseColor_Specified, specifiedColor=<rgba 0.5029 0.5029 0.5029 1>}
    font {
     bForceMonospaced=false
     bMaterialIsStencil=false
     fontMaterial=null
     fontObject=&Roboto
     letterSpacing=0
     monospacedWidth=1
     size=9
     skewAmount=0
     typefaceFontName=Regular
     outlineSettings {
      bApplyOutlineToDropShadows=false
      bMiteredCorners=false
      bSeparateFillAlpha=false
      outlineMaterial=null
      outlineSize=0
      outlineColor=<rgba 0 0 0 1>
     }
    }
   }
   slot {class=VerticalBoxSlot, properties {padding=<ltrb 0 8 16 0>}}
  }
  Step4 {widgetClassPath=&HorizontalBox, parent=Body, widget=Step4, slot {class=VerticalBoxSlot, properties {padding=<ltrb 0 0 0 16>}}}
  VerticalBox_6 {widgetClassPath=&VerticalBox, parent=Step4, widget=VerticalBox_6, slot {class=HorizontalBoxSlot}}
  StepNumber_4 {
   widgetClassPath=&TextBlock
   parent=VerticalBox_6
   widget=StepNumber_4
   properties {
    text=4.
    font {
     bForceMonospaced=false
     bMaterialIsStencil=false
     fontMaterial=null
     fontObject=&Roboto
     letterSpacing=0
     monospacedWidth=1
     size=10.5
     skewAmount=0
     typefaceFontName=Bold
     outlineSettings {
      bApplyOutlineToDropShadows=false
      bMiteredCorners=false
      bSeparateFillAlpha=false
      outlineMaterial=null
      outlineSize=0
      outlineColor=<rgba 0 0 0 1>
     }
    }
   }
   slot {class=VerticalBoxSlot}
  }
  VerticalBox_7 {
   widgetClassPath=&VerticalBox
   parent=Step4
   widget=VerticalBox_7
   slot {class=HorizontalBoxSlot, properties {padding=<ltrb 4 0 0 0>, size {sizeRule=Fill, value=1}}}
  }
  StepHeader_4 {
   widgetClassPath=&TextBlock
   parent=VerticalBox_7
   widget=StepHeader_4
   properties {
    autoWrapText=true
    text=Migrate Preset
    font {
     bForceMonospaced=false
     bMaterialIsStencil=false
     fontMaterial=null
     fontObject=&Roboto
     letterSpacing=0
     monospacedWidth=1
     size=10.5
     skewAmount=0
     typefaceFontName=Bold
     outlineSettings {
      bApplyOutlineToDropShadows=false
      bMiteredCorners=false
      bSeparateFillAlpha=false
      outlineMaterial=null
      outlineSize=0
      outlineColor=<rgba 0 0 0 1>
     }
    }
   }
   slot {class=VerticalBoxSlot}
  }
  StepBody_4 {
   widgetClassPath=&TextBlock
   parent=VerticalBox_7
   widget=StepBody_4
   properties {
    autoWrapText=true
    lineHeightPercentage=1.25
    text="Once finished, double-check the name of the new asset being created and click 'Migrate Preset'. The new asset will be created in the same folder as the original MetaSound Source preset asset."
    colorAndOpacity {colorUseRule=UseColor_Specified, specifiedColor=<rgba 0.5029 0.5029 0.5029 1>}
    font {
     bForceMonospaced=false
     bMaterialIsStencil=false
     fontMaterial=null
     fontObject=&Roboto
     letterSpacing=0
     monospacedWidth=1
     size=9
     skewAmount=0
     typefaceFontName=Regular
     outlineSettings {
      bApplyOutlineToDropShadows=false
      bMiteredCorners=false
      bSeparateFillAlpha=false
      outlineMaterial=null
      outlineSize=0
      outlineColor=<rgba 0 0 0 1>
     }
    }
   }
   slot {class=VerticalBoxSlot, properties {padding=<ltrb 0 8 16 0>}}
  }
  Footer {widgetClassPath=&VerticalBox, parent=VerticalBox_21, widget=Footer, slot {class=VerticalBoxSlot}}
  HorizontalBox_50 {widgetClassPath=&HorizontalBox, parent=Footer, widget=HorizontalBox_50, slot {class=VerticalBoxSlot}}
  Contact {
   widgetClassPath=&TextBlock
   parent=HorizontalBox_50
   widget=Contact
   properties {
    text=Contact
    font {
     bForceMonospaced=false
     bMaterialIsStencil=false
     fontMaterial=null
     fontObject=&Roboto
     letterSpacing=0
     monospacedWidth=1
     size=12
     skewAmount=0
     typefaceFontName=Bold
     outlineSettings {
      bApplyOutlineToDropShadows=false
      bMiteredCorners=false
      bSeparateFillAlpha=false
      outlineMaterial=null
      outlineSize=0
      outlineColor=<rgba 0 0 0 1>
     }
    }
   }
   slot {class=HorizontalBoxSlot, properties {padding=<ltrb 16 16 16 0>}}
  }
  HorizontalBox_51 {
   widgetClassPath=&HorizontalBox
   parent=Footer
   widget=HorizontalBox_51
   slot {class=VerticalBoxSlot, properties {padding=<ltrb 16 0 16 16>, size {sizeRule=Fill, value=1}}}
  }
  StepBody_5 {
   widgetClassPath=&TextBlock
   parent=HorizontalBox_51
   widget=StepBody_5
   properties {
    autoWrapText=true
    lineHeightPercentage=1.25
    text="Owner: Eric Buchholz\r\n##fn-audio-tech"
    colorAndOpacity {colorUseRule=UseColor_Specified, specifiedColor=<rgba 0.5029 0.5029 0.5029 1>}
    font {
     bForceMonospaced=false
     bMaterialIsStencil=false
     fontMaterial=null
     fontObject=&Roboto
     letterSpacing=0
     monospacedWidth=1
     size=9
     skewAmount=0
     typefaceFontName=Regular
     outlineSettings {
      bApplyOutlineToDropShadows=false
      bMiteredCorners=false
      bSeparateFillAlpha=false
      outlineMaterial=null
      outlineSize=0
      outlineColor=<rgba 0 0 0 1>
     }
    }
   }
   slot {class=HorizontalBoxSlot, properties {padding=<ltrb 16 8 16 0>, size {sizeRule=Fill, value=1}}}
  }
 ]
}
