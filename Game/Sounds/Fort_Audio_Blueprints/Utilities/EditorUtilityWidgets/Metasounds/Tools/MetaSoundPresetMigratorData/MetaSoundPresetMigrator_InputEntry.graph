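# Editor utility list entry for the MetaSound Preset Migrator tool, showing one preset input with name, type color and link status.
# It reads a MetaSoundPresetMigrator_InputData object received as the list item, which also holds shared drag, hover and link state.
# Drag-over highlights entries of matching data type on the opposite origin/destination side.
# Drop clears existing links, links both inputs mutually and fires dispatchers; getter functions derive colors, tooltips and warning icon from that state.
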EditorUtilityWidgetBlueprint /Game/Sounds/Fort_Audio_Blueprints/Utilities/EditorUtilityWidgets/Metasounds/Tools/MetaSoundPresetMigratorData/MetaSoundPresetMigrator_InputEntry
graphs [
 EventGraph {nodes [
   Event OnDragCancelled!() -> (OutputDelegate, then, PointerEvent:"Pointer Event Structure (by ref)", Operation:Drag Drop Operation Object Reference) {is=title K2Node_Event}
   > Call On Drag Ended!(execute:Exec, self:Meta Sound Preset Migrator Input Entry Object Reference) -> (then:Exec) {is=title K2Node_CallDelegate}
   > Set Is Being Dragged!(
    execute
    Is Being Dragged:bool=false
    self:Meta Sound Preset Migrator Input Data Object Reference=Get Input Data Object() -> (Input Data Object:Meta Sound Preset Migrator Input Data Object Reference)
   ) -> (then, Output_Get:bool)
   n1 = Event OnListItemObjectSet()
   > n2 = Cast To MetaSoundPresetMigrator_InputData!(execute, Object=@n1.ListItemObject) -> (then, CastFailed, AsMeta Sound Preset Migrator Input Data:Meta Sound Preset Migrator Input Data Object Reference) {is=title K2Node_DynamicCast}
   > Set Input Data Object(Input Data Object:Meta Sound Preset Migrator Input Data Object Reference=@n2) -> (Output_Get:Meta Sound Preset Migrator Input Data Object Reference)
   Event OnDragLeave!() -> (OutputDelegate, then, PointerEvent:Pointer Event Structure, Operation:Drag Drop Operation Object Reference) {is=title K2Node_Event}
   > Set Is Dragged Over!(execute, Is Dragged Over:bool=false, self:Meta Sound Preset Migrator Input Data Object Reference=Get Input Data Object()) -> (then, Output_Get:bool)
  ]}
 Get Background Color {nodes [
   Get Background Color() {is=title K2Node_FunctionEntry}
   > n1 = Sequence()
   > n2 = Get!(execute:Exec, self:Meta Sound Preset Migrator Input Data Object Reference=Get Input Data Object()) -> (then:Exec, else:Exec, Linked Input:Meta Sound Preset Migrator Input Data Object Reference)
   n3 = Get Input Data Object()
   > Set Linked Input Is Hovered(Linked Input Is Hovered:bool=Get Is Hovered!(self:Meta Sound Preset Migrator Input Data Object Reference=Get Linked Input!(self:Meta Sound Preset Migrator Input Data Object Reference=@n3) -> (Linked Input:Meta Sound Preset Migrator Input Data Object Reference)) -> (Is Hovered:bool)) -> (Output_Get:bool)
   n4 = Get Input Data Object()
   > Set Linked Input Is Being Dragged(Linked Input Is Being Dragged:bool=Get Is Being Dragged!(self:Meta Sound Preset Migrator Input Data Object Reference=Get Linked Input(self=@n4)) -> (Is Being Dragged:bool)) -> (Output_Get:bool)
   n5 = Get Input Data Object()
   > Set Linked Input Is Selected(Linked Input Is Selected:bool=Get Is Selected!(self:Meta Sound Preset Migrator Input Data Object Reference=Get Linked Input(self=@n5)) -> (Is Selected:bool)) -> (Output_Get:bool)
   Set Linked Input Is Hovered(execute=@n2.else, Linked Input Is Hovered=false)
   > Set Linked Input Is Being Dragged(Linked Input Is Being Dragged=false)
   > Set Linked Input Is Selected(Linked Input Is Selected=false)
   Set Is Hovered!(execute=@n1.then_1, Is Hovered:bool=IsHovered(), self:Meta Sound Preset Migrator Input Data Object Reference=Get Input Data Object()) -> (then, Output_Get:bool)
   n6 = Get Linked Input Is Being Dragged() -> (Linked Input Is Being Dragged:bool)
   n7 = Get Is Being Dragged(self=Get Input Data Object())
   n8 = Get Input Data Object()
   n9 = OR Boolean!(
    A=Get Is Dragged Over!(self:Meta Sound Preset Migrator Input Data Object Reference=@n8) -> (Is Dragged Over:bool)
    B=Get Is Selected(self=@n8)
    C:bool=Get Linked Input Is Selected() -> (Linked Input Is Selected:bool)
   ) -> (ReturnValue)
   n10 = "% (Integer)"(A=Get Row Index!(self:Meta Sound Preset Migrator Input Data Object Reference=Get Input Data Object()) -> (Row Index:int=0), B=2) -> (ReturnValue=0)
   n11 = Select(
    Option 0:Linear Color Structure=Select(Option 0=(R=0.0103,G=0.0103,B=0.0103,A=1), Option 1=(R=0.0075,G=0.0075,B=0.0075,A=1), Index="Equal (Integer)"(B=@n10))
    Option 1:Linear Color Structure=(R=0.0176,G=0.0176,B=0.0176,A=1)
    Index=OR Boolean!(A=IsHovered(), B=Get Linked Input Is Hovered() -> (Linked Input Is Hovered:bool)) -> (ReturnValue)
   ) -> (ReturnValue:Linear Color Structure)
   > Return Node!(
    execute
    ReturnValue:Linear Color Structure=Select(
     Option 0=Select(Option 0=@n11, Option 1=(R=0,G=0.162,B=0.7454,A=1), Index=@n9)
     Option 1=(R=0.0513,G=0.0953,B=0.159,A=1)
     Index=OR Boolean(A=@n7, B=@n6)
    )
   )
  ]}
 Get Input Name {nodes [
   Get Input Name() {is=title K2Node_FunctionEntry}
   n1 = Get Input Data Object()
   > Return Node(ReturnValue:Text="To Text (Name)"!(InName:Name=Get Input Name!(self:Meta Sound Preset Migrator Input Data Object Reference=@n1) -> (Input Name:Name=None) {is=title K2Node_VariableGet}) -> (ReturnValue:Text))
  ]}
 Get Is Linked Input Visible {nodes [
   Get Is Linked Input Visible() {is=title K2Node_FunctionEntry}
   > n1 = Get(self=Get Input Data Object())
   n2 = Get Input Data Object()
   > n3 = Set Linked Input Is Destination(Linked Input Is Destination:bool=NOT Boolean(A=Get Is Origin!(self:Meta Sound Preset Migrator Input Data Object Reference=@n2) -> (Is Origin:bool))) -> (Output_Get:bool)
   n4 = Get!(execute=[@n7.then, @n3.then]) -> (then, else, Parent Widget:Meta Sound Preset Migrator Object Reference)
   n5 = AND Boolean(
    A=Are Both Assets Valid!(self:Meta Sound Preset Migrator Object Reference=@n4) -> (ReturnValue:bool)
    B=Get Is Origin(self=Get Input Data Object())
   )
   n6 = Get Linked Input Is Destination() -> (Linked Input Is Destination:bool)
   > Return Node(ReturnValue:ESlateVisibility Enum=Select(Option 0:ESlateVisibility Enum=Hidden, Option 1:ESlateVisibility Enum=Visible, Index=OR Boolean(A=@n5, B=@n6)) -> (ReturnValue:ESlateVisibility Enum))
   Return Node(execute=@n4.else, ReturnValue=Hidden)
   n7 = Set Linked Input Is Destination(execute=@n1.else, Linked Input Is Destination=false)
  ]}
 Get Is Selected {nodes [
   Get Is Selected() {is=title K2Node_FunctionEntry}
   > Return Node!(execute, Is Selected:bool=false)
  ]}
 Get Linked Status Color {nodes [
   Get Linked Status Color() {is=title K2Node_FunctionEntry}
   n1 = IsValid(Object=Get Linked Input(self=Get Input Data Object()))
   > Return Node!(
    execute
    ReturnValue:Slate Color Structure=Make Slate Color(SpecifiedColor=Select(Option 0:Linear Color Structure=(R=0.7305,G=0.5089,B=0.0578,A=1), Option 1:Linear Color Structure=(R=0,G=0.2961,B=0.0823,A=1), Index=@n1) -> (ReturnValue:Linear Color Structure))
   )
  ]}
 Get Linked Status Icon {nodes [
   Get Linked Status Icon() {is=title K2Node_FunctionEntry}
   n1 = IsValid(Object=Get Linked Input(self=Get Input Data Object()))
   > Return Node(ReturnValue:Slate Brush Structure=Make Slate Brush!(
     TintColor:Slate Color Structure=(SpecifiedColor=(R=1,G=1,B=1,A=1),ColorUseRule=UseColor_Specified)
     DrawAs:ESlateBrushDrawType Enum=Image
     Tiling:ESlateBrushTileType Enum=NoTile
     Mirroring:ESlateBrushMirrorType Enum=NoMirror
     ImageSize:Vector 2D Structure=MakeVector2D(X=16, Y=16)
     Margin:Margin Structure=(Left=0,Top=0,Right=0,Bottom=0)
     ResourceObject:Object Reference=Select(Option 0:Object Reference=/Game/UI/Foundation/Shell/Textures/T-Icon-Warning-32, Option 1:Object Reference, Index=@n1) -> (ReturnValue:Object Reference)
     OutlineSettings:Slate Brush Outline Settings Structure=(CornerRadii=(X=0,Y=0,Z=0,W=0),Color=(SpecifiedColor=(R=0,G=0,B=0,A=0),ColorUseRule=UseColor_Specified),Width=0,RoundingType=HalfHeightRadius,bUseBrushTransparency=False)
    ) -> (SlateBrush:Slate Brush Structure) {is=title K2Node_MakeStruct})
  ]}
 Get Linked ToolTip {nodes [
   Get Linked ToolTip() {is=title K2Node_FunctionEntry}
   > n1 = Get!(execute, self:Meta Sound Preset Migrator Input Data Object Reference=Get Input Data Object()) -> (then, else, Linked Input:Meta Sound Preset Migrator Input Data Object Reference)
   > Return Node(ReturnValue:Text=Format Text!(Format="LOCTEXT(\"Linked with {link}\")", link:Name=Get Input Name(self=@n1) -> (Input Name=None)) -> (Result))
   Return Node(execute=@n1.else, ReturnValue=Format Text!(Format="LOCTEXT(\"Input is not linked and will be discarded.\")") -> (Result))
  ]}
 Get Row Index {nodes [
   Get Row Index() {is=title K2Node_FunctionEntry}
   > Return Node!(execute, Row Index:int=Get Row Index(self=Get Input Data Object()) -> (Row Index=0) {is=title K2Node_VariableGet})
  ]}
 Get Shift Enabled {nodes [
   Get Shift Enabled() {is=title K2Node_FunctionEntry}
   n1 = Get Input Data Object()
   n2 = AND Boolean(
    A=Is Shift Down!(KeysState:"Modifier Keys State Structure (by ref)"=GetModifierKeysState!() -> (ReturnValue:Modifier Keys State Structure)) -> (ReturnValue:bool)
    B=IsValid(Object=Get Linked Input(self=@n1))
   )
   > SetCursor!(
    execute
    self:Widget Object Reference
    InCursor:EMouseCursor Enum=Select(Option 0:EMouseCursor Enum=Default, Option 1:EMouseCursor Enum=Crosshairs, Index=@n2) -> (ReturnValue:EMouseCursor Enum)
   ) -> (then)
   n3 = Is Shift Down(KeysState=GetModifierKeysState())
   > Return Node!(execute, ReturnValue:bool=OR Boolean(A=@n2, B=NOT Boolean(A=@n3)))
  ]}
 Get Text Color {nodes [
   Get Text Color() {is=title K2Node_FunctionEntry}
   > n1 = Get(self=Get Input Data Object())
   > n2 = Set Linked Input Is Selected(Linked Input Is Selected=Get Is Selected(self=@n1) {is=title K2Node_VariableGet})
   n3 = Self-Reference()
   n4 = Get Input Data Object()
   n5 = Get Is Dragged Over(self=@n4)
   n6 = Get Is Selected(self=@n4)
   n7 = Get Linked Input Is Selected()
   Return Node(
    execute=[@n2.then, @n8.then]
    ReturnValue:Slate Color Structure=Select(
     Option 0:Slate Color Structure=Get List Row Text Color!(List Entry Widget:Editor Utility Widget Object Reference=@n3, __WorldContext:Object Reference) -> (ReturnValue:Slate Color Structure) {is=title K2Node_CallFunction}
     Option 1:Slate Color Structure=Make Slate Color(SpecifiedColor=(R=1,G=1,B=1,A=1))
     Index=OR Boolean!(A=@n5, B=@n6, C:bool=@n7) -> (ReturnValue)
    ) -> (ReturnValue:Slate Color Structure)
   )
   n8 = Set Linked Input Is Selected(execute=@n1.else, Linked Input Is Selected=false)
  ]}
 Get Type Color {nodes [
   Get Type Color() {is=title K2Node_FunctionEntry}
   > n1 = Get MetaSound Data Type Color!(
    execute
    self:Fort Audio Tools Style Object Reference=Get Style() -> (Style:Meta Sound Preset Migrator Style Object Reference)
    Data Type:Name=Get Data Type!(self:Meta Sound Preset Migrator Input Data Object Reference=Get Input Data Object()) -> (Data Type:Name=None)
   ) -> (then, Color:Linear Color Structure) {is=title K2Node_CallFunction}
   > Return Node(ReturnValue:Linear Color Structure=@n1)
  ]}
 Get Unlink ToolTip {nodes [
   Get Unlink ToolTip() {is=title K2Node_FunctionEntry}
   > n1 = Get(self=Get Input Data Object())
   > Branch(Condition=Is Shift Down(KeysState=GetModifierKeysState()))
   > Return Node(ReturnValue:Text=Format Text!(Format="LOCTEXT(\"Click to unlink from {linkedinput}.\")", linkedinput:Name=Get Input Name(self=@n1) -> (Input Name=None)) -> (Result))
   Return Node(execute=@n1.else)
  ]}
 Get_TypeImage_ToolTipText {nodes [
   Get_TypeImage_ToolTipText() {is=title K2Node_FunctionEntry}
   n1 = Get Input Data Object()
   > Return Node(ReturnValue="To Text (Name)"(InName=Get Data Type(self=@n1) -> (Data Type=None)))
  ]}
 OnDragDetected {nodes [
   OnDragDetected!() -> (then, MyGeometry:Geometry Structure, PointerEvent:"Pointer Event Structure (by ref)") {is=title K2Node_FunctionEntry}
   > n1 = "Create Drag && Drop Operation"!(
    execute:Exec
    WorldContextObject:Object Reference
    Class:Drag Drop Operation Class Reference=/Script/UMG.DragDropOperation
    Tag:String
    Payload:Object Reference=Get Input Data Object()
    DefaultDragVisual:Widget Object Reference=Self-Reference()
    Pivot:EDragPivot Enum=TopLeft
    Offset:Vector 2D Structure=(X=0,Y=0)
   ) -> (then:Exec, ReturnValue:Drag Drop Operation Object Reference) {is=title K2Node_CreateDragDropOperation}
   > Set Is Being Dragged(Is Being Dragged=true, self=Get Input Data Object())
   > Call On Drag Started!(
    execute:Exec
    self:Meta Sound Preset Migrator Input Entry Object Reference
    Input Data Object:Meta Sound Preset Migrator Input Data Object Reference=Get Input Data Object()
   ) -> (then:Exec) {is=title K2Node_CallDelegate}
   > Return Node!(execute, Operation:Drag Drop Operation Object Reference=@n1)
  ]}
 OnDragEnded {nodes [OnDragEnded() {is=title K2Node_FunctionEntry}]}
 OnDragOver {nodes [
   n1 = OnDragOver!() -> (then, MyGeometry:Geometry Structure, PointerEvent:Pointer Event Structure, Operation:Drag Drop Operation Object Reference) {is=title K2Node_FunctionEntry}
   > n2 = Cast To MetaSoundPresetMigrator_InputData(Object=Get Payload!(self:Drag Drop Operation Object Reference=@n1.Operation) -> (Payload:Object Reference))
   n3 = Get Data Type(self=@n2) -> (Data Type=None)
   n4 = Get Data Type(self=Get Input Data Object()) -> (Data Type=None)
   n5 = Get Is Origin(self=@n2)
   n6 = Get Is Origin(self=Get Input Data Object())
   > n7 = Set Is Dragged Over(
    Is Dragged Over=AND Boolean(A="Equal (Name)"(A=@n3, B=@n4), B="Not Equal (Boolean)"(A=@n5, B=@n6) -> (ReturnValue=~))
    self=Get Input Data Object()
   )
   > Return Node!(execute, ReturnValue:bool=@n7)
   Return Node(execute=@n2.CastFailed, ReturnValue=false)
  ]}
 OnDragStarted {nodes [OnDragStarted!() -> (then, Input Data Object:Meta Sound Preset Migrator Input Data Object Reference) {is=title K2Node_FunctionEntry}]}
 OnDrop {nodes [
   OnDrop!() -> (then, MyGeometry:Geometry Structure, PointerEvent:Pointer Event Structure, Operation:Drag Drop Operation Object Reference) {is=title K2Node_FunctionEntry}
   > Set Is Dragged Over(Is Dragged Over=false, self=Get Input Data Object())
   > n1 = Sequence()
   > Branch(Condition=Get Is Enabled!(self:Meta Sound Preset Migrator Input Data Object Reference=Get Input Data Object()) -> (Is Enabled:bool))
   > n2 = Cast To MetaSoundPresetMigrator_InputData(Object=Get Payload(self=Get Operation() -> (Operation:Drag Drop Operation Object Reference)))
   > n3 = Set Dropped Data(Dropped Data:Meta Sound Preset Migrator Input Data Object Reference=@n2) -> (Output_Get:Meta Sound Preset Migrator Input Data Object Reference)
   n4 = Get Linked Input(self=@n3)
   n5 = Get Input Data Object()
   > Branch(Condition=AND Boolean(
     A="Not Equal (Object)"(A:Object Reference=@n3, B:Object Reference=@n5) -> (ReturnValue:bool=~) {is=title K2Node_PromotableOperator}
     B="Not Equal (Object)"(A=@n5, B=@n4) -> (ReturnValue=~)
    ))
   > n6 = Get!(execute) -> (then, else, Input Data Object:Meta Sound Preset Migrator Input Data Object Reference)
   > n7 = Get!(execute, self:Meta Sound Preset Migrator Input Data Object Reference=@n6) -> (then, else, Linked Input:Meta Sound Preset Migrator Input Data Object Reference)
   > n8 = Set Linked Input!(
    execute
    Linked Input:Meta Sound Preset Migrator Input Data Object Reference
    self:Meta Sound Preset Migrator Input Data Object Reference=@n7
   ) -> (then, Output_Get:Meta Sound Preset Migrator Input Data Object Reference)
   Set Linked Input(
    execute=[@n8.then, @n7.else, @n6.else]
    Linked Input=Get Dropped Data() -> (Dropped Data:Meta Sound Preset Migrator Input Data Object Reference)
    self=Get Input Data Object()
   )
   > n9 = Get(self=Get Dropped Data())
   > n10 = Set Linked Input(self=@n9)
   Set Linked Input(execute=[@n10.then, @n9.else], Linked Input=Get Input Data Object(), self=Get Dropped Data())
   > Call On Dropped!(execute:Exec, self:Meta Sound Preset Migrator Input Entry Object Reference) -> (then:Exec) {is=title K2Node_CallDelegate}
   n11 = Sequence(execute=@n1.then_1)
   > n12 = Cast To MetaSoundPresetMigrator_InputEntry!(
    execute
    Object=Get DefaultDragVisual!(self:Drag Drop Operation Object Reference=Get Operation()) -> (DefaultDragVisual:Widget Object Reference)
   ) -> (then, CastFailed, AsMeta Sound Preset Migrator Input Entry:Meta Sound Preset Migrator Input Entry Object Reference) {is=title K2Node_DynamicCast}
   > Set Is Being Dragged(
    Is Being Dragged=false
    self=Get Input Data Object!(self:Meta Sound Preset Migrator Input Entry Object Reference=@n12) -> (Input Data Object)
   )
   Sequence(execute=@n11.then_1)
   > Call On Drag Ended()
   > Return Node(ReturnValue=true)
  ]}
 OnDropped {nodes [OnDropped() {is=title K2Node_FunctionEntry}]}
]
variables [
 Input Data Object {type=MetaSoundPresetMigrator_InputData_C, default=null}
 OnDragEnded {default=()}
 OnDragStarted {default=()}
 OnDropped {default=()}
 Parent Widget {type=MetaSoundPresetMigrator_C, default=null}
 Style {type=MetaSoundPresetMigrator_Style_C, default=null}
]
widgets {
 info {rootWidgetClass=&VerticalBox, widgetCount=7}
 widgets [
  Widget {widgetClassPath=&VerticalBox, widget=Widget}
  BackgroundBorder {
   widgetClassPath=&Border
   bIsVariable=true
   parent=Widget
   widget=BackgroundBorder
   properties {brushColor=<rgba 0 0 0 0.1>, padding=<ltrb 6 4 6 4>}
   slot {class=VerticalBoxSlot}
  }
  HorizontalBox_31 {
   widgetClassPath=&HorizontalBox
   parent=BackgroundBorder
   widget=HorizontalBox_31
   slot {class=BorderSlot, properties {padding=<ltrb 6 4 6 4>}}
  }
  TypeImage {
   widgetClassPath=&Image
   bIsVariable=true
   parent=HorizontalBox_31
   widget=TypeImage
   properties {brush {
     bIsDynamicallyLoaded=false
     drawAs=Image
     imageType=FullColor
     mirroring=NoMirror
     resourceName=None
     resourceObject=/Game/UI/Friends_UI/Social/circle
     tiling=NoTile
     imageSize=<6 6>
     margin=<ltrb 0 0 0 0>
     outlineSettings {
      bUseBrushTransparency=false
      roundingType=HalfHeightRadius
      width=0
      color {colorUseRule=UseColor_Specified, specifiedColor=<rgba 0 0 0 0>}
      cornerRadii=<0 0 0 0>
     }
     tintColor {colorUseRule=UseColor_Specified, specifiedColor=<rgba 1 1 1 1>}
     uVRegion {bIsValid=false, max=<0 0>, min=<0 0>}
    }}
   slot {class=HorizontalBoxSlot, properties {horizontalAlignment=HAlign_Center, verticalAlignment=VAlign_Center}}
  }
  TextBlock_26 {
   widgetClassPath=&TextBlock
   parent=HorizontalBox_31
   widget=TextBlock_26
   properties {
    text=Input Name
    font {
     bForceMonospaced=false
     bMaterialIsStencil=false
     fontMaterial=null
     fontObject=&Roboto
     letterSpacing=0
     monospacedWidth=1
     size=9
     skewAmount=0
     typefaceFontName=Regular
     outlineSettings {
      bApplyOutlineToDropShadows=false
      bMiteredCorners=false
      bSeparateFillAlpha=false
      outlineMaterial=null
      outlineSize=0
      outlineColor=<rgba 0 0 0 1>
     }
    }
   }
   slot {class=HorizontalBoxSlot, properties {horizontalAlignment=HAlign_Left, verticalAlignment=VAlign_Center, padding=<ltrb 6 0 0 0>}}
  }
  VerticalBox_1 {
   widgetClassPath=&VerticalBox
   parent=HorizontalBox_31
   widget=VerticalBox_1
   slot {class=HorizontalBoxSlot, properties {horizontalAlignment=HAlign_Right, verticalAlignment=VAlign_Center, size {sizeRule=Fill, value=1}}}
  }
  LinkedIcon {
   widgetClassPath=&Image
   bIsVariable=true
   parent=HorizontalBox_31
   widget=LinkedIcon
   properties {
    brush {
     bIsDynamicallyLoaded=false
     drawAs=Image
     imageType=FullColor
     mirroring=NoMirror
     resourceName=None
     resourceObject=null
     tiling=NoTile
     imageSize=<16 16>
     margin=<ltrb 0 0 0 0>
     outlineSettings {
      bUseBrushTransparency=false
      roundingType=HalfHeightRadius
      width=0
      color {colorUseRule=UseColor_Specified, specifiedColor=<rgba 0 0 0 0>}
      cornerRadii=<0 0 0 0>
     }
     tintColor {colorUseRule=UseColor_Specified, specifiedColor=<rgba 1 1 1 1>}
     uVRegion {bIsValid=false, max=<0 0>, min=<0 0>}
    }
    colorAndOpacity=<rgba 0.2582 0.5395 0.0685 1>
   }
   slot {class=HorizontalBoxSlot}
  }
 ]
}
