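# Editor utility widget that migrates a MetaSound Source preset to a new parent MetaSound, carrying over input values and source properties.
# Builds searchable origin and destination input lists from per-input data objects, linking type-matched pairs in both directions.
# Links load from a saved-links asset keyed by parent pair, falling back to matching names.
# Migration creates a preset from linked literals, offering to discard it if property migration fails.
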EditorUtilityWidgetBlueprint /Game/Sounds/Fort_Audio_Blueprints/Utilities/EditorUtilityWidgets/Metasounds/MetaSoundPresetMigrator
graphs [
 Are Both Assets Valid {nodes [
   Are Both Assets Valid() {is=title K2Node_FunctionEntry}
   n1 = Get Origin MetaSound() -> (Origin MetaSound:Meta Sound Source Object Reference)
   n2 = Is Preset!(MetaSound Source:Meta Sound Source Object Reference=@n1, __WorldContext:Object Reference) -> (ReturnValue:bool)
   n3 = IsValid(Object=@n1)
   n4 = Get Destination MetaSound() -> (Destination MetaSound:Meta Sound Source Object Reference)
   n5 = IsValid(Object=@n4)
   n6 = NOT Boolean(A=Is Preset(MetaSound Source=@n4))
   > Return Node!(execute, ReturnValue:bool=AND Boolean(A=AND Boolean(A=@n3, B=@n2), B=AND Boolean(A=@n5, B=@n6)))
  ]}
 Clear All Links {nodes [
   Clear All Links() {is=title K2Node_FunctionEntry}
   > n1 = Sequence()
   > Get!(execute:Exec) -> (then:Exec, else:Exec, Origin MetaSound:Meta Sound Source Object Reference)
   > n2 = Values!(
    execute
    TargetMap:Map of Names to Meta Sound Preset Migrator Input Data Object References=Get Origin MetaSound Input Data Objects() -> (Origin MetaSound Input Data Objects:Map of Names to Meta Sound Preset Migrator Input Data Object References)
   ) -> (then, Values:Array of Meta Sound Preset Migrator Input Data Object References)
   > n3 = For Each Loop(Array:Array of Meta Sound Preset Migrator Input Data Object References=@n2) -> (Array Element:Meta Sound Preset Migrator Input Data Object Reference)
   > Set Linked Input!(
    execute
    Linked Input:Meta Sound Preset Migrator Input Data Object Reference
    self:Meta Sound Preset Migrator Input Data Object Reference=@n3
   ) -> (then, Output_Get:Meta Sound Preset Migrator Input Data Object Reference)
   Sequence(execute=@n1.then_1)
   > Get!(execute) -> (then, else, Destination MetaSound:Meta Sound Source Object Reference)
   > n4 = Values(TargetMap=Get Destination MetaSound Input Data Objects() -> (Destination MetaSound Input Data Objects:Map of Names to Meta Sound Preset Migrator Input Data Object References))
   > n5 = For Each Loop(Array=@n4)
   > Set Linked Input(self=@n5)
  ]}
 Clear Selection Cache {nodes [
   Clear Selection Cache() {is=title K2Node_FunctionEntry}
   > n1 = Sequence()
   > n2 = Get!(execute) -> (then, else, Selected Origin Widget Cache:Meta Sound Preset Migrator Input Data Object Reference)
   > n3 = Set Is Selected!(execute, Is Selected:bool=false, self:Meta Sound Preset Migrator Input Data Object Reference=@n2) -> (then, Output_Get:bool)
   Set Selected Origin Widget Cache(execute=[@n3.then, @n2.else], Selected Origin Widget Cache:Meta Sound Preset Migrator Input Data Object Reference) -> (Output_Get:Meta Sound Preset Migrator Input Data Object Reference)
   Sequence(execute=@n1.then_1)
   > n4 = Get!(execute) -> (then, else, Selected Destination Widget Cache:Meta Sound Preset Migrator Input Data Object Reference)
   > n5 = Set Is Selected(Is Selected=false, self=@n4)
   Set Selected Destination Widget Cache(execute=[@n5.then, @n4.else], Selected Destination Widget Cache:Meta Sound Preset Migrator Input Data Object Reference) -> (Output_Get:Meta Sound Preset Migrator Input Data Object Reference)
  ]}
 Create MetaSound Input Data Objects {nodes [
   n1 = Create MetaSound Input Data Objects!() -> (then, MetaSound Source:Meta Sound Source Object Reference, Is Origin:bool=~) {is=title K2Node_FunctionEntry}
   > Set Input Literals(Input Literals:Map of Names to Meta Sound Literal Structures=Get All MetaSound Source Input Literals!(MetaSound Source:Meta Sound Source Object Reference=@n1, __WorldContext:Object Reference) -> (Input Default Literals:Map of Names to Meta Sound Literal Structures) {is=title K2Node_CallFunction}) -> (Output_Get:Map of Names to Meta Sound Literal Structures)
   > n2 = Set Input Types(Input Types:Map of Names to Names=Get All MetaSound Source Input Types!(MetaSound Source:Meta Sound Source Object Reference=@n1, __WorldContext:Object Reference) -> (Input Types:Map of Names to Names) {is=title K2Node_CallFunction}) -> (Output_Get:Map of Names to Names)
   > n3 = Keys!(execute, TargetMap:Map of Names to Names=@n2) -> (then, Keys:Array of Names)
   > n4 = For Each Loop(Array:Array of Names=@n3) -> (Array Element:Name)
   Return Node!(
    execute=@n4.Completed
    MetaSound Input Literals:Map of Names to Meta Sound Literal Structures=Get Input Literals() -> (Input Literals:Map of Names to Meta Sound Literal Structures)
    MetaSound Input Data Objects:Map of Names to Meta Sound Preset Migrator Input Data Object References=Get Input Data Objects() -> (Input Data Objects:Map of Names to Meta Sound Preset Migrator Input Data Object References)
   )
   > Set Input Name(Input Name:Name=@n4) -> (Output_Get:Name=None)
   > Set Input Data Type(Input Data Type:Name=Find(
     TargetMap:Map of Names to Names=Get Input Types() -> (Input Types:Map of Names to Names)
     Key:"Name (by ref)"=Get Input Name() -> (Input Name:Name=None)
    ) -> (Value:Name)) -> (Output_Get:Name=None)
   > Branch(Condition="Not Equal (Name)"(A=Get Input Data Type() -> (Input Data Type:Name=None), B=Trigger) -> (ReturnValue=~))
   > Increment Int(Value=Get Row Index() -> (Row Index:int=0))
   > n5 = Construct Object from Class!(
    execute
    Class=/Game/Sounds/Fort_Audio_Blueprints/Utilities/EditorUtilityWidgets/Metasounds/Tools/MetaSoundPresetMigratorData/MetaSoundPresetMigrator_InputData.MetaSoundPresetMigrator_InputData_C
    self
    Is Origin:bool=Get Is Origin() -> (Is Origin:bool)
    Input Name:Name=Get Input Name() -> (Input Name=None)
    Row Index:int=Get Row Index() -> (Row Index=0)
    Data Type:Name=Get Input Data Type() -> (Input Data Type=None)
    Input Literal:Meta Sound Literal Structure=Find(TargetMap:Map of Names to Meta Sound Literal Structures=Get Input Literals(), Key=Get Input Name() -> (Input Name=None)) -> (Value:Meta Sound Literal Structure)
   ) -> (then, ReturnValue:Meta Sound Preset Migrator Input Data Object Reference)
   > Add!(
    execute
    TargetMap:Map of Names to Meta Sound Preset Migrator Input Data Object References=Get Input Data Objects()
    Key:"Name (by ref)"=Get Input Name() -> (Input Name=None)
    Value:"Meta Sound Preset Migrator Input Data Object Reference (by ref)"=@n5
   ) -> (then) {is=title K2Node_CallFunction}
  ]}
 Disable Inputs Ineligible for Linking {nodes [
   Disable Inputs Ineligible for Linking!() -> (then, Input Data Object:Meta Sound Preset Migrator Input Data Object Reference) {is=title K2Node_FunctionEntry}
   > n1 = Sequence()
   > Set Input Belongs to Origin(Input Belongs to Origin:bool=Get Is Origin!(self:Meta Sound Preset Migrator Input Data Object Reference=Get Input Data Object() -> (Input Data Object:Meta Sound Preset Migrator Input Data Object Reference)) -> (Is Origin)) -> (Output_Get:bool)
   > Set Data Type(Data Type:Name=Get Data Type!(self:Meta Sound Preset Migrator Input Data Object Reference=Get Input Data Object()) -> (Data Type:Name=None)) -> (Output_Get:Name=None)
   n2 = Sequence(execute=@n1.then_1)
   n3 = Get OriginMetaSoundInputsList() -> (OriginMetaSoundInputsList:Editor Utility List View Object Reference)
   n4 = Get DestinationMetaSoundInputsList() -> (DestinationMetaSoundInputsList:Editor Utility List View Object Reference)
   n5 = Get Input Belongs to Origin() -> (Input Belongs to Origin:bool)
   > n6 = For Each Loop(Array:Array of Object References=GetListItems!(self:List View Object Reference=Select(Option 0:Editor Utility List View Object Reference=@n3, Option 1:Editor Utility List View Object Reference=@n4, Index=@n5) -> (ReturnValue:Editor Utility List View Object Reference)) -> (ReturnValue:Array of Object References)) -> (Array Element:Object Reference)
   > n7 = Cast To MetaSoundPresetMigrator_InputData!(execute, Object=@n6) -> (then, CastFailed, AsMeta Sound Preset Migrator Input Data:Meta Sound Preset Migrator Input Data Object Reference) {is=title K2Node_DynamicCast}
   > Set Is Enabled!(
    execute
    Is Enabled:bool="Equal (Name)"(A=Get Data Type(self=@n7) -> (Data Type=None), B=Get Data Type!() -> (Data Type=None))
    self:Meta Sound Preset Migrator Input Data Object Reference=@n7
   ) -> (then, Output_Get:bool)
   n8 = Sequence(execute=@n2.then_1)
   n9 = Get Input Belongs to Origin()
   n10 = Get OriginMetaSoundInputsList()
   n11 = Get DestinationMetaSoundInputsList()
   > n12 = For Each Loop(Array=GetListItems(self=Select(Option 0=@n11, Option 1=@n10, Index=@n9)))
   > n13 = Cast To MetaSoundPresetMigrator_InputData(Object=@n12)
   > Set Is Enabled(Is Enabled="Equal (Object)"(A=@n13, B=Get Input Data Object()), self=@n13)
   Sequence(execute=@n8.then_1)
   > Clear Selection Cache() {is=title K2Node_CallFunction}
  ]}
 Enable All Widgets {nodes [
   Enable All Widgets() {is=title K2Node_FunctionEntry}
   > n1 = Sequence()
   > Append Array!(
    execute
    TargetArray:Array of Object References=Get All Input Objects() -> (All Input Objects:Array of Object References)
    SourceArray:Array of Object References=GetListItems(self=Get OriginMetaSoundInputsList())
   ) -> (then)
   > Append Array(TargetArray=Get All Input Objects(), SourceArray=GetListItems(self=Get DestinationMetaSoundInputsList()))
   Sequence(execute=@n1.then_1)
   > n2 = For Each Loop(Array=Get All Input Objects())
   > n3 = Cast To MetaSoundPresetMigrator_InputData(Object=@n2)
   > Set Is Enabled(Is Enabled=true, self=@n3)
  ]}
 EventGraph {nodes [
   n1 = "On Item Double Clicked (DestinationMetaSoundInputsList)"!() -> (OutputDelegate, then, Item:Object Reference) {is=title K2Node_ComponentBoundEvent}
   > n2 = Cast To MetaSoundPresetMigrator_InputData(Object=@n1.Item)
   > Scroll Item Into View!(
    execute
    self:List View Object Reference=Get OriginMetaSoundInputsList()
    Item:Object Reference=Get Linked Input!(self:Meta Sound Preset Migrator Input Data Object Reference=@n2) -> (Linked Input:Meta Sound Preset Migrator Input Data Object Reference)
   ) -> (then)
   "On Property Changed (DestinationMetaSoundPicker)"!() -> (OutputDelegate, then, PropertyName:Name=None) {is=title K2Node_ComponentBoundEvent}
   > n3 = Create MetaSound Input Data Objects!(execute, self, MetaSound Source:Meta Sound Source Object Reference=Get Destination MetaSound(), Is Origin:bool=false) -> (
    then
    MetaSound Input Literals:Map of Names to Meta Sound Literal Structures
    MetaSound Input Data Objects:Map of Names to Meta Sound Preset Migrator Input Data Object References
   ) {is=title K2Node_CallFunction}
   > Set Destination MetaSound Input Literals(Destination MetaSound Input Literals:Map of Names to Meta Sound Literal Structures=@n3) -> (Output_Get:Map of Names to Meta Sound Literal Structures)
   > n4 = Set Destination MetaSound Input Data Objects(Destination MetaSound Input Data Objects:Map of Names to Meta Sound Preset Migrator Input Data Object References=@n3.MetaSound Input Data Objects) -> (Output_Get:Map of Names to Meta Sound Preset Migrator Input Data Object References)
   > n5 = Set List Items!(
    execute
    self:List View Object Reference=Get DestinationMetaSoundInputsList()
    InListItems:Array of Object References=Sort Input List!(Input Objects:Map of Names to Meta Sound Preset Migrator Input Data Object References=@n4) -> (Alphabetical List:Array of Meta Sound Preset Migrator Input Data Object References)
   ) -> (then) {is=title K2Node_CallFunction}
   Initialize Links(execute=[@n5.then, @n40.then])
   n6 = "On Text Changed (AssetNameTextBox)"!() -> (OutputDelegate, then, Text:"Text (by ref)") {is=title K2Node_ComponentBoundEvent}
   > Set Asset Name(Asset Name:String="To String (Text)"(InText=@n6.Text)) -> (Output_Get:String)
   n7 = "On Text Committed (AssetNameTextBox)"!() -> (OutputDelegate, then, Text:"Text (by ref)", CommitMethod:ETextCommit Enum=Default) {is=title K2Node_ComponentBoundEvent}
   > n8 = Switch on ETextCommit!(execute, Selection:ETextCommit Enum=@n7.CommitMethod) -> (Default:Exec, OnEnter:Exec, OnUserMovedFocus:Exec, OnCleared:Exec) {is=title K2Node_SwitchEnum}
   Migrate Preset(execute=@n8.OnEnter)
   "On Clicked (UserGuideButton)"() {is=title K2Node_ComponentBoundEvent}
   > SetVisibility(
    self=Get MetaSoundPresetMigrator_HelpWidget() -> (MetaSoundPresetMigrator_HelpWidget:Meta Sound Preset Migrator Help Widget Object Reference)
    InVisibility=Visible
   )
   > n9 = SetIsEnabled(self=Get Content() -> (Content:Vertical Box Object Reference), bInIsEnabled=false)
   n10 = Close Help Widget() {is=title K2Node_CustomEvent}
   Bind Event to On Close Button Clicked(execute=@n9.then, self:Meta Sound Preset Migrator Help Widget Object Reference=Get MetaSoundPresetMigrator_HelpWidget(), Delegate=@n10) {is=title K2Node_AddDelegate}
   "On Clicked (ToggleThemeButton)"() {is=title K2Node_ComponentBoundEvent}
   > Update Theme()
   n11 = "On Item Clicked (OriginMetaSoundInputsList)"!() -> (OutputDelegate, then, Item:Object Reference) {is=title K2Node_ComponentBoundEvent}
   > Clear Selection Cache()
   > n12 = Cast To MetaSoundPresetMigrator_InputData(Object=@n11.Item)
   > n13 = Set Clicked Data(Clicked Data:Meta Sound Preset Migrator Input Data Object Reference=@n12) -> (Output_Get:Meta Sound Preset Migrator Input Data Object Reference)
   > Set Is Selected(Is Selected=true, self=@n13)
   > n14 = Set Selected Origin Widget Cache(Selected Origin Widget Cache=Get Clicked Data() -> (Clicked Data:Meta Sound Preset Migrator Input Data Object Reference))
   Branch(
    execute=[@n18.then, @n14.then]
    Condition=Is Shift Down!(KeysState:"Modifier Keys State Structure (by ref)"=GetModifierKeysState!() -> (ReturnValue:Modifier Keys State Structure)) -> (ReturnValue:bool)
   )
   > Unlink Data!(execute, self, Data to Unlink:Meta Sound Preset Migrator Input Data Object Reference=Get Clicked Data()) -> (then)
   n15 = "On Item Clicked (DestinationMetaSoundInputsList)"!() -> (OutputDelegate, then, Item:Object Reference) {is=title K2Node_ComponentBoundEvent}
   > Clear Selection Cache()
   > n16 = Cast To MetaSoundPresetMigrator_InputData(Object=@n15.Item)
   > n17 = Set Clicked Data(Clicked Data=@n16)
   > Set Is Selected(Is Selected=true, self=@n17)
   > n18 = Set Selected Destination Widget Cache(Selected Destination Widget Cache=Get Clicked Data())
   n19 = "On Item Double Clicked (OriginMetaSoundInputsList)"!() -> (OutputDelegate, then, Item:Object Reference) {is=title K2Node_ComponentBoundEvent}
   > n20 = Cast To MetaSoundPresetMigrator_InputData(Object=@n19.Item)
   > Scroll Item Into View(self=Get DestinationMetaSoundInputsList(), Item=Get Linked Input(self=@n20))
   n21 = "On Clicked (MigratePresetButton)"() {is=title K2Node_ComponentBoundEvent}
   n22 = Get Origin MetaSound()
   n23 = Create MetaSound Preset with Literals!(
    execute=[@n21.then, @n45.then]
    Parent MetaSound Source:Meta Sound Source Object Reference=Get Destination MetaSound()
    MetaSound Input Literals:Map of Names to Meta Sound Literal Structures=Get Linked Input Literals!() -> (Linked Inputs:Map of Names to Meta Sound Literal Structures) {is=title K2Node_CallFunction}
    Package Path:String=Split!(InPath:String=Get Object Path String(Object=@n22)) -> (PathPart:String, FilenamePart:String, ExtensionPart:String)
    Asset Name:String=Get Asset Name() -> (Asset Name:String)
    Suffix:String
    __WorldContext:Object Reference
   ) -> (then, MetaSound Source Preset:Meta Sound Source Object Reference)
   > n24 = Is Valid(InputObject=@n23)
   K2Node_Composite_1 = Notification Failed to migrate(execute=@n24.Is Not Valid) {is=title K2Node_Composite}
   > n25 = Migrate Source Properties!(execute, self) -> (then, Success:bool)
   > n26 = Branch(Condition=@n25)
   > Clear All Links() {is=title K2Node_CallFunction}
   > Set Origin MetaSound(Origin MetaSound:Meta Sound Source Object Reference) -> (Output_Get:Meta Sound Source Object Reference)
   > Initialize Asset Name()
   > K2Node_Composite_3 = Notification Successfully migrated() {is=title K2Node_Composite}
   K2Node_Composite_2 = Would you like to keep the newly created asset?!(execute=@n26.else) -> (No:Exec, Yes:Exec) {is=title K2Node_Composite}
   > DeleteLoadedAsset!(execute, AssetToDelete:Object Reference=@n23) -> (then, ReturnValue:bool)
   "On Clicked (UnlinkAllButton)"() {is=title K2Node_ComponentBoundEvent}
   > n27 = For Each Loop(Array=GetListItems(self=Get DestinationMetaSoundInputsList()))
   n28 = For Each Loop(Exec=@n27.Completed, Array=GetListItems(self=Get OriginMetaSoundInputsList()))
   Clear Selection Cache(execute=[@n31.then, @n28.Completed])
   n29 = Cast To MetaSoundPresetMigrator_InputData(execute=@n28.LoopBody, Object=@n28)
   > Unlink Data(Data to Unlink=@n29)
   n30 = Cast To MetaSoundPresetMigrator_InputData(execute=@n27.LoopBody, Object=@n27)
   > Unlink Data(Data to Unlink=@n30)
   "On Clicked (LinkMatchingButton)"() {is=title K2Node_ComponentBoundEvent}
   > n31 = Link Matching Input Names!(execute, self) -> (then, Success:bool)
   n32 = "On Entry Generated (DestinationMetaSoundInputsList)"!() -> (OutputDelegate, then, Widget:User Widget Object Reference) {is=title K2Node_ComponentBoundEvent}
   n33 = "On Entry Generated (OriginMetaSoundInputsList)"!() -> (OutputDelegate, then, Widget:User Widget Object Reference) {is=title K2Node_ComponentBoundEvent}
   > n34 = Route Value!(
    Origin:Exec
    Origin Value:User Widget Object Reference=@n33.Widget
    Destination:Exec=@n32.then
    Destination Value:User Widget Object Reference=@n32.Widget
   ) -> (then:Exec, Value:Meta Sound Preset Migrator Input Entry Object Reference) {is=title K2Node_MacroInstance}
   > n35 = Cast To MetaSoundPresetMigrator_InputEntry!(execute, Object=@n34) -> (then, CastFailed, AsMeta Sound Preset Migrator Input Entry:Meta Sound Preset Migrator Input Entry Object Reference) {is=title K2Node_DynamicCast}
   > Bind Event to On Drag Started(self:Meta Sound Preset Migrator Input Entry Object Reference=@n35, Delegate=Create Event()) {is=title K2Node_AddDelegate}
   > n36 = Bind Event to On Drag Ended(self:Meta Sound Preset Migrator Input Entry Object Reference=@n35, Delegate=Create Event()) {is=title K2Node_AddDelegate}
   n37 = Mark Dirty() {is=title K2Node_CustomEvent}
   Bind Event to On Dropped(execute=@n36.then, self:Meta Sound Preset Migrator Input Entry Object Reference=@n35, Delegate=@n37) {is=title K2Node_AddDelegate}
   > Set Parent Widget!(
    execute
    Parent Widget:Meta Sound Preset Migrator Object Reference=Self-Reference()
    self:Meta Sound Preset Migrator Input Entry Object Reference=@n35
   ) -> (then, Output_Get:Meta Sound Preset Migrator Object Reference)
   > Set Style!(
    execute
    Style:Meta Sound Preset Migrator Style Object Reference=Get Style!() -> (Style:Meta Sound Preset Migrator Style Object Reference) {is=title K2Node_CallFunction}
    self:Meta Sound Preset Migrator Input Entry Object Reference=@n35
   ) -> (then, Output_Get:Meta Sound Preset Migrator Style Object Reference)
   "On Clicked (SaveLinksButton)"() {is=title K2Node_ComponentBoundEvent}
   > Save Linked Inputs()
   "On Property Changed (OriginMetaSoundPicker)"!() -> (OutputDelegate, then, PropertyName:Name=None) {is=title K2Node_ComponentBoundEvent}
   > Initialize Asset Name()
   > n38 = Create MetaSound Input Data Objects(MetaSound Source=Get Origin MetaSound(), Is Origin=true)
   > Set Origin MetaSound Input Literals(Origin MetaSound Input Literals:Map of Names to Meta Sound Literal Structures=@n38) -> (Output_Get:Map of Names to Meta Sound Literal Structures)
   > n39 = Set Origin MetaSound Input Data Objects(Origin MetaSound Input Data Objects:Map of Names to Meta Sound Preset Migrator Input Data Object References=@n38.MetaSound Input Data Objects) -> (Output_Get:Map of Names to Meta Sound Preset Migrator Input Data Object References)
   > n40 = Set List Items(self=Get OriginMetaSoundInputsList(), InListItems=Sort Input List(Input Objects=@n39))
   n41 = "On Text Changed (OriginInputSearchBox)"!() -> (OutputDelegate, then, Text:"Text (by ref)") {is=title K2Node_ComponentBoundEvent}
   > n42 = Filter Input Entries!(
    execute
    self
    Data Objects:Map of Names to Meta Sound Preset Migrator Input Data Object References=Get Origin MetaSound Input Data Objects()
    Search Text:Text=@n41.Text
   ) -> (then, Filtered Data Objects:Map of Names to Meta Sound Preset Migrator Input Data Object References)
   > Set List Items(self=Get OriginMetaSoundInputsList(), InListItems=Sort Input List(Input Objects=@n42))
   n43 = "On Text Changed (DestinationInputSearchBox)"!() -> (OutputDelegate, then, Text:"Text (by ref)") {is=title K2Node_ComponentBoundEvent}
   > n44 = Filter Input Entries(Data Objects=Get Destination MetaSound Input Data Objects(), Search Text=@n43.Text)
   > Set List Items(self=Get DestinationMetaSoundInputsList(), InListItems=Sort Input List(Input Objects=@n44))
   "On Clicked (ReloadSavedLinksButton)"() {is=title K2Node_ComponentBoundEvent}
   > Load Saved Links!(execute, self) -> (then, Success:bool)
   n45 = Migrate Preset() {is=title K2Node_CustomEvent}
   Update Theme() {is=title K2Node_CustomEvent}
   > Set Text Color(Text Color:Linear Color Structure=Get Class Defaults!(Class:Object Class Reference=Get Style Class() -> (Style Class:Meta Sound Preset Migrator Style Class Reference)) -> (Text Color:Linear Color Structure) {is=title K2Node_GetClassDefaults}) -> (Output_Get:Linear Color Structure)
   Event PreConstruct()
   > SetObject(
    self=[
     Get DestinationMetaSoundPicker() -> (DestinationMetaSoundPicker:Single Property View Object Reference)
     Get OriginMetaSoundPicker() -> (OriginMetaSoundPicker:Single Property View Object Reference)
    ]
    NewObject=Self-Reference()
   )
   > SetVisibility(self=Get MetaSoundPresetMigrator_HelpWidget(), InVisibility=Hidden)
   Event Construct()
   > Update Theme() {is=title K2Node_CallFunction}
   SetIsEnabled(execute=@n46.then, self=Get Content(), bInIsEnabled=true)
   Set Is Save Dirty(execute=@n37.then, Is Save Dirty:bool=true) -> (Output_Get:bool)
   n46 = SetVisibility(execute=@n10.then, self=Get MetaSoundPresetMigrator_HelpWidget(), InVisibility=Hidden)
  ]}
 Notification Failed to migrate {
  path=EventGraph.K2Node_Composite_1.Notification Failed to migrate
  nodes [
   Inputs()
   > n1 = AddEditorNotification!(
    execute
    EditorNotificationSettings:"Editor Notification Settings Structure (by ref)"=MakeEditorNotificationSettings!(
     Text:Text="LOCTEXT(\"Failed to migrate MetaSound Source preset to new parent.\")"
     SubText:Text=Format Text(Format="LOCTEXT(\"`Create MetaSound Preset with Literals` operation failed or was canceled by user.\")")
     FadeInDuration:float=0.5
     FadeOutDuration:float=2
     ExpireDuration:float=15
     bUseThrobber:bool=true
     bUseSuccessFailIcons:bool=true
     bFireAndForget:bool=true
     bAllowThrottleWhenFrameRateIsLow:bool=true
    ) -> (ReturnValue:Editor Notification Settings Structure)
   ) -> (then, ReturnValue:Editor Notification Handle Structure)
   > SetCompletionState!(execute, NotificationHandle:"Editor Notification Handle Structure (by ref)"=@n1, State:EEditorNotificationCompletionState Enum=Fail) -> (then)
   Outputs()
  ]
 }
 Would you like to keep the newly created asset? {
  path=EventGraph.K2Node_Composite_2.Would you like to keep the newly created asset?
  nodes [
   Inputs()
   > n1 = Show Message Dialog(
    Title=Format Text(Format="LOCTEXT(\"Failed to migrate source properties!\")")
    Message=Format Text(Format="LOCTEXT(\"Successfully migrated inputs to the new preset, but failed to migrate source properties. \\r\\n\\r\\nWould you like to keep the newly created asset? Properties that failed to migrate will need to be fixed manually.\\r\\n\\r\\n\")")
    MessageType=YesNo
    DefaultValue=Yes
    MessageCategory=Warning
   )
   > n2 = Switch on EAppReturnType(Selection=@n1)
   > Outputs!(No:Exec, Yes:Exec=@n2.Yes)
  ]
 }
 Notification Successfully migrated {
  path=EventGraph.K2Node_Composite_3.Notification Successfully migrated
  nodes [
   Inputs()
   > n1 = AddEditorNotification(EditorNotificationSettings=MakeEditorNotificationSettings(
     Text="LOCTEXT(\"Successfully migrated MetaSound Source preset.\")"
     FadeInDuration=0.5
     FadeOutDuration=2
     ExpireDuration=3
     bUseThrobber=true
     bUseSuccessFailIcons=true
     bFireAndForget=true
     bAllowThrottleWhenFrameRateIsLow=true
    ))
   > SetCompletionState(NotificationHandle=@n1, State=Success)
   Outputs!()
  ]
 }
 Filter Input Entries {nodes [
   n1 = Filter Input Entries!() -> (then, Data Objects:Map of Names to Meta Sound Preset Migrator Input Data Object References, Search Text:Text) {is=title K2Node_FunctionEntry}
   > n2 = Values(TargetMap=@n1)
   > n3 = For Each Loop(Array:Array of Meta Sound Preset Migrator Input Data Object References=@n2) -> (Array Element:Meta Sound Preset Migrator Input Data Object Reference)
   Return Node!(
    execute=@n3.Completed
    Filtered Data Objects:Map of Names to Meta Sound Preset Migrator Input Data Object References=Get Filtered Data() -> (Filtered Data:Map of Names to Meta Sound Preset Migrator Input Data Object References)
   )
   > n4 = Set Data Object(Data Object:Meta Sound Preset Migrator Input Data Object Reference=@n3) -> (Output_Get:Meta Sound Preset Migrator Input Data Object Reference)
   n5 = Get Search Text() -> (Search Text:Text)
   n6 = "To String (Name)"(InName=Get Input Name!(self:Meta Sound Preset Migrator Input Data Object Reference=@n4) -> (Input Name=None))
   > Branch(Condition=Contains(SearchIn=Replace(SourceString=@n6, From=.), Substring="To String (Text)"(InText=@n5)))
   n7 = Get Data Object() -> (Data Object:Meta Sound Preset Migrator Input Data Object Reference)
   > Add(TargetMap=Get Filtered Data(), Key=Get Input Name(self=@n7) -> (Input Name=None), Value=@n7)
  ]}
 Get Destination Instruction Text Visibility {nodes [
   Get Destination Instruction Text Visibility() {is=title K2Node_FunctionEntry}
   n1 = Get Destination MetaSound()
   > Return Node!(
    execute
    ReturnValue:ESlateVisibility Enum=Select(Option 0:ESlateVisibility Enum=Visible, Option 1:ESlateVisibility Enum=Collapsed, Index=IsValid(Object=@n1)) -> (ReturnValue:ESlateVisibility Enum)
   )
  ]}
 Get Destination List Visibility {nodes [
   Get Destination List Visibility() {is=title K2Node_FunctionEntry}
   n1 = Get Destination MetaSound()
   > Return Node(ReturnValue=Select(Option 0=Collapsed, Option 1=Visible, Index=IsValid(Object=@n1)))
  ]}
 Get Linked Input Literals {nodes [
   Get Linked Input Literals() {is=title K2Node_FunctionEntry}
   > n1 = Values(TargetMap=Get Destination MetaSound Input Data Objects())
   > n2 = For Each Loop(Array=@n1)
   Return Node!(
    execute=@n2.Completed
    Linked Inputs:Map of Names to Meta Sound Literal Structures=Get Linked Input Literals() -> (Linked Input Literals:Map of Names to Meta Sound Literal Structures) {is=title K2Node_VariableGet}
   )
   > n3 = Get!(execute, self:Meta Sound Preset Migrator Input Data Object Reference=@n2) -> (then, else, Linked Input:Meta Sound Preset Migrator Input Data Object Reference)
   > Add(
    TargetMap:Map of Names to Meta Sound Literal Structures=Get Linked Input Literals()
    Key=Get Input Name(self=@n2) -> (Input Name=None)
    Value:"Meta Sound Literal Structure (by ref)"=Get Input Literal!(self:Meta Sound Preset Migrator Input Data Object Reference=@n3) -> (Input Literal:Meta Sound Literal Structure)
   )
  ]}
 Get Linked Inputs {nodes [
   Get Linked Inputs() {is=title K2Node_FunctionEntry}
   > n1 = Values(TargetMap=Get Destination MetaSound Input Data Objects())
   > n2 = For Each Loop(Array=@n1)
   Return Node(execute=@n2.Completed, Linked Inputs:Map of Names to Names=Get Local Linked Inputs() -> (Local Linked Inputs:Map of Names to Names))
   > n3 = Get(self=@n2)
   > Add(
    TargetMap:Map of Names to Names=Get Local Linked Inputs()
    Key=Get Input Name(self=@n2) -> (Input Name=None)
    Value:"Name (by ref)"=Get Input Name(self=@n3) -> (Input Name=None)
   )
  ]}
 Get Origin Instruction Text Visibility {nodes [
   Get Origin Instruction Text Visibility() {is=title K2Node_FunctionEntry}
   n1 = Get Origin MetaSound()
   > Return Node!(execute, ReturnValue:ESlateVisibility Enum=Select(Option 0=Visible, Option 1=Collapsed, Index=IsValid(Object=@n1)))
  ]}
 Get Origin List Visibility {nodes [
   Get Origin List Visibility() {is=title K2Node_FunctionEntry}
   n1 = Get Origin MetaSound()
   > Return Node(ReturnValue=Select(Option 0=Collapsed, Option 1=Visible, Index=IsValid(Object=@n1)))
  ]}
 Get Status Color {nodes [
   Get Status Color() {is=title K2Node_FunctionEntry}
   n1 = Has Saved Asset Pair!() -> (ReturnValue:bool)
   n2 = NOT Boolean(A=Are Both Assets Valid!() -> (ReturnValue:bool) {is=title K2Node_CallFunction})
   n3 = Select(
    Option 0:Linear Color Structure=(R=0.2384,G=0.2384,B=0.2384,A=0.8)
    Option 1:Linear Color Structure=(R=0.7305,G=0.5089,B=0.0578,A=0.8)
    Index=@n2
   ) -> (ReturnValue:Linear Color Structure)
   n4 = Select(
    Option 0=(R=0,G=0.2961,B=0.0823,A=0.8)
    Option 1=(R=0.7305,G=0.5089,B=0.0578,A=0.8)
    Index=OR Boolean(A=@n2, B=Get Is Save Dirty() -> (Is Save Dirty:bool))
   )
   > Return Node(ReturnValue:Slate Color Structure=Make Slate Color(SpecifiedColor=Select(Option 0=@n3, Option 1=@n4, Index=@n1)))
  ]}
 Get Status Icon {nodes [
   Get Status Icon() {is=title K2Node_FunctionEntry}
   n1 = Get Origin MetaSound()
   n2 = Is Preset(MetaSound Source=@n1)
   n3 = Get Destination MetaSound()
   n4 = OR Boolean!(
    A=Get Is Save Dirty()
    B=AND Boolean(A=IsValid(Object=@n1), B=NOT Boolean(A=@n2))
    C:bool=AND Boolean(A=IsValid(Object=@n3), B=Is Preset(MetaSound Source=@n3))
   ) -> (ReturnValue)
   > Return Node(ReturnValue:Slate Brush Structure=Make Slate Brush!(
     TintColor:Slate Color Structure=Get Status Color!() -> (ReturnValue:Slate Color Structure) {is=title K2Node_CallFunction}
     DrawAs:ESlateBrushDrawType Enum=Image
     Tiling:ESlateBrushTileType Enum=NoTile
     Mirroring:ESlateBrushMirrorType Enum=NoMirror
     ImageSize:Vector 2D Structure=MakeVector2D(X=20, Y=20)
     Margin:Margin Structure=(Left=0,Top=0,Right=0,Bottom=0)
     ResourceObject:Object Reference=Select(
      Option 0:Object Reference=/BRCosmetics/UI/Foundation/Textures/Icons/Checks/T-Icon-CheckedBox-52
      Option 1:Object Reference=/Game/UI/Foundation/Shell/Textures/T-Icon-Warning-32
      Index=@n4
     ) -> (ReturnValue:Object Reference)
     OutlineSettings:Slate Brush Outline Settings Structure=(CornerRadii=(X=0,Y=0,Z=0,W=0),Color=(SpecifiedColor=(R=0,G=0,B=0,A=0),ColorUseRule=UseColor_Specified),Width=0,RoundingType=HalfHeightRadius,bUseBrushTransparency=False)
    ) -> (SlateBrush:Slate Brush Structure) {is=title K2Node_MakeStruct})
  ]}
 Get Status Icon Visibility {nodes [
   Get Status Icon Visibility() {is=title K2Node_FunctionEntry}
   n1 = AND Boolean(A=Get Is Save Dirty(), B=Has Saved Asset Pair())
   n2 = Get Origin MetaSound()
   n3 = AND Boolean(A=IsValid(Object=@n2), B=NOT Boolean(A=Is Preset(MetaSound Source=@n2)))
   n4 = Get Destination MetaSound()
   n5 = AND Boolean(A=IsValid(Object=@n4), B=Is Preset(MetaSound Source=@n4))
   n6 = AND Boolean(A=Are Both Assets Valid(), B=Has Saved Asset Pair())
   > Return Node(ReturnValue:ESlateVisibility Enum=Select(
     Option 0:ESlateVisibility Enum=Collapsed
     Option 1:ESlateVisibility Enum=Visible
     Index=OR Boolean!(A=@n3, B=@n5, C=@n6, D:bool=@n1) -> (ReturnValue)
    ) -> (ReturnValue:ESlateVisibility Enum))
  ]}
 Get Status Text {nodes [
   Get Status Text() {is=title K2Node_FunctionEntry}
   n1 = Get Is Save Dirty()
   n2 = SelectText(
    A=SelectText(
     A=SelectText(A="LOCTEXT(\"Unsaved Changes\")", B="LOCTEXT(\"Saved\")", bPickA=@n1)
     B="LOCTEXT(\"No Save Found\")"
     bPickA=Has Saved Asset Pair()
    )
    bPickA=Are Both Assets Valid()
   )
   n3 = Get Destination MetaSound()
   n4 = AND Boolean(A=IsValid(Object=@n3), B=Is Preset(MetaSound Source=@n3))
   n5 = Get Origin MetaSound()
   n6 = NOT Boolean(A=Is Preset(MetaSound Source=@n5))
   n7 = IsValid(Object=@n5)
   > Return Node(ReturnValue:Text=SelectText(
     A="LOCTEXT(\"Original MetaSound is not a preset.\")"
     B=SelectText(A="LOCTEXT(\"New parent MetaSound should not be a preset.\")", B=@n2, bPickA=@n4)
     bPickA=AND Boolean(A=@n7, B=@n6)
    ))
  ]}
 Get Style {nodes [
   Get Style() {is=title K2Node_FunctionEntry}
   > n1 = GetClassDefaultObject!(execute, Class:Object Class Reference=Get Style Class()) -> (then, ReturnValue:Object Reference)
   > n2 = Cast To MetaSoundPresetMigrator_Style!(execute, Object=@n1) -> (then, CastFailed, AsMeta Sound Preset Migrator Style:Meta Sound Preset Migrator Style Object Reference) {is=title K2Node_DynamicCast}
   > Return Node!(execute, Style:Meta Sound Preset Migrator Style Object Reference=@n2)
  ]}
 Has Saved Asset Pair {nodes [
   Has Saved Asset Pair() {is=title K2Node_FunctionEntry}
   > n1 = Get!(execute) -> (then, else, Saved Links Data Asset:Meta Sound Preset Migrator Saved Meta Sound Links Object Reference)
   n2 = Get Origin MetaSound()
   > Return Node!(
    execute
    ReturnValue:bool=Get Saved Input Links!(
     self:Meta Sound Preset Migrator Saved Meta Sound Links Object Reference=@n1
     Origin Parent MetaSound Source:Meta Sound Source Object Reference=Get Preset Parent!(MetaSound Source Preset:Meta Sound Source Object Reference=@n2, __WorldContext:Object Reference) -> (Parent MetaSound Source:Meta Sound Source Object Reference) {is=title K2Node_CallFunction}
     Destination MetaSound Source:Meta Sound Source Object Reference=Get Destination MetaSound()
    ) -> (Success:bool, Linked Inputs:Map of Names to Names) {is=title K2Node_CallFunction}
   )
  ]}
 Initialize Asset Name {nodes [
   Initialize Asset Name() {is=title K2Node_FunctionEntry}
   > n1 = Get!(execute) -> (then, else, Origin MetaSound:Meta Sound Source Object Reference)
   > n2 = GetAssetTools()
   n3 = Split(InPath=Get Object Path String(Object=@n1))
   > n4 = CreateUniqueAssetName!(
    execute
    self:Asset Tools Interface=@n2
    InBasePackageName:String=Append!(A=@n3, B=/, C:String=@n3.FilenamePart) -> (ReturnValue)
    InSuffix:String
   ) -> (then, OutPackageName:String, OutAssetName:String)
   > n5 = Set Asset Name(Asset Name=@n4.OutAssetName)
   n6 = Get AssetNameTextBox() -> (AssetNameTextBox:Editor Utility Editable Text Box Object Reference)
   > Set Text(Text="To Text (String)"(InString=@n5), self:Text Box Object Reference=@n6)
   Set Asset Name(execute=@n1.else)
   > Set Text(self=@n6)
  ]}
 Initialize Links {nodes [
   Initialize Links() {is=title K2Node_FunctionEntry}
   > Clear All Links()
   > Branch(Condition=Are Both Assets Valid())
   > n1 = Load Saved Links()
   > Branch(Condition=NOT Boolean(A=@n1))
   > n2 = Link Matching Input Names()
   > n3 = AddEditorNotification(EditorNotificationSettings=Select(
     Option 0:"Editor Notification Settings Structure (by ref)"=MakeEditorNotificationSettings(
      Text="LOCTEXT(\"No inputs with matching names were found.\")"
      FadeInDuration=0.5
      FadeOutDuration=2
      ExpireDuration=1
      bUseThrobber=true
      bUseSuccessFailIcons=true
      bFireAndForget=true
      bAllowThrottleWhenFrameRateIsLow=true
     )
     Option 1:"Editor Notification Settings Structure (by ref)"=MakeEditorNotificationSettings(
      Text="LOCTEXT(\"Linked all inputs with matching names.\")"
      FadeInDuration=0.5
      FadeOutDuration=2
      ExpireDuration=3
      bUseThrobber=true
      bUseSuccessFailIcons=true
      bFireAndForget=true
      bAllowThrottleWhenFrameRateIsLow=true
     )
     Index=@n2
    ) -> (ReturnValue:"Editor Notification Settings Structure (by ref)"))
   > SetCompletionState(
    NotificationHandle=@n3
    State=Select(Option 0:EEditorNotificationCompletionState Enum=Fail, Option 1:EEditorNotificationCompletionState Enum=Success, Index=@n2) -> (ReturnValue:EEditorNotificationCompletionState Enum)
   )
  ]}
 Is Destination Input List Enabled {nodes [
   Is Destination Input List Enabled() {is=title K2Node_FunctionEntry}
   n1 = Get Destination MetaSound()
   n2 = Is Preset(MetaSound Source=@n1)
   > Return Node(ReturnValue=AND Boolean(A=IsValid(Object=@n1), B=NOT Boolean(A=@n2)))
  ]}
 Is Origin Input List Enabled {nodes [
   Is Origin Input List Enabled() {is=title K2Node_FunctionEntry}
   n1 = Get Origin MetaSound()
   > Return Node(ReturnValue=AND Boolean(A=IsValid(Object=@n1), B=Is Preset(MetaSound Source=@n1)))
  ]}
 Is Reload Button Enabled {nodes [
   Is Reload Button Enabled() {is=title K2Node_FunctionEntry}
   > Return Node(ReturnValue=AND Boolean(A=Has Saved Asset Pair() {is=title K2Node_CallFunction}, B=Get Is Save Dirty()))
  ]}
 Is Save Button Enabled {nodes [
   Is Save Button Enabled() {is=title K2Node_FunctionEntry}
   > Return Node(ReturnValue=AND Boolean(A=Are Both Assets Valid(), B=Get Is Save Dirty()))
  ]}
 Link Data {nodes [
   Link Data!() -> (then, Origin Input Name:Name, Destination Input Name:Name) {is=title K2Node_FunctionEntry}
   > n1 = Sequence()
   n2 = Find(
    TargetMap:Map of Names to Meta Sound Preset Migrator Input Data Object References=Get Origin MetaSound Input Data Objects()
    Key=Get Origin Input Name() -> (Origin Input Name:Name=None)
   ) -> (Value:Meta Sound Preset Migrator Input Data Object Reference)
   > n3 = Branch(Condition=@n2.ReturnValue)
   > Set Origin Data(Origin Data:Meta Sound Preset Migrator Input Data Object Reference=@n2) -> (Output_Get:Meta Sound Preset Migrator Input Data Object Reference)
   Return Node!(execute=@n3.else, Success:bool=false)
   n4 = Sequence(execute=@n1.then_1)
   n5 = Find(TargetMap=Get Destination MetaSound Input Data Objects(), Key=Get Destination Input Name() -> (Destination Input Name:Name=None))
   > n6 = Branch(Condition=@n5.ReturnValue)
   > Set Destination Data(Destination Data:Meta Sound Preset Migrator Input Data Object Reference=@n5) -> (Output_Get:Meta Sound Preset Migrator Input Data Object Reference)
   Return Node(execute=@n6.else, Success=false)
   Sequence(execute=@n4.then_1)
   > n7 = Get!(execute) -> (then, else, Origin Data:Meta Sound Preset Migrator Input Data Object Reference)
   > n8 = Get!(execute) -> (then, else, Destination Data:Meta Sound Preset Migrator Input Data Object Reference)
   n9 = Get Linked Input(self=@n7)
   n10 = Get Data Type!(self:Meta Sound Preset Migrator Input Data Object Reference=@n8) -> (Data Type=None)
   n11 = Get Data Type(self=@n7) -> (Data Type=None)
   > n12 = Branch(Condition=AND Boolean(
     A="Equal (Name)"(A=@n10, B=@n11)
     B="Not Equal (Object)"(A:Object Reference=@n8, B:Object Reference=@n9) -> (ReturnValue:bool=~) {is=title K2Node_PromotableOperator}
    ))
   > Set Linked Input(
    Linked Input=Get Destination Data() -> (Destination Data:Meta Sound Preset Migrator Input Data Object Reference)
    self=Get Origin Data() -> (Origin Data:Meta Sound Preset Migrator Input Data Object Reference)
   )
   > Set Linked Input(Linked Input=Get Origin Data(), self=Get Destination Data())
   > Return Node(Success=true)
   Return Node(execute=@n12.else, Success=false)
   Return Node(execute=[@n8.else, @n7.else], Success=false)
  ]}
 Link Matching Input Names {nodes [
   Link Matching Input Names() {is=title K2Node_FunctionEntry}
   > n1 = Keys(TargetMap:Map of Names to Meta Sound Preset Migrator Input Data Object References=Get Destination MetaSound Input Data Objects())
   > n2 = For Each Loop(Array:Array of Names=@n1) -> (Array Element:Name)
   Return Node(execute=@n2.Completed, Success=Get Found At Least One Link() -> (Found At Least One Link:bool))
   > n3 = Set Input Name(Input Name=@n2) -> (Output_Get=None)
   > Branch(Condition=Find(TargetMap=Get Destination MetaSound Input Data Objects(), Key=@n3).ReturnValue)
   n4 = Get Input Name!() -> (Input Name=None)
   > n5 = Link Data!(execute, self, Origin Input Name:Name=@n4, Destination Input Name:Name=@n4) -> (then, Success:bool) {is=title K2Node_CallFunction}
   > n6 = Set Is Save Dirty(Is Save Dirty=@n5)
   > Set Found At Least One Link(Found At Least One Link:bool=@n6) -> (Output_Get:bool)
  ]}
 Load Saved Links {nodes [
   Load Saved Links() {is=title K2Node_FunctionEntry}
   > n1 = Get!(execute) -> (then, else, Saved Links Data Asset:Meta Sound Preset Migrator Saved Meta Sound Links Object Reference)
   n2 = Get Saved Input Links(
    self=@n1
    Origin Parent MetaSound Source=Get Preset Parent(MetaSound Source Preset=Get Origin MetaSound())
    Destination MetaSound Source=Get Destination MetaSound()
   )
   > n3 = Branch(Condition=@n2)
   > n4 = Keys(TargetMap:Map of Names to Names=@n2.Linked Inputs)
   > n5 = For Each Loop(Array=@n4)
   RegenerateAllEntries(execute=@n5.Completed, self:List View Base Object Reference=[Get DestinationMetaSoundInputsList(), Get OriginMetaSoundInputsList()])
   > Set Is Save Dirty(Is Save Dirty=false)
   n6 = GetDisplayName(Object=Get Saved Links Data Asset() -> (Saved Links Data Asset:Meta Sound Preset Migrator Saved Meta Sound Links Object Reference))
   > n7 = AddEditorNotification(EditorNotificationSettings=MakeEditorNotificationSettings(
     Text="LOCTEXT(\"Successfully loaded saved links!\")"
     SubText=Format Text!(Format="LOCTEXT(\"Recalled saved input links from data asset \\'{dataasset}\\'.\")", dataasset:String=@n6) -> (Result)
     FadeInDuration=0.5
     FadeOutDuration=2
     ExpireDuration=5
     bUseThrobber=true
     bUseSuccessFailIcons=true
     bFireAndForget=true
     bAllowThrottleWhenFrameRateIsLow=true
    ))
   > SetCompletionState(NotificationHandle=@n7, State=Success)
   > Return Node(Success=true)
   Link Data(
    execute=@n5.LoopBody
    Origin Input Name=Find(TargetMap:Map of Names to Names=@n2.Linked Inputs, Key=@n5) -> (Value:Name)
    Destination Input Name=@n5
   )
   Return Node(execute=@n3.else, Success=false)
  ]}
 Migrate Source Properties {nodes [
   Migrate Source Properties() {is=title K2Node_FunctionEntry}
   > n1 = Migrate Source Properties!(
    execute
    Asset to Migrate From:Sound Base Object Reference=Get Origin MetaSound()
    Asset to Migrate To:Sound Base Object Reference=Get Destination MetaSound()
    __WorldContext:Object Reference
   ) -> (then, Success, Results:Array of Fort Audio Tools Result Structures) {is=title K2Node_CallFunction}
   > n2 = Branch(Condition=@n1)
   > Return Node(Success=true)
   Print Results!(
    execute=@n2.else
    Results:Array of Fort Audio Tools Result Structures=@n1.Results
    Exclude Successful Results:bool=true
    __WorldContext:Object Reference
   ) -> (then)
   > Return Node(Success=false)
  ]}
 Route Value {nodes [
   n1 = Inputs!() -> (Origin:Exec, Origin Value:Wildcard, Destination:Exec, Destination Value:Wildcard)
   n2 = Local Boolean!() -> (Variable:bool=~) {is=title K2Node_TemporaryVariable}
   n3 = Assign!(execute:Exec=@n1.Destination, Variable:bool=@n2, Value:bool=true) -> (then:Exec) {is=title K2Node_AssignmentStatement}
   Outputs!(
    then:Exec=[@n4.then, @n3.then]
    Value:Wildcard=Select(Option 0:Wildcard=@n1, Option 1:Wildcard=@n1.Destination Value, Index=@n2) -> (ReturnValue:Wildcard)
   )
   n4 = Assign(execute=@n1.Origin, Variable=@n2)
  ]}
 Save Linked Inputs {nodes [
   Save Linked Inputs() {is=title K2Node_FunctionEntry}
   > n1 = Get()
   n2 = Get Origin MetaSound()
   > n3 = Branch(Condition=Contains Paired MetaSounds!(
     self:Meta Sound Preset Migrator Saved Meta Sound Links Object Reference=@n1
     Origin MetaSound Source:Meta Sound Source Object Reference=Get Preset Parent(MetaSound Source Preset=@n2)
     Destination MetaSound Source:Meta Sound Source Object Reference=Get Destination MetaSound()
    ) -> (ReturnValue:bool))
   > n4 = Show Message Dialog(
    Title=Format Text!(Format="LOCTEXT(\"Overwrite Saved Links\")") -> (Result)
    Message=Format Text(Format="LOCTEXT(\"Saving will overwrite the existing saved links between these parent assets. \\r\\n\\r\\nAre you sure you would like to do this?\")")
    MessageType=YesNo
    DefaultValue=Yes
    MessageCategory=Warning
   )
   > n5 = Switch on EAppReturnType(Selection=@n4)
   Save Origin MetaSound Input Links!(
    execute=[@n5.Yes, @n3.else]
    self:Meta Sound Preset Migrator Saved Meta Sound Links Object Reference=Get Saved Links Data Asset()
    Destination MetaSound:Meta Sound Source Object Reference=Get Destination MetaSound()
    Origin Parent MetaSound:Meta Sound Source Object Reference=Get Preset Parent(MetaSound Source Preset=Get Origin MetaSound())
    Linked Inputs:Map of Names to Names=Get Linked Inputs!() -> (Linked Inputs:Map of Names to Names) {is=title K2Node_CallFunction}
   ) -> (then)
   > n6 = CheckoutLoadedAsset!(execute, AssetToCheckout:Object Reference=Get Saved Links Data Asset()) -> (then, ReturnValue:bool)
   > n7 = Branch(Condition=@n6)
   > n8 = SaveLoadedAsset(AssetToSave=Get Saved Links Data Asset(), bOnlyIfIsDirty=false)
   > Set Is Save Dirty(Is Save Dirty=NOT Boolean(A=@n8))
   > n9 = AddEditorNotification(EditorNotificationSettings=Select(
     Option 0:"Editor Notification Settings Structure (by ref)"=MakeEditorNotificationSettings(
      Text="LOCTEXT(\"Failed to save linked inputs.\")"
      SubText="LOCTEXT(\"Save operation failed.\")"
      FadeInDuration=0.5
      FadeOutDuration=2
      ExpireDuration=15
      bUseThrobber=true
      bUseSuccessFailIcons=true
      bFireAndForget=true
      bAllowThrottleWhenFrameRateIsLow=true
     )
     Option 1:"Editor Notification Settings Structure (by ref)"=MakeEditorNotificationSettings(
      Text="LOCTEXT(\"Successfully saved input links.\")"
      FadeInDuration=0.5
      FadeOutDuration=2
      ExpireDuration=3
      bUseThrobber=true
      bUseSuccessFailIcons=true
      bFireAndForget=true
      bAllowThrottleWhenFrameRateIsLow=true
     )
     Index=@n8
    ) -> (ReturnValue:"Editor Notification Settings Structure (by ref)"))
   > SetCompletionState(
    NotificationHandle=@n9
    State=Select(Option 0:EEditorNotificationCompletionState Enum=Fail, Option 1:EEditorNotificationCompletionState Enum=Success, Index=@n8) -> (ReturnValue:EEditorNotificationCompletionState Enum)
   )
   n10 = AddEditorNotification(
    execute=@n7.else
    EditorNotificationSettings=MakeEditorNotificationSettings(
     Text="LOCTEXT(\"Failed to save linked inputs.\")"
     SubText="LOCTEXT(\"Failed to check out Saved Links data asset from source control.\")"
     FadeInDuration=0.5
     FadeOutDuration=2
     ExpireDuration=15
     bUseThrobber=true
     bUseSuccessFailIcons=true
     bFireAndForget=true
     bAllowThrottleWhenFrameRateIsLow=true
    )
   )
   > SetCompletionState(NotificationHandle=@n10, State=Fail)
   n11 = AddEditorNotification(
    execute=@n1.else
    EditorNotificationSettings=MakeEditorNotificationSettings(
     Text="LOCTEXT(\"Failed to save linked inputs.\")"
     SubText="LOCTEXT(\"Saved Links data asset was null.\")"
     FadeInDuration=0.5
     FadeOutDuration=2
     ExpireDuration=15
     bUseThrobber=true
     bUseSuccessFailIcons=true
     bFireAndForget=true
     bAllowThrottleWhenFrameRateIsLow=true
    )
   )
   > SetCompletionState(NotificationHandle=@n11, State=Fail)
  ]}
 Sort Input List {nodes [
   n1 = Sort Input List!() -> (then, Input Objects:Map of Names to Meta Sound Preset Migrator Input Data Object References) {is=title K2Node_FunctionEntry}
   > n2 = Keys(TargetMap:Map of Names to Meta Sound Preset Migrator Input Data Object References=@n1)
   > n3 = Set Keys(Keys:Array of Names=@n2) -> (Output_Get:Array of Names)
   > SortNameArray!(execute, TargetArray:Array of Names=@n3, bStableSort:bool=false, bLexicalSort:bool=true, SortOrder:EArraySortOrder Enum=Ascending) -> (then)
   > n4 = For Each Loop(Array=Get Keys() -> (Keys:Array of Names))
   Return Node!(
    execute=@n4.Completed
    Alphabetical List:Array of Meta Sound Preset Migrator Input Data Object References=Get Sorted List() -> (Sorted List:Array of Meta Sound Preset Migrator Input Data Object References)
   )
   n5 = Find(
    TargetMap:Map of Names to Meta Sound Preset Migrator Input Data Object References=Get Input Objects() -> (Input Objects:Map of Names to Meta Sound Preset Migrator Input Data Object References)
    Key=@n4
   ) -> (Value:Meta Sound Preset Migrator Input Data Object Reference)
   > n6 = Add(
    TargetArray:Array of Meta Sound Preset Migrator Input Data Object References=Get Sorted List()
    NewItem:"Meta Sound Preset Migrator Input Data Object Reference (by ref)"=@n5
   ) {is=title K2Node_CallArrayFunction}
   > Set Row Index!(execute, Row Index:int=@n6, self:Meta Sound Preset Migrator Input Data Object Reference=@n5) -> (then, Output_Get:int=0)
  ]}
 Unlink Data {nodes [
   n1 = Unlink Data!() -> (then, Data to Unlink:Meta Sound Preset Migrator Input Data Object Reference) {is=title K2Node_FunctionEntry}
   > n2 = Get!(execute, self:Meta Sound Preset Migrator Input Data Object Reference=@n1) -> (then, else, Linked Input:Meta Sound Preset Migrator Input Data Object Reference)
   > Set Linked Input(self=@n2)
   > Set Linked Input(self=Get Data to Unlink() -> (Data to Unlink:Meta Sound Preset Migrator Input Data Object Reference))
   > Set Is Save Dirty(Is Save Dirty=true)
  ]}
]
variables [
 Asset Name {type=string, default=_Migrated}
 Clicked Data {type=MetaSoundPresetMigrator_InputData_C, default=null}
 Destination MetaSound {type=MetaSoundSource, default=null, class=~, properties=~}
 Destination MetaSound Input Data Objects {type="map<MetaSoundPresetMigrator_InputData_C>", default {}}
 Destination MetaSound Input Literals {type="map<MetasoundFrontendLiteral>", default {}}
 Is Save Dirty {type=boolean}
 Origin MetaSound {type=MetaSoundSource, default=null, class=~, properties=~}
 Origin MetaSound Input Data Objects {type="map<MetaSoundPresetMigrator_InputData_C>", default {}}
 Origin MetaSound Input Literals {type="map<MetasoundFrontendLiteral>", default {}}
 Saved Links Data Asset {
  type=MetaSoundPresetMigrator_SavedMetaSoundLinks_C
  default=/Game/Sounds/Fort_Audio_Blueprints/Utilities/EditorUtilityWidgets/Metasounds/Tools/MetaSoundPresetMigratorData/SavedMetaSoundLinks
 }
 Selected Destination Widget Cache {type=MetaSoundPresetMigrator_InputData_C, default=null}
 Selected Origin Widget Cache {type=MetaSoundPresetMigrator_InputData_C, default=null}
 Style Class {
  type=Class
  default=/Game/Sounds/Fort_Audio_Blueprints/Utilities/EditorUtilityWidgets/Metasounds/Tools/MetaSoundPresetMigratorData/MetaSoundPresetMigrator_Style.MetaSoundPresetMigrator_Style_C
 }
 Text Color {type=LinearColor, default=<rgba 0.5906 0.5906 0.5906 1>}
]
widgets {
 info {rootWidgetClass=&VerticalBox, widgetCount=78}
 widgets [
  VerticalBox_32 {widgetClassPath=&VerticalBox, widget=VerticalBox_32}
  Background {
   widgetClassPath=&Border
   parent=VerticalBox_32
   widget=Background
   properties {brushColor=<rgba 0.0176 0.0176 0.0176 1>, padding=<ltrb 0 0 0 0>}
   slot {class=VerticalBoxSlot, properties {size {sizeRule=Fill, value=1}}}
  }
  Overlay_3 {widgetClassPath=&Overlay, parent=Background, widget=Overlay_3, slot {class=BorderSlot, properties {padding=<ltrb 0 0 0 0>}}}
  Content {
   widgetClassPath=&VerticalBox
   bIsVariable=true
   parent=Overlay_3
   widget=Content
   slot {class=OverlaySlot, properties {horizontalAlignment=HAlign_Fill, verticalAlignment=VAlign_Fill}}
  }
  Header {widgetClassPath=&HorizontalBox, parent=Content, widget=Header, slot {class=VerticalBoxSlot}}
  ToolNameHeaderText {
   widgetClassPath=&TextBlock
   bIsVariable=true
   parent=Header
   widget=ToolNameHeaderText
   properties {
    text=MetaSound Preset Migrator
    colorAndOpacity {colorUseRule=UseColor_Specified, specifiedColor=<rgba 0.5906 0.5906 0.5906 1>}
    font {
     bForceMonospaced=false
     bMaterialIsStencil=false
     fontMaterial=null
     fontObject=&Roboto
     letterSpacing=0
     monospacedWidth=1
     size=15
     skewAmount=0
     typefaceFontName=Bold
     outlineSettings {
      bApplyOutlineToDropShadows=false
      bMiteredCorners=false
      bSeparateFillAlpha=false
      outlineMaterial=null
      outlineSize=0
      outlineColor=<rgba 0 0 0 1>
     }
    }
   }
   slot {class=HorizontalBoxSlot, properties {verticalAlignment=VAlign_Center, padding=<ltrb 16 8 8 8>}}
  }
  WrapBox_0 {
   widgetClassPath=&WrapBox
   parent=Header
   widget=WrapBox_0
   properties {innerSlotPadding=<8 8>}
   slot {class=HorizontalBoxSlot, properties {padding=<ltrb 4 8 4 8>, size {sizeRule=Fill, value=1}}}
  }
  ReloadSavedLinksButton {
   widgetClassPath=&EditorUtilityButton
   parent=WrapBox_0
   widget=ReloadSavedLinksButton
   properties {
    toolTipText="Reloads the saved input links for the current asset pair, overwriting all unsaved changes."
    widgetStyle {
     clickedSlateSound {resourceObject=null}
     disabled {
      bIsDynamicallyLoaded=false
      drawAs=RoundedBox
      imageType=NoImage
      mirroring=NoMirror
      resourceName=None
      resourceObject=null
      tiling=NoTile
      imageSize=<0 0>
      margin=<ltrb 0 0 0 0>
      outlineSettings {
       bUseBrushTransparency=false
       roundingType=FixedRadius
       width=1
       color {colorUseRule=UseColor_ColorTable, specifiedColor=<rgba 0 0 0 0>}
       cornerRadii=<4 4 4 4>
      }
      tintColor {colorUseRule=UseColor_Specified, specifiedColor=<rgba 0.0395 0.0395 0.0395 1>}
      uVRegion {bIsValid=false, max=<0 0>, min=<0 0>}
     }
     disabledForeground {colorUseRule=UseColor_ColorTable, specifiedColor=<rgba 0 0 0 0>}
     hovered {
      bIsDynamicallyLoaded=false
      drawAs=RoundedBox
      imageType=NoImage
      mirroring=NoMirror
      resourceName=None
      resourceObject=null
      tiling=NoTile
      imageSize=<32 32>
      margin=<ltrb 0 0 0 0>
      outlineSettings {
       bUseBrushTransparency=false
       roundingType=FixedRadius
       width=1
       color {colorUseRule=UseColor_ColorTable, specifiedColor=<rgba 0 0 0 0>}
       cornerRadii=<4 4 4 4>
      }
      tintColor {colorUseRule=UseColor_Specified, specifiedColor=<rgba 0.0953 0.0953 0.0953 1>}
      uVRegion {bIsValid=false, max=<0 0>, min=<0 0>}
     }
     hoveredForeground {colorUseRule=UseColor_ColorTable, specifiedColor=<rgba 0 0 0 0>}
     hoveredSlateSound {resourceObject=null}
     normal {
      bIsDynamicallyLoaded=false
      drawAs=RoundedBox
      imageType=NoImage
      mirroring=NoMirror
      resourceName=None
      resourceObject=null
      tiling=NoTile
      imageSize=<32 32>
      margin=<ltrb 0 0 0 0>
      outlineSettings {
       bUseBrushTransparency=false
       roundingType=FixedRadius
       width=1
       color {colorUseRule=UseColor_ColorTable, specifiedColor=<rgba 0 0 0 0>}
       cornerRadii=<4 4 4 4>
      }
      tintColor {colorUseRule=UseColor_Specified, specifiedColor=<rgba 0.0395 0.0395 0.0395 1>}
      uVRegion {bIsValid=false, max=<0 0>, min=<0 0>}
     }
     normalForeground {colorUseRule=UseColor_ColorTable, specifiedColor=<rgba 0 0 0 0>}
     normalPadding=<ltrb 4 2 4 2>
     pressed {
      bIsDynamicallyLoaded=false
      drawAs=RoundedBox
      imageType=NoImage
      mirroring=NoMirror
      resourceName=None
      resourceObject=null
      tiling=NoTile
      imageSize=<32 32>
      margin=<ltrb 0 0 0 0>
      outlineSettings {
       bUseBrushTransparency=false
       roundingType=FixedRadius
       width=1
       color {colorUseRule=UseColor_ColorTable, specifiedColor=<rgba 0 0 0 0>}
       cornerRadii=<4 4 4 4>
      }
      tintColor {colorUseRule=UseColor_Specified, specifiedColor=<rgba 0.0284 0.0284 0.0284 1>}
      uVRegion {bIsValid=false, max=<0 0>, min=<0 0>}
     }
     pressedForeground {colorUseRule=UseColor_ColorTable, specifiedColor=<rgba 0 0 0 0>}
     pressedPadding=<ltrb 4 2.5 4 0.5>
     pressedSlateSound {resourceObject=null}
    }
   }
   slot {class=WrapBoxSlot, properties {verticalAlignment=VAlign_Center}}
  }
  HorizontalBox_4 {widgetClassPath=&HorizontalBox, parent=ReloadSavedLinksButton, widget=HorizontalBox_4, slot {class=ButtonSlot}}
  ReloadIcon {
   widgetClassPath=&Image
   parent=HorizontalBox_4
   widget=ReloadIcon
   properties {brush {
     bIsDynamicallyLoaded=false
     drawAs=Image
     imageType=FullColor
     mirroring=NoMirror
     resourceName=None
     resourceObject=/Game/Athena/UI/InGame/Creative/PropertyEditor/Icon_Picker/Textures_V2/64x64/T_UI_IconLibrary_Reset_64
     tiling=NoTile
     imageSize=<20 20>
     margin=<ltrb 0 0 0 0>
     outlineSettings {
      bUseBrushTransparency=false
      roundingType=HalfHeightRadius
      width=0
      color {colorUseRule=UseColor_Specified, specifiedColor=<rgba 0 0 0 0>}
      cornerRadii=<0 0 0 0>
     }
     tintColor {colorUseRule=UseColor_Specified, specifiedColor=<rgba 0.5885 0.5885 0.5885 1>}
     uVRegion {bIsValid=false, max=<0 0>, min=<0 0>}
    }}
   slot {class=HorizontalBoxSlot, properties {verticalAlignment=VAlign_Center}}
  }
  SaveLinksButton {
   widgetClassPath=&EditorUtilityButton
   parent=WrapBox_0
   widget=SaveLinksButton
   properties {
    toolTipText=Save the current links to be later recalled when using the same asset pair.
    widgetStyle {
     clickedSlateSound {resourceObject=null}
     disabled {
      bIsDynamicallyLoaded=false
      drawAs=RoundedBox
      imageType=NoImage
      mirroring=NoMirror
      resourceName=None
      resourceObject=null
      tiling=NoTile
      imageSize=<0 0>
      margin=<ltrb 0 0 0 0>
      outlineSettings {
       bUseBrushTransparency=false
       roundingType=FixedRadius
       width=1
       color {colorUseRule=UseColor_ColorTable, specifiedColor=<rgba 0 0 0 0>}
       cornerRadii=<4 4 4 4>
      }
      tintColor {colorUseRule=UseColor_Specified, specifiedColor=<rgba 0.0395 0.0395 0.0395 1>}
      uVRegion {bIsValid=false, max=<0 0>, min=<0 0>}
     }
     disabledForeground {colorUseRule=UseColor_ColorTable, specifiedColor=<rgba 0 0 0 0>}
     hovered {
      bIsDynamicallyLoaded=false
      drawAs=RoundedBox
      imageType=NoImage
      mirroring=NoMirror
      resourceName=None
      resourceObject=null
      tiling=NoTile
      imageSize=<32 32>
      margin=<ltrb 0 0 0 0>
      outlineSettings {
       bUseBrushTransparency=false
       roundingType=FixedRadius
       width=1
       color {colorUseRule=UseColor_ColorTable, specifiedColor=<rgba 0 0 0 0>}
       cornerRadii=<4 4 4 4>
      }
      tintColor {colorUseRule=UseColor_Specified, specifiedColor=<rgba 0.0953 0.0953 0.0953 1>}
      uVRegion {bIsValid=false, max=<0 0>, min=<0 0>}
     }
     hoveredForeground {colorUseRule=UseColor_ColorTable, specifiedColor=<rgba 0 0 0 0>}
     hoveredSlateSound {resourceObject=null}
     normal {
      bIsDynamicallyLoaded=false
      drawAs=RoundedBox
      imageType=NoImage
      mirroring=NoMirror
      resourceName=None
      resourceObject=null
      tiling=NoTile
      imageSize=<32 32>
      margin=<ltrb 0 0 0 0>
      outlineSettings {
       bUseBrushTransparency=false
       roundingType=FixedRadius
       width=1
       color {colorUseRule=UseColor_ColorTable, specifiedColor=<rgba 0 0 0 0>}
       cornerRadii=<4 4 4 4>
      }
      tintColor {colorUseRule=UseColor_Specified, specifiedColor=<rgba 0.0395 0.0395 0.0395 1>}
      uVRegion {bIsValid=false, max=<0 0>, min=<0 0>}
     }
     normalForeground {colorUseRule=UseColor_ColorTable, specifiedColor=<rgba 0 0 0 0>}
     normalPadding=<ltrb 4 2 4 2>
     pressed {
      bIsDynamicallyLoaded=false
      drawAs=RoundedBox
      imageType=NoImage
      mirroring=NoMirror
      resourceName=None
      resourceObject=null
      tiling=NoTile
      imageSize=<32 32>
      margin=<ltrb 0 0 0 0>
      outlineSettings {
       bUseBrushTransparency=false
       roundingType=FixedRadius
       width=1
       color {colorUseRule=UseColor_ColorTable, specifiedColor=<rgba 0 0 0 0>}
       cornerRadii=<4 4 4 4>
      }
      tintColor {colorUseRule=UseColor_Specified, specifiedColor=<rgba 0.0284 0.0284 0.0284 1>}
      uVRegion {bIsValid=false, max=<0 0>, min=<0 0>}
     }
     pressedForeground {colorUseRule=UseColor_ColorTable, specifiedColor=<rgba 0 0 0 0>}
     pressedPadding=<ltrb 4 2.5 4 0.5>
     pressedSlateSound {resourceObject=null}
    }
   }
   slot {class=WrapBoxSlot, properties {verticalAlignment=VAlign_Center}}
  }
  HorizontalBox_1 {widgetClassPath=&HorizontalBox, parent=SaveLinksButton, widget=HorizontalBox_1, slot {class=ButtonSlot}}
  SaveIcon {
   widgetClassPath=&Image
   parent=HorizontalBox_1
   widget=SaveIcon
   properties {brush {
     bIsDynamicallyLoaded=false
     drawAs=Image
     imageType=FullColor
     mirroring=NoMirror
     resourceName=None
     resourceObject=/Game/Athena/UI/InGame/Creative/PropertyEditor/Icon_Picker/Textures_V2/64x64/T_UI_IconLibrary_FloppyDisc_64
     tiling=NoTile
     imageSize=<20 20>
     margin=<ltrb 0 0 0 0>
     outlineSettings {
      bUseBrushTransparency=false
      roundingType=HalfHeightRadius
      width=0
      color {colorUseRule=UseColor_Specified, specifiedColor=<rgba 0 0 0 0>}
      cornerRadii=<0 0 0 0>
     }
     tintColor {colorUseRule=UseColor_Specified, specifiedColor=<rgba 0.5885 0.5885 0.5885 1>}
     uVRegion {bIsValid=false, max=<0 0>, min=<0 0>}
    }}
   slot {class=HorizontalBoxSlot, properties {verticalAlignment=VAlign_Center}}
  }
  SaveStatusBox {
   widgetClassPath=&HorizontalBox
   bIsVariable=true
   parent=WrapBox_0
   widget=SaveStatusBox
   properties {visibility=Visible}
   slot {class=WrapBoxSlot}
  }
  SaveStatusIcon {
   widgetClassPath=&Image
   bIsVariable=true
   parent=SaveStatusBox
   widget=SaveStatusIcon
   properties {brush {
     bIsDynamicallyLoaded=false
     drawAs=Image
     imageType=FullColor
     mirroring=NoMirror
     resourceName=None
     resourceObject=null
     tiling=NoTile
     imageSize=<20 20>
     margin=<ltrb 0 0 0 0>
     outlineSettings {
      bUseBrushTransparency=false
      roundingType=HalfHeightRadius
      width=0
      color {colorUseRule=UseColor_Specified, specifiedColor=<rgba 0 0 0 0>}
      cornerRadii=<0 0 0 0>
     }
     tintColor {colorUseRule=UseColor_Specified, specifiedColor=<rgba 0.043 0.2292 0.0578 1>}
     uVRegion {bIsValid=false, max=<0 0>, min=<0 0>}
    }}
   slot {class=HorizontalBoxSlot, properties {horizontalAlignment=HAlign_Center, verticalAlignment=VAlign_Center}}
  }
  TextBlock_4 {
   widgetClassPath=&TextBlock
   parent=SaveStatusBox
   widget=TextBlock_4
   properties {
    text=Saved
    colorAndOpacity {colorUseRule=UseColor_Specified, specifiedColor=<rgba 0.043 0.2292 0.0578 1>}
    font {
     bForceMonospaced=false
     bMaterialIsStencil=false
     fontMaterial=null
     fontObject=&Roboto
     letterSpacing=0
     monospacedWidth=1
     size=12
     skewAmount=0
     typefaceFontName=Bold
     outlineSettings {
      bApplyOutlineToDropShadows=false
      bMiteredCorners=false
      bSeparateFillAlpha=false
      outlineMaterial=null
      outlineSize=0
      outlineColor=<rgba 0 0 0 1>
     }
    }
   }
   slot {class=HorizontalBoxSlot, properties {verticalAlignment=VAlign_Center, padding=<ltrb 4 1 0 0>}}
  }
  WrapBox_1 {
   widgetClassPath=&WrapBox
   parent=Header
   widget=WrapBox_1
   properties {innerSlotPadding=<4 4>}
   slot {class=HorizontalBoxSlot, properties {padding=<ltrb 4 8 8 8>}}
  }
  AdvancedSettingsButton {
   widgetClassPath=&EditorUtilityButton
   parent=WrapBox_1
   widget=AdvancedSettingsButton
   properties {
    toolTipText=Advanced Settings
    visibility=Hidden
    widgetStyle {
     clickedSlateSound {resourceObject=null}
     disabled {
      bIsDynamicallyLoaded=false
      drawAs=RoundedBox
      imageType=NoImage
      mirroring=NoMirror
      resourceName=None
      resourceObject=null
      tiling=NoTile
      imageSize=<0 0>
      margin=<ltrb 0 0 0 0>
      outlineSettings {
       bUseBrushTransparency=false
       roundingType=FixedRadius
       width=1
       color {colorUseRule=UseColor_ColorTable, specifiedColor=<rgba 0 0 0 0>}
       cornerRadii=<4 4 4 4>
      }
      tintColor {colorUseRule=UseColor_Specified, specifiedColor=<rgba 0.0395 0.0395 0.0395 1>}
      uVRegion {bIsValid=false, max=<0 0>, min=<0 0>}
     }
     disabledForeground {colorUseRule=UseColor_ColorTable, specifiedColor=<rgba 0 0 0 0>}
     hovered {
      bIsDynamicallyLoaded=false
      drawAs=RoundedBox
      imageType=NoImage
      mirroring=NoMirror
      resourceName=None
      resourceObject=null
      tiling=NoTile
      imageSize=<32 32>
      margin=<ltrb 0 0 0 0>
      outlineSettings {
       bUseBrushTransparency=false
       roundingType=FixedRadius
       width=1
       color {colorUseRule=UseColor_ColorTable, specifiedColor=<rgba 0 0 0 0>}
       cornerRadii=<4 4 4 4>
      }
      tintColor {colorUseRule=UseColor_Specified, specifiedColor=<rgba 0.0953 0.0953 0.0953 1>}
      uVRegion {bIsValid=false, max=<0 0>, min=<0 0>}
     }
     hoveredForeground {colorUseRule=UseColor_ColorTable, specifiedColor=<rgba 0 0 0 0>}
     hoveredSlateSound {resourceObject=null}
     normal {
      bIsDynamicallyLoaded=false
      drawAs=RoundedBox
      imageType=NoImage
      mirroring=NoMirror
      resourceName=None
      resourceObject=null
      tiling=NoTile
      imageSize=<32 32>
      margin=<ltrb 0 0 0 0>
      outlineSettings {
       bUseBrushTransparency=false
       roundingType=FixedRadius
       width=1
       color {colorUseRule=UseColor_ColorTable, specifiedColor=<rgba 0 0 0 0>}
       cornerRadii=<4 4 4 4>
      }
      tintColor {colorUseRule=UseColor_Specified, specifiedColor=<rgba 0.0395 0.0395 0.0395 1>}
      uVRegion {bIsValid=false, max=<0 0>, min=<0 0>}
     }
     normalForeground {colorUseRule=UseColor_ColorTable, specifiedColor=<rgba 0 0 0 0>}
     normalPadding=<ltrb 4 2 4 2>
     pressed {
      bIsDynamicallyLoaded=false
      drawAs=RoundedBox
      imageType=NoImage
      mirroring=NoMirror
      resourceName=None
      resourceObject=null
      tiling=NoTile
      imageSize=<32 32>
      margin=<ltrb 0 0 0 0>
      outlineSettings {
       bUseBrushTransparency=false
       roundingType=FixedRadius
       width=1
       color {colorUseRule=UseColor_ColorTable, specifiedColor=<rgba 0 0 0 0>}
       cornerRadii=<4 4 4 4>
      }
      tintColor {colorUseRule=UseColor_Specified, specifiedColor=<rgba 0.0284 0.0284 0.0284 1>}
      uVRegion {bIsValid=false, max=<0 0>, min=<0 0>}
     }
     pressedForeground {colorUseRule=UseColor_ColorTable, specifiedColor=<rgba 0 0 0 0>}
     pressedPadding=<ltrb 4 2.5 4 0.5>
     pressedSlateSound {resourceObject=null}
    }
   }
   slot {class=WrapBoxSlot, properties {verticalAlignment=VAlign_Center}}
  }
  HorizontalBox_5 {widgetClassPath=&HorizontalBox, parent=AdvancedSettingsButton, widget=HorizontalBox_5, slot {class=ButtonSlot}}
  SettingsIcon {
   widgetClassPath=&Image
   parent=HorizontalBox_5
   widget=SettingsIcon
   properties {brush {
     bIsDynamicallyLoaded=false
     drawAs=Image
     imageType=FullColor
     mirroring=NoMirror
     resourceName=None
     resourceObject=/Game/UI/Sidebar/Textures/T_UI_Icon_Sidebar_Settings
     tiling=NoTile
     imageSize=<20 20>
     margin=<ltrb 0 0 0 0>
     outlineSettings {
      bUseBrushTransparency=false
      roundingType=HalfHeightRadius
      width=0
      color {colorUseRule=UseColor_Specified, specifiedColor=<rgba 0 0 0 0>}
      cornerRadii=<0 0 0 0>
     }
     tintColor {colorUseRule=UseColor_Specified, specifiedColor=<rgba 0.5885 0.5885 0.5885 1>}
     uVRegion {bIsValid=false, max=<0 0>, min=<0 0>}
    }}
   slot {class=HorizontalBoxSlot, properties {verticalAlignment=VAlign_Center}}
  }
  UserGuideButton {
   widgetClassPath=&EditorUtilityButton
   parent=WrapBox_1
   widget=UserGuideButton
   properties {
    toolTipText=User Guide
    widgetStyle {
     clickedSlateSound {resourceObject=null}
     disabled {
      bIsDynamicallyLoaded=false
      drawAs=RoundedBox
      imageType=NoImage
      mirroring=NoMirror
      resourceName=None
      resourceObject=null
      tiling=NoTile
      imageSize=<0 0>
      margin=<ltrb 0 0 0 0>
      outlineSettings {
       bUseBrushTransparency=false
       roundingType=FixedRadius
       width=1
       color {colorUseRule=UseColor_ColorTable, specifiedColor=<rgba 0 0 0 0>}
       cornerRadii=<4 4 4 4>
      }
      tintColor {colorUseRule=UseColor_Specified, specifiedColor=<rgba 0.0395 0.0395 0.0395 1>}
      uVRegion {bIsValid=false, max=<0 0>, min=<0 0>}
     }
     disabledForeground {colorUseRule=UseColor_ColorTable, specifiedColor=<rgba 0 0 0 0>}
     hovered {
      bIsDynamicallyLoaded=false
      drawAs=RoundedBox
      imageType=NoImage
      mirroring=NoMirror
      resourceName=None
      resourceObject=null
      tiling=NoTile
      imageSize=<32 32>
      margin=<ltrb 0 0 0 0>
      outlineSettings {
       bUseBrushTransparency=false
       roundingType=FixedRadius
       width=1
       color {colorUseRule=UseColor_ColorTable, specifiedColor=<rgba 0 0 0 0>}
       cornerRadii=<4 4 4 4>
      }
      tintColor {colorUseRule=UseColor_Specified, specifiedColor=<rgba 0.0953 0.0953 0.0953 1>}
      uVRegion {bIsValid=false, max=<0 0>, min=<0 0>}
     }
     hoveredForeground {colorUseRule=UseColor_ColorTable, specifiedColor=<rgba 0 0 0 0>}
     hoveredSlateSound {resourceObject=null}
     normal {
      bIsDynamicallyLoaded=false
      drawAs=RoundedBox
      imageType=NoImage
      mirroring=NoMirror
      resourceName=None
      resourceObject=null
      tiling=NoTile
      imageSize=<32 32>
      margin=<ltrb 0 0 0 0>
      outlineSettings {
       bUseBrushTransparency=false
       roundingType=FixedRadius
       width=1
       color {colorUseRule=UseColor_ColorTable, specifiedColor=<rgba 0 0 0 0>}
       cornerRadii=<4 4 4 4>
      }
      tintColor {colorUseRule=UseColor_Specified, specifiedColor=<rgba 0.0395 0.0395 0.0395 1>}
      uVRegion {bIsValid=false, max=<0 0>, min=<0 0>}
     }
     normalForeground {colorUseRule=UseColor_ColorTable, specifiedColor=<rgba 0 0 0 0>}
     normalPadding=<ltrb 4 2 4 2>
     pressed {
      bIsDynamicallyLoaded=false
      drawAs=RoundedBox
      imageType=NoImage
      mirroring=NoMirror
      resourceName=None
      resourceObject=null
      tiling=NoTile
      imageSize=<32 32>
      margin=<ltrb 0 0 0 0>
      outlineSettings {
       bUseBrushTransparency=false
       roundingType=FixedRadius
       width=1
       color {colorUseRule=UseColor_ColorTable, specifiedColor=<rgba 0 0 0 0>}
       cornerRadii=<4 4 4 4>
      }
      tintColor {colorUseRule=UseColor_Specified, specifiedColor=<rgba 0.0284 0.0284 0.0284 1>}
      uVRegion {bIsValid=false, max=<0 0>, min=<0 0>}
     }
     pressedForeground {colorUseRule=UseColor_ColorTable, specifiedColor=<rgba 0 0 0 0>}
     pressedPadding=<ltrb 4 2.5 4 0.5>
     pressedSlateSound {resourceObject=null}
    }
   }
   slot {class=WrapBoxSlot, properties {verticalAlignment=VAlign_Center}}
  }
  HorizontalBox_6 {widgetClassPath=&HorizontalBox, parent=UserGuideButton, widget=HorizontalBox_6, slot {class=ButtonSlot}}
  HelpIcon {
   widgetClassPath=&Image
   parent=HorizontalBox_6
   widget=HelpIcon
   properties {brush {
     bIsDynamicallyLoaded=false
     drawAs=Image
     imageType=FullColor
     mirroring=NoMirror
     resourceName=None
     resourceObject=/Game/Athena/UI/InGame/Creative/PropertyEditor/Icon_Picker/Textures_V2/64x64/T_UI_IconLibrary_LightBulb_64
     tiling=NoTile
     imageSize=<20 20>
     margin=<ltrb 0 0 0 0>
     outlineSettings {
      bUseBrushTransparency=false
      roundingType=HalfHeightRadius
      width=0
      color {colorUseRule=UseColor_Specified, specifiedColor=<rgba 0 0 0 0>}
      cornerRadii=<0 0 0 0>
     }
     tintColor {colorUseRule=UseColor_Specified, specifiedColor=<rgba 0.5885 0.5885 0.5885 1>}
     uVRegion {bIsValid=false, max=<0 0>, min=<0 0>}
    }}
   slot {class=HorizontalBoxSlot, properties {verticalAlignment=VAlign_Center}}
  }
  ToggleThemeButton {
   widgetClassPath=&EditorUtilityButton
   parent=WrapBox_1
   widget=ToggleThemeButton
   properties {
    toolTipText=User Guide
    visibility=Collapsed
    widgetStyle {
     clickedSlateSound {resourceObject=null}
     disabled {
      bIsDynamicallyLoaded=false
      drawAs=RoundedBox
      imageType=NoImage
      mirroring=NoMirror
      resourceName=None
      resourceObject=null
      tiling=NoTile
      imageSize=<0 0>
      margin=<ltrb 0 0 0 0>
      outlineSettings {
       bUseBrushTransparency=false
       roundingType=FixedRadius
       width=1
       color {colorUseRule=UseColor_ColorTable, specifiedColor=<rgba 0 0 0 0>}
       cornerRadii=<4 4 4 4>
      }
      tintColor {colorUseRule=UseColor_Specified, specifiedColor=<rgba 0.0395 0.0395 0.0395 1>}
      uVRegion {bIsValid=false, max=<0 0>, min=<0 0>}
     }
     disabledForeground {colorUseRule=UseColor_ColorTable, specifiedColor=<rgba 0 0 0 0>}
     hovered {
      bIsDynamicallyLoaded=false
      drawAs=RoundedBox
      imageType=NoImage
      mirroring=NoMirror
      resourceName=None
      resourceObject=null
      tiling=NoTile
      imageSize=<32 32>
      margin=<ltrb 0 0 0 0>
      outlineSettings {
       bUseBrushTransparency=false
       roundingType=FixedRadius
       width=1
       color {colorUseRule=UseColor_ColorTable, specifiedColor=<rgba 0 0 0 0>}
       cornerRadii=<4 4 4 4>
      }
      tintColor {colorUseRule=UseColor_Specified, specifiedColor=<rgba 0.0953 0.0953 0.0953 1>}
      uVRegion {bIsValid=false, max=<0 0>, min=<0 0>}
     }
     hoveredForeground {colorUseRule=UseColor_ColorTable, specifiedColor=<rgba 0 0 0 0>}
     hoveredSlateSound {resourceObject=null}
     normal {
      bIsDynamicallyLoaded=false
      drawAs=RoundedBox
      imageType=NoImage
      mirroring=NoMirror
      resourceName=None
      resourceObject=null
      tiling=NoTile
      imageSize=<32 32>
      margin=<ltrb 0 0 0 0>
      outlineSettings {
       bUseBrushTransparency=false
       roundingType=FixedRadius
       width=1
       color {colorUseRule=UseColor_ColorTable, specifiedColor=<rgba 0 0 0 0>}
       cornerRadii=<4 4 4 4>
      }
      tintColor {colorUseRule=UseColor_Specified, specifiedColor=<rgba 0.0395 0.0395 0.0395 1>}
      uVRegion {bIsValid=false, max=<0 0>, min=<0 0>}
     }
     normalForeground {colorUseRule=UseColor_ColorTable, specifiedColor=<rgba 0 0 0 0>}
     normalPadding=<ltrb 4 2 4 2>
     pressed {
      bIsDynamicallyLoaded=false
      drawAs=RoundedBox
      imageType=NoImage
      mirroring=NoMirror
      resourceName=None
      resourceObject=null
      tiling=NoTile
      imageSize=<32 32>
      margin=<ltrb 0 0 0 0>
      outlineSettings {
       bUseBrushTransparency=false
       roundingType=FixedRadius
       width=1
       color {colorUseRule=UseColor_ColorTable, specifiedColor=<rgba 0 0 0 0>}
       cornerRadii=<4 4 4 4>
      }
      tintColor {colorUseRule=UseColor_Specified, specifiedColor=<rgba 0.0284 0.0284 0.0284 1>}
      uVRegion {bIsValid=false, max=<0 0>, min=<0 0>}
     }
     pressedForeground {colorUseRule=UseColor_ColorTable, specifiedColor=<rgba 0 0 0 0>}
     pressedPadding=<ltrb 4 2.5 4 0.5>
     pressedSlateSound {resourceObject=null}
    }
   }
   slot {class=WrapBoxSlot, properties {verticalAlignment=VAlign_Center}}
  }
  HorizontalBox_8 {widgetClassPath=&HorizontalBox, parent=ToggleThemeButton, widget=HorizontalBox_8, slot {class=ButtonSlot}}
  HelpIcon_1 {
   widgetClassPath=&Image
   parent=HorizontalBox_8
   widget=HelpIcon_1
   properties {brush {
     bIsDynamicallyLoaded=false
     drawAs=Image
     imageType=FullColor
     mirroring=NoMirror
     resourceName=None
     resourceObject=/Game/Athena/UI/InGame/Creative/PropertyEditor/Icon_Picker/Textures_V2/64x64/T_UI_IconLibrary_Moon_64
     tiling=NoTile
     imageSize=<20 20>
     margin=<ltrb 0 0 0 0>
     outlineSettings {
      bUseBrushTransparency=false
      roundingType=HalfHeightRadius
      width=0
      color {colorUseRule=UseColor_Specified, specifiedColor=<rgba 0 0 0 0>}
      cornerRadii=<0 0 0 0>
     }
     tintColor {colorUseRule=UseColor_Specified, specifiedColor=<rgba 0.5885 0.5885 0.5885 1>}
     uVRegion {bIsValid=false, max=<0 0>, min=<0 0>}
    }}
   slot {class=HorizontalBoxSlot, properties {verticalAlignment=VAlign_Center}}
  }
  Body {
   widgetClassPath=&VerticalBox
   parent=Content
   widget=Body
   slot {class=VerticalBoxSlot, properties {size {sizeRule=Fill, value=1}}}
  }
  MetaSoundAssetPickers {widgetClassPath=&HorizontalBox, parent=Body, widget=MetaSoundAssetPickers, slot {class=VerticalBoxSlot}}
  VerticalBox_263 {
   widgetClassPath=&VerticalBox
   parent=MetaSoundAssetPickers
   widget=VerticalBox_263
   slot {class=HorizontalBoxSlot, properties {size {sizeRule=Fill, value=1}}}
  }
  Border_1 {
   widgetClassPath=&Border
   parent=VerticalBox_263
   widget=Border_1
   properties {brushColor=<rgba 0.0103 0.0103 0.0103 1>, padding=<ltrb 0 0 0 0>}
   slot {class=VerticalBoxSlot}
  }
  VerticalBox_3 {widgetClassPath=&VerticalBox, parent=Border_1, widget=VerticalBox_3, slot {class=BorderSlot, properties {padding=<ltrb 0 0 0 0>}}}
  HorizontalBox_219 {
   widgetClassPath=&HorizontalBox
   parent=VerticalBox_3
   widget=HorizontalBox_219
   slot {class=VerticalBoxSlot, properties {horizontalAlignment=HAlign_Left, verticalAlignment=VAlign_Center, padding=<ltrb 16 8 0 6>}}
  }
  Image_122 {
   widgetClassPath=&Image
   parent=HorizontalBox_219
   widget=Image_122
   properties {brush {
     bIsDynamicallyLoaded=false
     drawAs=Image
     imageType=FullColor
     mirroring=NoMirror
     resourceName=None
     resourceObject=null
     tiling=NoTile
     imageSize=<16 16>
     margin=<ltrb 0 0 0 0>
     outlineSettings {
      bUseBrushTransparency=false
      roundingType=HalfHeightRadius
      width=0
      color {colorUseRule=UseColor_Specified, specifiedColor=<rgba 0 0 0 0>}
      cornerRadii=<0 0 0 0>
     }
     tintColor {colorUseRule=UseColor_Specified, specifiedColor=<rgba 0.2384 0.2384 0.2384 1>}
     uVRegion {bIsValid=false, max=<0 0>, min=<0 0>}
    }}
   slot {class=HorizontalBoxSlot, properties {horizontalAlignment=HAlign_Center, verticalAlignment=VAlign_Center, padding=<ltrb 0 0 0 1>}}
  }
  TextBlock_343 {
   widgetClassPath=&TextBlock
   parent=HorizontalBox_219
   widget=TextBlock_343
   properties {
    text=Original MetaSound Source Preset
    toolTipText=The preset that you want to migrate to a new parent preset.
    colorAndOpacity {colorUseRule=UseColor_Specified, specifiedColor=<rgba 0.2384 0.2384 0.2384 1>}
    font {
     bForceMonospaced=false
     bMaterialIsStencil=false
     fontMaterial=null
     fontObject=&Roboto
     letterSpacing=0
     monospacedWidth=1
     size=12
     skewAmount=0
     typefaceFontName=Bold
     outlineSettings {
      bApplyOutlineToDropShadows=false
      bMiteredCorners=false
      bSeparateFillAlpha=false
      outlineMaterial=null
      outlineSize=0
      outlineColor=<rgba 0 0 0 1>
     }
    }
   }
   slot {class=HorizontalBoxSlot, properties {horizontalAlignment=HAlign_Left, verticalAlignment=VAlign_Center, padding=<ltrb 4 0 0 0>}}
  }
  OriginMetaSoundPicker {
   widgetClassPath=&SinglePropertyView
   parent=VerticalBox_3
   widget=OriginMetaSoundPicker
   properties {nameOverride=" ", propertyName=Origin MetaSound}
   slot {class=VerticalBoxSlot, properties {padding=<ltrb 0 0 0 8>}}
  }
  VerticalBox_1 {
   widgetClassPath=&VerticalBox
   parent=MetaSoundAssetPickers
   widget=VerticalBox_1
   slot {class=HorizontalBoxSlot, properties {size {sizeRule=Fill, value=1}}}
  }
  Border_4 {
   widgetClassPath=&Border
   parent=VerticalBox_1
   widget=Border_4
   properties {brushColor=<rgba 0.0103 0.0103 0.0103 1>, padding=<ltrb 0 0 0 0>}
   slot {class=VerticalBoxSlot}
  }
  VerticalBox_4 {widgetClassPath=&VerticalBox, parent=Border_4, widget=VerticalBox_4, slot {class=BorderSlot, properties {padding=<ltrb 0 0 0 0>}}}
  HorizontalBox_220 {
   widgetClassPath=&HorizontalBox
   parent=VerticalBox_4
   widget=HorizontalBox_220
   slot {class=VerticalBoxSlot, properties {padding=<ltrb 16 8 0 6>}}
  }
  Image_1 {
   widgetClassPath=&Image
   parent=HorizontalBox_220
   widget=Image_1
   properties {
    brush {
     bIsDynamicallyLoaded=false
     drawAs=Image
     imageType=FullColor
     mirroring=NoMirror
     resourceName=None
     resourceObject=null
     tiling=NoTile
     imageSize=<16 16>
     margin=<ltrb 0 0 0 0>
     outlineSettings {
      bUseBrushTransparency=false
      roundingType=HalfHeightRadius
      width=0
      color {colorUseRule=UseColor_Specified, specifiedColor=<rgba 0 0 0 0>}
      cornerRadii=<0 0 0 0>
     }
     tintColor {colorUseRule=UseColor_Specified, specifiedColor=<rgba 0.2384 0.2384 0.2384 1>}
     uVRegion {bIsValid=false, max=<0 0>, min=<0 0>}
    }
    renderTransform {angle=90, scale=<1 1>, shear=<0 0>, translation=<0 0>}
   }
   slot {class=HorizontalBoxSlot, properties {horizontalAlignment=HAlign_Center, verticalAlignment=VAlign_Center, padding=<ltrb 0 0 0 1>}}
  }
  TextBlock_502 {
   widgetClassPath=&TextBlock
   parent=HorizontalBox_220
   widget=TextBlock_502
   properties {
    text=New MetaSound Source Parent
    toolTipText=The MetaSound Source that you want to migrate the original preset to.
    colorAndOpacity {colorUseRule=UseColor_Specified, specifiedColor=<rgba 0.2384 0.2384 0.2384 1>}
    font {
     bForceMonospaced=false
     bMaterialIsStencil=false
     fontMaterial=null
     fontObject=&Roboto
     letterSpacing=0
     monospacedWidth=1
     size=12
     skewAmount=0
     typefaceFontName=Bold
     outlineSettings {
      bApplyOutlineToDropShadows=false
      bMiteredCorners=false
      bSeparateFillAlpha=false
      outlineMaterial=null
      outlineSize=0
      outlineColor=<rgba 0 0 0 1>
     }
    }
   }
   slot {class=HorizontalBoxSlot, properties {horizontalAlignment=HAlign_Left, verticalAlignment=VAlign_Center, padding=<ltrb 4 0 0 0>}}
  }
  DestinationMetaSoundPicker {
   widgetClassPath=&SinglePropertyView
   parent=VerticalBox_4
   widget=DestinationMetaSoundPicker
   properties {nameOverride=" ", propertyName=Destination MetaSound}
   slot {class=VerticalBoxSlot, properties {padding=<ltrb 0 0 0 8>}}
  }
  Spacer_213 {widgetClassPath=&Spacer, parent=Body, widget=Spacer_213, properties {size=<1 4>}, slot {class=VerticalBoxSlot}}
  MetaSoundInputLists {
   widgetClassPath=&HorizontalBox
   parent=Body
   widget=MetaSoundInputLists
   slot {class=VerticalBoxSlot, properties {size {sizeRule=Fill, value=1}}}
  }
  VerticalBox_103 {
   widgetClassPath=&VerticalBox
   parent=MetaSoundInputLists
   widget=VerticalBox_103
   slot {class=HorizontalBoxSlot, properties {size {sizeRule=Fill, value=1}}}
  }
  OriginInputSearchBox {
   widgetClassPath=&EditorUtilityEditableTextBox
   bIsVariable=true
   parent=VerticalBox_103
   widget=OriginInputSearchBox
   properties {
    hintText=Search...
    selectAllTextWhenFocused=true
    widgetStyle {
     backgroundColor {colorUseRule=UseColor_Specified, specifiedColor=<rgba 1 1 1 1>}
     backgroundImageFocused {
      bIsDynamicallyLoaded=false
      drawAs=RoundedBox
      imageType=NoImage
      mirroring=NoMirror
      resourceName=None
      resourceObject=null
      tiling=NoTile
      imageSize=<0 0>
      margin=<ltrb 0 0 0 0>
      outlineSettings {
       bUseBrushTransparency=false
       roundingType=FixedRadius
       width=1
       color {colorUseRule=UseColor_ColorTable, specifiedColor=<rgba 0 0 0 0>}
       cornerRadii=<10 10 10 10>
      }
      tintColor {colorUseRule=UseColor_Specified, specifiedColor=<rgba 0.0048 0.0048 0.0048 1>}
      uVRegion {bIsValid=false, max=<0 0>, min=<0 0>}
     }
     backgroundImageHovered {
      bIsDynamicallyLoaded=false
      drawAs=RoundedBox
      imageType=NoImage
      mirroring=NoMirror
      resourceName=None
      resourceObject=null
      tiling=NoTile
      imageSize=<0 0>
      margin=<ltrb 0 0 0 0>
      outlineSettings {
       bUseBrushTransparency=false
       roundingType=FixedRadius
       width=1
       color {colorUseRule=UseColor_ColorTable, specifiedColor=<rgba 0 0 0 0>}
       cornerRadii=<10 10 10 10>
      }
      tintColor {colorUseRule=UseColor_Specified, specifiedColor=<rgba 0.0048 0.0048 0.0048 1>}
      uVRegion {bIsValid=false, max=<0 0>, min=<0 0>}
     }
     backgroundImageNormal {
      bIsDynamicallyLoaded=false
      drawAs=RoundedBox
      imageType=NoImage
      mirroring=NoMirror
      resourceName=None
      resourceObject=null
      tiling=NoTile
      imageSize=<0 0>
      margin=<ltrb 0 0 0 0>
      outlineSettings {
       bUseBrushTransparency=false
       roundingType=FixedRadius
       width=1
       color {colorUseRule=UseColor_Specified, specifiedColor=<rgba 0.0331 0.0331 0.0331 1>}
       cornerRadii=<10 10 10 10>
      }
      tintColor {colorUseRule=UseColor_Specified, specifiedColor=<rgba 0.0048 0.0048 0.0048 1>}
      uVRegion {bIsValid=false, max=<0 0>, min=<0 0>}
     }
     backgroundImageReadOnly {
      bIsDynamicallyLoaded=false
      drawAs=RoundedBox
      imageType=NoImage
      mirroring=NoMirror
      resourceName=None
      resourceObject=null
      tiling=NoTile
      imageSize=<0 0>
      margin=<ltrb 0 0 0 0>
      outlineSettings {
       bUseBrushTransparency=false
       roundingType=FixedRadius
       width=1
       color {colorUseRule=UseColor_ColorTable, specifiedColor=<rgba 0 0 0 0>}
       cornerRadii=<10 10 10 10>
      }
      tintColor {colorUseRule=UseColor_Specified, specifiedColor=<rgba 0.0284 0.0284 0.0284 1>}
      uVRegion {bIsValid=false, max=<0 0>, min=<0 0>}
     }
     focusedForegroundColor {colorUseRule=UseColor_Specified, specifiedColor=<rgba 1 1 1 1>}
     foregroundColor {colorUseRule=UseColor_Specified, specifiedColor=<rgba 0.5271 0.5271 0.5271 1>}
     hScrollBarPadding=<ltrb 0 0 0 0>
     padding=<ltrb 8 3 8 3>
     readOnlyForegroundColor {colorUseRule=UseColor_Foreground, specifiedColor=<rgba 1 0 1 1>}
     scrollBarStyle {
      thickness=8
      draggedThumbImage {
       bIsDynamicallyLoaded=false
       drawAs=RoundedBox
       imageType=NoImage
       mirroring=NoMirror
       resourceName=None
       resourceObject=null
       tiling=NoTile
       imageSize=<0 0>
       margin=<ltrb 0 0 0 0>
       outlineSettings {
        bUseBrushTransparency=false
        roundingType=FixedRadius
        width=0
        color {colorUseRule=UseColor_Specified, specifiedColor=<rgba 0 0 0 0>}
        cornerRadii=<4 4 4 4>
       }
       tintColor {colorUseRule=UseColor_Specified, specifiedColor=<rgba 0.2159 0.2159 0.2159 1>}
       uVRegion {bIsValid=false, max=<0 0>, min=<0 0>}
      }
      horizontalBackgroundImage {
       bIsDynamicallyLoaded=false
       drawAs=NoDrawType
       imageType=NoImage
       mirroring=NoMirror
       resourceName=None
       resourceObject=null
       tiling=NoTile
       imageSize=<0 0>
       margin=<ltrb 0 0 0 0>
       outlineSettings {
        bUseBrushTransparency=false
        roundingType=HalfHeightRadius
        width=0
        color {colorUseRule=UseColor_Specified, specifiedColor=<rgba 0 0 0 0>}
        cornerRadii=<0 0 0 0>
       }
       tintColor {colorUseRule=UseColor_Specified, specifiedColor=<rgba 1 1 1 1>}
       uVRegion {bIsValid=false, max=<0 0>, min=<0 0>}
      }
      horizontalBottomSlotImage {
       bIsDynamicallyLoaded=false
       drawAs=NoDrawType
       imageType=NoImage
       mirroring=NoMirror
       resourceName=None
       resourceObject=null
       tiling=NoTile
       imageSize=<0 0>
       margin=<ltrb 0 0 0 0>
       outlineSettings {
        bUseBrushTransparency=false
        roundingType=HalfHeightRadius
        width=0
        color {colorUseRule=UseColor_Specified, specifiedColor=<rgba 0 0 0 0>}
        cornerRadii=<0 0 0 0>
       }
       tintColor {colorUseRule=UseColor_Specified, specifiedColor=<rgba 1 1 1 1>}
       uVRegion {bIsValid=false, max=<0 0>, min=<0 0>}
      }
      horizontalTopSlotImage {
       bIsDynamicallyLoaded=false
       drawAs=NoDrawType
       imageType=NoImage
       mirroring=NoMirror
       resourceName=None
       resourceObject=null
       tiling=NoTile
       imageSize=<0 0>
       margin=<ltrb 0 0 0 0>
       outlineSettings {
        bUseBrushTransparency=false
        roundingType=HalfHeightRadius
        width=0
        color {colorUseRule=UseColor_Specified, specifiedColor=<rgba 0 0 0 0>}
        cornerRadii=<0 0 0 0>
       }
       tintColor {colorUseRule=UseColor_Specified, specifiedColor=<rgba 1 1 1 1>}
       uVRegion {bIsValid=false, max=<0 0>, min=<0 0>}
      }
      hoveredThumbImage {
       bIsDynamicallyLoaded=false
       drawAs=RoundedBox
       imageType=NoImage
       mirroring=NoMirror
       resourceName=None
       resourceObject=null
       tiling=NoTile
       imageSize=<0 0>
       margin=<ltrb 0 0 0 0>
       outlineSettings {
        bUseBrushTransparency=false
        roundingType=FixedRadius
        width=0
        color {colorUseRule=UseColor_Specified, specifiedColor=<rgba 0 0 0 0>}
        cornerRadii=<4 4 4 4>
       }
       tintColor {colorUseRule=UseColor_Specified, specifiedColor=<rgba 0.2159 0.2159 0.2159 1>}
       uVRegion {bIsValid=false, max=<0 0>, min=<0 0>}
      }
      normalThumbImage {
       bIsDynamicallyLoaded=false
       drawAs=RoundedBox
       imageType=NoImage
       mirroring=NoMirror
       resourceName=None
       resourceObject=null
       tiling=NoTile
       imageSize=<0 0>
       margin=<ltrb 0 0 0 0>
       outlineSettings {
        bUseBrushTransparency=false
        roundingType=FixedRadius
        width=0
        color {colorUseRule=UseColor_Specified, specifiedColor=<rgba 0 0 0 0>}
        cornerRadii=<4 4 4 4>
       }
       tintColor {colorUseRule=UseColor_Specified, specifiedColor=<rgba 0.0953 0.0953 0.0953 1>}
       uVRegion {bIsValid=false, max=<0 0>, min=<0 0>}
      }
      verticalBackgroundImage {
       bIsDynamicallyLoaded=false
       drawAs=NoDrawType
       imageType=NoImage
       mirroring=NoMirror
       resourceName=None
       resourceObject=null
       tiling=NoTile
       imageSize=<0 0>
       margin=<ltrb 0 0 0 0>
       outlineSettings {
        bUseBrushTransparency=false
        roundingType=HalfHeightRadius
        width=0
        color {colorUseRule=UseColor_Specified, specifiedColor=<rgba 0 0 0 0>}
        cornerRadii=<0 0 0 0>
       }
       tintColor {colorUseRule=UseColor_Specified, specifiedColor=<rgba 1 1 1 1>}
       uVRegion {bIsValid=false, max=<0 0>, min=<0 0>}
      }
      verticalBottomSlotImage {
       bIsDynamicallyLoaded=false
       drawAs=NoDrawType
       imageType=NoImage
       mirroring=NoMirror
       resourceName=None
       resourceObject=null
       tiling=NoTile
       imageSize=<0 0>
       margin=<ltrb 0 0 0 0>
       outlineSettings {
        bUseBrushTransparency=false
        roundingType=HalfHeightRadius
        width=0
        color {colorUseRule=UseColor_Specified, specifiedColor=<rgba 0 0 0 0>}
        cornerRadii=<0 0 0 0>
       }
       tintColor {colorUseRule=UseColor_Specified, specifiedColor=<rgba 1 1 1 1>}
       uVRegion {bIsValid=false, max=<0 0>, min=<0 0>}
      }
      verticalTopSlotImage {
       bIsDynamicallyLoaded=false
       drawAs=NoDrawType
       imageType=NoImage
       mirroring=NoMirror
       resourceName=None
       resourceObject=null
       tiling=NoTile
       imageSize=<0 0>
       margin=<ltrb 0 0 0 0>
       outlineSettings {
        bUseBrushTransparency=false
        roundingType=HalfHeightRadius
        width=0
        color {colorUseRule=UseColor_Specified, specifiedColor=<rgba 0 0 0 0>}
        cornerRadii=<0 0 0 0>
       }
       tintColor {colorUseRule=UseColor_Specified, specifiedColor=<rgba 1 1 1 1>}
       uVRegion {bIsValid=false, max=<0 0>, min=<0 0>}
      }
     }
     textStyle {
      overflowPolicy=Clip
      transformPolicy=None
      colorAndOpacity {colorUseRule=UseColor_Foreground, specifiedColor=<rgba 1 0 1 1>}
      font {
       bForceMonospaced=false
       bMaterialIsStencil=false
       fontMaterial=null
       fontObject=null
       letterSpacing=0
       monospacedWidth=1
       size=10
       skewAmount=0
       typefaceFontName=Regular
       outlineSettings {
        bApplyOutlineToDropShadows=false
        bMiteredCorners=false
        bSeparateFillAlpha=false
        outlineMaterial=null
        outlineSize=0
        outlineColor=<rgba 0 0 0 1>
       }
      }
      highlightColor {colorUseRule=UseColor_ColorTable, specifiedColor=<rgba 0 0 0 0>}
      highlightShape {
       bIsDynamicallyLoaded=false
       drawAs=Image
       imageType=NoImage
       mirroring=NoMirror
       resourceName=None
       resourceObject=null
       tiling=NoTile
       imageSize=<0 0>
       margin=<ltrb 0 0 0 0>
       outlineSettings {
        bUseBrushTransparency=false
        roundingType=HalfHeightRadius
        width=0
        color {colorUseRule=UseColor_Specified, specifiedColor=<rgba 0 0 0 0>}
        cornerRadii=<0 0 0 0>
       }
       tintColor {colorUseRule=UseColor_ColorTable, specifiedColor=<rgba 0 0 0 0>}
       uVRegion {bIsValid=false, max=<0 0>, min=<0 0>}
      }
      selectedBackgroundColor {colorUseRule=UseColor_ColorTable, specifiedColor=<rgba 0 0 0 0>}
      shadowColorAndOpacity=<rgba 0 0 0 1>
      shadowOffset=<0 0>
      strikeBrush {
       bIsDynamicallyLoaded=false
       drawAs=Image
       imageType=NoImage
       mirroring=NoMirror
       resourceName=None
       resourceObject=null
       tiling=NoTile
       imageSize=<32 32>
       margin=<ltrb 0 0 0 0>
       outlineSettings {
        bUseBrushTransparency=false
        roundingType=HalfHeightRadius
        width=0
        color {colorUseRule=UseColor_Specified, specifiedColor=<rgba 0 0 0 0>}
        cornerRadii=<0 0 0 0>
       }
       tintColor {colorUseRule=UseColor_Specified, specifiedColor=<rgba 1 1 1 1>}
       uVRegion {bIsValid=false, max=<0 0>, min=<0 0>}
      }
      underlineBrush {
       bIsDynamicallyLoaded=false
       drawAs=Image
       imageType=NoImage
       mirroring=NoMirror
       resourceName=None
       resourceObject=null
       tiling=NoTile
       imageSize=<32 32>
       margin=<ltrb 0 0 0 0>
       outlineSettings {
        bUseBrushTransparency=false
        roundingType=HalfHeightRadius
        width=0
        color {colorUseRule=UseColor_Specified, specifiedColor=<rgba 0 0 0 0>}
        cornerRadii=<0 0 0 0>
       }
       tintColor {colorUseRule=UseColor_Specified, specifiedColor=<rgba 1 1 1 1>}
       uVRegion {bIsValid=false, max=<0 0>, min=<0 0>}
      }
     }
     vScrollBarPadding=<ltrb 0 0 0 0>
    }
   }
   slot {class=VerticalBoxSlot, properties {padding=<ltrb 2 0 2 4>}}
  }
  Border_2 {
   widgetClassPath=&Border
   parent=VerticalBox_103
   widget=Border_2
   properties {brushColor=<rgba 0.0103 0.0103 0.0103 1>, padding=<ltrb 0 0 0 0>}
   slot {class=VerticalBoxSlot, properties {size {sizeRule=Fill, value=1}}}
  }
  VerticalBox_267 {
   widgetClassPath=&VerticalBox
   parent=Border_2
   widget=VerticalBox_267
   slot {class=BorderSlot, properties {padding=<ltrb 0 0 0 0>}}
  }
  OriginMetaSoundInputsList {
   widgetClassPath=/Script/Blutility.EditorUtilityListView
   bIsVariable=true
   parent=VerticalBox_267
   widget=OriginMetaSoundInputsList
   properties {
    bEnableScrollAnimation=true
    bIsFocusable=false
    entryWidgetClass=/Game/Sounds/Fort_Audio_Blueprints/Utilities/EditorUtilityWidgets/Metasounds/Tools/MetaSoundPresetMigratorData/MetaSoundPresetMigrator_InputEntry.MetaSoundPresetMigrator_InputEntry_C
    numDesignerPreviewEntries=10
   }
   slot {class=VerticalBoxSlot, properties {size {sizeRule=Fill, value=1}}}
  }
  VerticalBox_0 {
   widgetClassPath=&VerticalBox
   parent=VerticalBox_267
   widget=VerticalBox_0
   slot {
    class=VerticalBoxSlot
    properties {horizontalAlignment=HAlign_Center, verticalAlignment=VAlign_Top, padding=<ltrb 0 64 0 0>, size {sizeRule=Fill, value=1}}
   }
  }
  Image_125 {
   widgetClassPath=&Image
   parent=VerticalBox_0
   widget=Image_125
   properties {brush {
     bIsDynamicallyLoaded=false
     drawAs=Image
     imageType=FullColor
     mirroring=NoMirror
     resourceName=None
     resourceObject=null
     tiling=NoTile
     imageSize=<64 64>
     margin=<ltrb 0 0 0 0>
     outlineSettings {
      bUseBrushTransparency=false
      roundingType=HalfHeightRadius
      width=0
      color {colorUseRule=UseColor_Specified, specifiedColor=<rgba 0 0 0 0>}
      cornerRadii=<0 0 0 0>
     }
     tintColor {colorUseRule=UseColor_Specified, specifiedColor=<rgba 0.0284 0.0284 0.0284 1>}
     uVRegion {bIsValid=false, max=<0 0>, min=<0 0>}
    }}
   slot {class=VerticalBoxSlot, properties {horizontalAlignment=HAlign_Center, verticalAlignment=VAlign_Center, padding=<ltrb 16 16 16 16>}}
  }
  TextBlock_153 {
   widgetClassPath=&TextBlock
   parent=VerticalBox_0
   widget=TextBlock_153
   properties {
    justification=Center
    text=Select a preset you wish to migrate.
    colorAndOpacity {colorUseRule=UseColor_Specified, specifiedColor=<rgba 0.5029 0.5029 0.5029 1>}
    font {
     bForceMonospaced=false
     bMaterialIsStencil=false
     fontMaterial=null
     fontObject=&Roboto
     letterSpacing=0
     monospacedWidth=1
     size=9
     skewAmount=0
     typefaceFontName=Regular
     outlineSettings {
      bApplyOutlineToDropShadows=false
      bMiteredCorners=false
      bSeparateFillAlpha=false
      outlineMaterial=null
      outlineSize=0
      outlineColor=<rgba 0 0 0 1>
     }
    }
   }
   slot {class=VerticalBoxSlot}
  }
  Spacer_104 {widgetClassPath=&Spacer, parent=MetaSoundInputLists, widget=Spacer_104, properties {size=<4 1>}, slot {class=HorizontalBoxSlot}}
  VerticalBox {
   widgetClassPath=&VerticalBox
   parent=MetaSoundInputLists
   widget=VerticalBox
   slot {class=HorizontalBoxSlot, properties {size {sizeRule=Fill, value=1}}}
  }
  DestinationInputSearchBox {
   widgetClassPath=&EditorUtilityEditableTextBox
   bIsVariable=true
   parent=VerticalBox
   widget=DestinationInputSearchBox
   properties {
    hintText=Search...
    selectAllTextWhenFocused=true
    widgetStyle {
     backgroundColor {colorUseRule=UseColor_Specified, specifiedColor=<rgba 1 1 1 1>}
     backgroundImageFocused {
      bIsDynamicallyLoaded=false
      drawAs=RoundedBox
      imageType=NoImage
      mirroring=NoMirror
      resourceName=None
      resourceObject=null
      tiling=NoTile
      imageSize=<0 0>
      margin=<ltrb 0 0 0 0>
      outlineSettings {
       bUseBrushTransparency=false
       roundingType=FixedRadius
       width=1
       color {colorUseRule=UseColor_ColorTable, specifiedColor=<rgba 0 0 0 0>}
       cornerRadii=<10 10 10 10>
      }
      tintColor {colorUseRule=UseColor_Specified, specifiedColor=<rgba 0.0048 0.0048 0.0048 1>}
      uVRegion {bIsValid=false, max=<0 0>, min=<0 0>}
     }
     backgroundImageHovered {
      bIsDynamicallyLoaded=false
      drawAs=RoundedBox
      imageType=NoImage
      mirroring=NoMirror
      resourceName=None
      resourceObject=null
      tiling=NoTile
      imageSize=<0 0>
      margin=<ltrb 0 0 0 0>
      outlineSettings {
       bUseBrushTransparency=false
       roundingType=FixedRadius
       width=1
       color {colorUseRule=UseColor_ColorTable, specifiedColor=<rgba 0 0 0 0>}
       cornerRadii=<10 10 10 10>
      }
      tintColor {colorUseRule=UseColor_Specified, specifiedColor=<rgba 0.0048 0.0048 0.0048 1>}
      uVRegion {bIsValid=false, max=<0 0>, min=<0 0>}
     }
     backgroundImageNormal {
      bIsDynamicallyLoaded=false
      drawAs=RoundedBox
      imageType=NoImage
      mirroring=NoMirror
      resourceName=None
      resourceObject=null
      tiling=NoTile
      imageSize=<0 0>
      margin=<ltrb 0 0 0 0>
      outlineSettings {
       bUseBrushTransparency=false
       roundingType=FixedRadius
       width=1
       color {colorUseRule=UseColor_Specified, specifiedColor=<rgba 0.0331 0.0331 0.0331 1>}
       cornerRadii=<10 10 10 10>
      }
      tintColor {colorUseRule=UseColor_Specified, specifiedColor=<rgba 0.0048 0.0048 0.0048 1>}
      uVRegion {bIsValid=false, max=<0 0>, min=<0 0>}
     }
     backgroundImageReadOnly {
      bIsDynamicallyLoaded=false
      drawAs=RoundedBox
      imageType=NoImage
      mirroring=NoMirror
      resourceName=None
      resourceObject=null
      tiling=NoTile
      imageSize=<0 0>
      margin=<ltrb 0 0 0 0>
      outlineSettings {
       bUseBrushTransparency=false
       roundingType=FixedRadius
       width=1
       color {colorUseRule=UseColor_ColorTable, specifiedColor=<rgba 0 0 0 0>}
       cornerRadii=<10 10 10 10>
      }
      tintColor {colorUseRule=UseColor_Specified, specifiedColor=<rgba 0.0284 0.0284 0.0284 1>}
      uVRegion {bIsValid=false, max=<0 0>, min=<0 0>}
     }
     focusedForegroundColor {colorUseRule=UseColor_Specified, specifiedColor=<rgba 1 1 1 1>}
     foregroundColor {colorUseRule=UseColor_Specified, specifiedColor=<rgba 0.5271 0.5271 0.5271 1>}
     hScrollBarPadding=<ltrb 0 0 0 0>
     padding=<ltrb 8 3 8 3>
     readOnlyForegroundColor {colorUseRule=UseColor_Foreground, specifiedColor=<rgba 1 0 1 1>}
     scrollBarStyle {
      thickness=8
      draggedThumbImage {
       bIsDynamicallyLoaded=false
       drawAs=RoundedBox
       imageType=NoImage
       mirroring=NoMirror
       resourceName=None
       resourceObject=null
       tiling=NoTile
       imageSize=<0 0>
       margin=<ltrb 0 0 0 0>
       outlineSettings {
        bUseBrushTransparency=false
        roundingType=FixedRadius
        width=0
        color {colorUseRule=UseColor_Specified, specifiedColor=<rgba 0 0 0 0>}
        cornerRadii=<4 4 4 4>
       }
       tintColor {colorUseRule=UseColor_Specified, specifiedColor=<rgba 0.2159 0.2159 0.2159 1>}
       uVRegion {bIsValid=false, max=<0 0>, min=<0 0>}
      }
      horizontalBackgroundImage {
       bIsDynamicallyLoaded=false
       drawAs=NoDrawType
       imageType=NoImage
       mirroring=NoMirror
       resourceName=None
       resourceObject=null
       tiling=NoTile
       imageSize=<0 0>
       margin=<ltrb 0 0 0 0>
       outlineSettings {
        bUseBrushTransparency=false
        roundingType=HalfHeightRadius
        width=0
        color {colorUseRule=UseColor_Specified, specifiedColor=<rgba 0 0 0 0>}
        cornerRadii=<0 0 0 0>
       }
       tintColor {colorUseRule=UseColor_Specified, specifiedColor=<rgba 1 1 1 1>}
       uVRegion {bIsValid=false, max=<0 0>, min=<0 0>}
      }
      horizontalBottomSlotImage {
       bIsDynamicallyLoaded=false
       drawAs=NoDrawType
       imageType=NoImage
       mirroring=NoMirror
       resourceName=None
       resourceObject=null
       tiling=NoTile
       imageSize=<0 0>
       margin=<ltrb 0 0 0 0>
       outlineSettings {
        bUseBrushTransparency=false
        roundingType=HalfHeightRadius
        width=0
        color {colorUseRule=UseColor_Specified, specifiedColor=<rgba 0 0 0 0>}
        cornerRadii=<0 0 0 0>
       }
       tintColor {colorUseRule=UseColor_Specified, specifiedColor=<rgba 1 1 1 1>}
       uVRegion {bIsValid=false, max=<0 0>, min=<0 0>}
      }
      horizontalTopSlotImage {
       bIsDynamicallyLoaded=false
       drawAs=NoDrawType
       imageType=NoImage
       mirroring=NoMirror
       resourceName=None
       resourceObject=null
       tiling=NoTile
       imageSize=<0 0>
       margin=<ltrb 0 0 0 0>
       outlineSettings {
        bUseBrushTransparency=false
        roundingType=HalfHeightRadius
        width=0
        color {colorUseRule=UseColor_Specified, specifiedColor=<rgba 0 0 0 0>}
        cornerRadii=<0 0 0 0>
       }
       tintColor {colorUseRule=UseColor_Specified, specifiedColor=<rgba 1 1 1 1>}
       uVRegion {bIsValid=false, max=<0 0>, min=<0 0>}
      }
      hoveredThumbImage {
       bIsDynamicallyLoaded=false
       drawAs=RoundedBox
       imageType=NoImage
       mirroring=NoMirror
       resourceName=None
       resourceObject=null
       tiling=NoTile
       imageSize=<0 0>
       margin=<ltrb 0 0 0 0>
       outlineSettings {
        bUseBrushTransparency=false
        roundingType=FixedRadius
        width=0
        color {colorUseRule=UseColor_Specified, specifiedColor=<rgba 0 0 0 0>}
        cornerRadii=<4 4 4 4>
       }
       tintColor {colorUseRule=UseColor_Specified, specifiedColor=<rgba 0.2159 0.2159 0.2159 1>}
       uVRegion {bIsValid=false, max=<0 0>, min=<0 0>}
      }
      normalThumbImage {
       bIsDynamicallyLoaded=false
       drawAs=RoundedBox
       imageType=NoImage
       mirroring=NoMirror
       resourceName=None
       resourceObject=null
       tiling=NoTile
       imageSize=<0 0>
       margin=<ltrb 0 0 0 0>
       outlineSettings {
        bUseBrushTransparency=false
        roundingType=FixedRadius
        width=0
        color {colorUseRule=UseColor_Specified, specifiedColor=<rgba 0 0 0 0>}
        cornerRadii=<4 4 4 4>
       }
       tintColor {colorUseRule=UseColor_Specified, specifiedColor=<rgba 0.0953 0.0953 0.0953 1>}
       uVRegion {bIsValid=false, max=<0 0>, min=<0 0>}
      }
      verticalBackgroundImage {
       bIsDynamicallyLoaded=false
       drawAs=NoDrawType
       imageType=NoImage
       mirroring=NoMirror
       resourceName=None
       resourceObject=null
       tiling=NoTile
       imageSize=<0 0>
       margin=<ltrb 0 0 0 0>
       outlineSettings {
        bUseBrushTransparency=false
        roundingType=HalfHeightRadius
        width=0
        color {colorUseRule=UseColor_Specified, specifiedColor=<rgba 0 0 0 0>}
        cornerRadii=<0 0 0 0>
       }
       tintColor {colorUseRule=UseColor_Specified, specifiedColor=<rgba 1 1 1 1>}
       uVRegion {bIsValid=false, max=<0 0>, min=<0 0>}
      }
      verticalBottomSlotImage {
       bIsDynamicallyLoaded=false
       drawAs=NoDrawType
       imageType=NoImage
       mirroring=NoMirror
       resourceName=None
       resourceObject=null
       tiling=NoTile
       imageSize=<0 0>
       margin=<ltrb 0 0 0 0>
       outlineSettings {
        bUseBrushTransparency=false
        roundingType=HalfHeightRadius
        width=0
        color {colorUseRule=UseColor_Specified, specifiedColor=<rgba 0 0 0 0>}
        cornerRadii=<0 0 0 0>
       }
       tintColor {colorUseRule=UseColor_Specified, specifiedColor=<rgba 1 1 1 1>}
       uVRegion {bIsValid=false, max=<0 0>, min=<0 0>}
      }
      verticalTopSlotImage {
       bIsDynamicallyLoaded=false
       drawAs=NoDrawType
       imageType=NoImage
       mirroring=NoMirror
       resourceName=None
       resourceObject=null
       tiling=NoTile
       imageSize=<0 0>
       margin=<ltrb 0 0 0 0>
       outlineSettings {
        bUseBrushTransparency=false
        roundingType=HalfHeightRadius
        width=0
        color {colorUseRule=UseColor_Specified, specifiedColor=<rgba 0 0 0 0>}
        cornerRadii=<0 0 0 0>
       }
       tintColor {colorUseRule=UseColor_Specified, specifiedColor=<rgba 1 1 1 1>}
       uVRegion {bIsValid=false, max=<0 0>, min=<0 0>}
      }
     }
     textStyle {
      overflowPolicy=Clip
      transformPolicy=None
      colorAndOpacity {colorUseRule=UseColor_Foreground, specifiedColor=<rgba 1 0 1 1>}
      font {
       bForceMonospaced=false
       bMaterialIsStencil=false
       fontMaterial=null
       fontObject=null
       letterSpacing=0
       monospacedWidth=1
       size=10
       skewAmount=0
       typefaceFontName=Regular
       outlineSettings {
        bApplyOutlineToDropShadows=false
        bMiteredCorners=false
        bSeparateFillAlpha=false
        outlineMaterial=null
        outlineSize=0
        outlineColor=<rgba 0 0 0 1>
       }
      }
      highlightColor {colorUseRule=UseColor_ColorTable, specifiedColor=<rgba 0 0 0 0>}
      highlightShape {
       bIsDynamicallyLoaded=false
       drawAs=Image
       imageType=NoImage
       mirroring=NoMirror
       resourceName=None
       resourceObject=null
       tiling=NoTile
       imageSize=<0 0>
       margin=<ltrb 0 0 0 0>
       outlineSettings {
        bUseBrushTransparency=false
        roundingType=HalfHeightRadius
        width=0
        color {colorUseRule=UseColor_Specified, specifiedColor=<rgba 0 0 0 0>}
        cornerRadii=<0 0 0 0>
       }
       tintColor {colorUseRule=UseColor_ColorTable, specifiedColor=<rgba 0 0 0 0>}
       uVRegion {bIsValid=false, max=<0 0>, min=<0 0>}
      }
      selectedBackgroundColor {colorUseRule=UseColor_ColorTable, specifiedColor=<rgba 0 0 0 0>}
      shadowColorAndOpacity=<rgba 0 0 0 1>
      shadowOffset=<0 0>
      strikeBrush {
       bIsDynamicallyLoaded=false
       drawAs=Image
       imageType=NoImage
       mirroring=NoMirror
       resourceName=None
       resourceObject=null
       tiling=NoTile
       imageSize=<32 32>
       margin=<ltrb 0 0 0 0>
       outlineSettings {
        bUseBrushTransparency=false
        roundingType=HalfHeightRadius
        width=0
        color {colorUseRule=UseColor_Specified, specifiedColor=<rgba 0 0 0 0>}
        cornerRadii=<0 0 0 0>
       }
       tintColor {colorUseRule=UseColor_Specified, specifiedColor=<rgba 1 1 1 1>}
       uVRegion {bIsValid=false, max=<0 0>, min=<0 0>}
      }
      underlineBrush {
       bIsDynamicallyLoaded=false
       drawAs=Image
       imageType=NoImage
       mirroring=NoMirror
       resourceName=None
       resourceObject=null
       tiling=NoTile
       imageSize=<32 32>
       margin=<ltrb 0 0 0 0>
       outlineSettings {
        bUseBrushTransparency=false
        roundingType=HalfHeightRadius
        width=0
        color {colorUseRule=UseColor_Specified, specifiedColor=<rgba 0 0 0 0>}
        cornerRadii=<0 0 0 0>
       }
       tintColor {colorUseRule=UseColor_Specified, specifiedColor=<rgba 1 1 1 1>}
       uVRegion {bIsValid=false, max=<0 0>, min=<0 0>}
      }
     }
     vScrollBarPadding=<ltrb 0 0 0 0>
    }
   }
   slot {class=VerticalBoxSlot, properties {padding=<ltrb 2 0 2 4>}}
  }
  Border_3 {
   widgetClassPath=&Border
   parent=VerticalBox
   widget=Border_3
   properties {brushColor=<rgba 0.0103 0.0103 0.0103 1>, padding=<ltrb 0 0 0 0>}
   slot {class=VerticalBoxSlot, properties {size {sizeRule=Fill, value=1}}}
  }
  VerticalBox_268 {
   widgetClassPath=&VerticalBox
   parent=Border_3
   widget=VerticalBox_268
   slot {class=BorderSlot, properties {padding=<ltrb 0 0 0 0>}}
  }
  DestinationMetaSoundInputsList {
   widgetClassPath=&EditorUtilityListView
   bIsVariable=true
   parent=VerticalBox_268
   widget=DestinationMetaSoundInputsList
   properties {
    bEnableScrollAnimation=true
    bIsFocusable=false
    entryWidgetClass=&MetaSoundPresetMigrator_InputEntry_C
    numDesignerPreviewEntries=10
   }
   slot {class=VerticalBoxSlot, properties {size {sizeRule=Fill, value=1}}}
  }
  VerticalBox_2 {
   widgetClassPath=&VerticalBox
   parent=VerticalBox_268
   widget=VerticalBox_2
   slot {
    class=VerticalBoxSlot
    properties {horizontalAlignment=HAlign_Center, verticalAlignment=VAlign_Top, padding=<ltrb 0 64 0 0>, size {sizeRule=Fill, value=1}}
   }
  }
  Image {
   widgetClassPath=&Image
   parent=VerticalBox_2
   widget=Image
   properties {
    brush {
     bIsDynamicallyLoaded=false
     drawAs=Image
     imageType=FullColor
     mirroring=NoMirror
     resourceName=None
     resourceObject=null
     tiling=NoTile
     imageSize=<64 64>
     margin=<ltrb 0 0 0 0>
     outlineSettings {
      bUseBrushTransparency=false
      roundingType=HalfHeightRadius
      width=0
      color {colorUseRule=UseColor_Specified, specifiedColor=<rgba 0 0 0 0>}
      cornerRadii=<0 0 0 0>
     }
     tintColor {colorUseRule=UseColor_Specified, specifiedColor=<rgba 0.0284 0.0284 0.0284 1>}
     uVRegion {bIsValid=false, max=<0 0>, min=<0 0>}
    }
    renderTransform {angle=90, scale=<1 1>, shear=<0 0>, translation=<0 0>}
   }
   slot {class=VerticalBoxSlot, properties {horizontalAlignment=HAlign_Center, verticalAlignment=VAlign_Center, padding=<ltrb 16 16 16 16>}}
  }
  TextBlock {
   widgetClassPath=&TextBlock
   parent=VerticalBox_2
   widget=TextBlock
   properties {
    justification=Center
    text=Select a parent preset you wish to migrate to.
    colorAndOpacity {colorUseRule=UseColor_Specified, specifiedColor=<rgba 0.5029 0.5029 0.5029 1>}
    font {
     bForceMonospaced=false
     bMaterialIsStencil=false
     fontMaterial=null
     fontObject=&Roboto
     letterSpacing=0
     monospacedWidth=1
     size=9
     skewAmount=0
     typefaceFontName=Regular
     outlineSettings {
      bApplyOutlineToDropShadows=false
      bMiteredCorners=false
      bSeparateFillAlpha=false
      outlineMaterial=null
      outlineSize=0
      outlineColor=<rgba 0 0 0 1>
     }
    }
   }
   slot {class=VerticalBoxSlot}
  }
  Controls {widgetClassPath=&HorizontalBox, parent=Body, widget=Controls, slot {class=VerticalBoxSlot}}
  HorizontalBox_104 {widgetClassPath=&HorizontalBox, parent=Controls, widget=HorizontalBox_104, slot {class=HorizontalBoxSlot}}
  LinkMatchingButton {
   widgetClassPath=&EditorUtilityButton
   parent=HorizontalBox_104
   widget=LinkMatchingButton
   properties {
    toolTipText=Link all inputs that have a matching name.
    widgetStyle {
     clickedSlateSound {resourceObject=null}
     disabled {
      bIsDynamicallyLoaded=false
      drawAs=RoundedBox
      imageType=NoImage
      mirroring=NoMirror
      resourceName=None
      resourceObject=null
      tiling=NoTile
      imageSize=<0 0>
      margin=<ltrb 0 0 0 0>
      outlineSettings {
       bUseBrushTransparency=false
       roundingType=FixedRadius
       width=1
       color {colorUseRule=UseColor_ColorTable, specifiedColor=<rgba 0 0 0 0>}
       cornerRadii=<4 4 4 4>
      }
      tintColor {colorUseRule=UseColor_Specified, specifiedColor=<rgba 0.0395 0.0395 0.0395 1>}
      uVRegion {bIsValid=false, max=<0 0>, min=<0 0>}
     }
     disabledForeground {colorUseRule=UseColor_ColorTable, specifiedColor=<rgba 0 0 0 0>}
     hovered {
      bIsDynamicallyLoaded=false
      drawAs=RoundedBox
      imageType=NoImage
      mirroring=NoMirror
      resourceName=None
      resourceObject=null
      tiling=NoTile
      imageSize=<32 32>
      margin=<ltrb 0 0 0 0>
      outlineSettings {
       bUseBrushTransparency=false
       roundingType=FixedRadius
       width=1
       color {colorUseRule=UseColor_ColorTable, specifiedColor=<rgba 0 0 0 0>}
       cornerRadii=<4 4 4 4>
      }
      tintColor {colorUseRule=UseColor_Specified, specifiedColor=<rgba 0.0953 0.0953 0.0953 1>}
      uVRegion {bIsValid=false, max=<0 0>, min=<0 0>}
     }
     hoveredForeground {colorUseRule=UseColor_ColorTable, specifiedColor=<rgba 0 0 0 0>}
     hoveredSlateSound {resourceObject=null}
     normal {
      bIsDynamicallyLoaded=false
      drawAs=RoundedBox
      imageType=NoImage
      mirroring=NoMirror
      resourceName=None
      resourceObject=null
      tiling=NoTile
      imageSize=<32 32>
      margin=<ltrb 0 0 0 0>
      outlineSettings {
       bUseBrushTransparency=false
       roundingType=FixedRadius
       width=1
       color {colorUseRule=UseColor_ColorTable, specifiedColor=<rgba 0 0 0 0>}
       cornerRadii=<4 4 4 4>
      }
      tintColor {colorUseRule=UseColor_Specified, specifiedColor=<rgba 0.0395 0.0395 0.0395 1>}
      uVRegion {bIsValid=false, max=<0 0>, min=<0 0>}
     }
     normalForeground {colorUseRule=UseColor_ColorTable, specifiedColor=<rgba 0 0 0 0>}
     normalPadding=<ltrb 4 1.5 4 1.5>
     pressed {
      bIsDynamicallyLoaded=false
      drawAs=RoundedBox
      imageType=NoImage
      mirroring=NoMirror
      resourceName=None
      resourceObject=null
      tiling=NoTile
      imageSize=<32 32>
      margin=<ltrb 0 0 0 0>
      outlineSettings {
       bUseBrushTransparency=false
       roundingType=FixedRadius
       width=1
       color {colorUseRule=UseColor_ColorTable, specifiedColor=<rgba 0 0 0 0>}
       cornerRadii=<4 4 4 4>
      }
      tintColor {colorUseRule=UseColor_Specified, specifiedColor=<rgba 0.0284 0.0284 0.0284 1>}
      uVRegion {bIsValid=false, max=<0 0>, min=<0 0>}
     }
     pressedForeground {colorUseRule=UseColor_ColorTable, specifiedColor=<rgba 0 0 0 0>}
     pressedPadding=<ltrb 4 2.5 4 0.5>
     pressedSlateSound {resourceObject=null}
    }
   }
   slot {class=HorizontalBoxSlot, properties {padding=<ltrb 4 8 4 8>}}
  }
  HorizontalBox_2 {widgetClassPath=&HorizontalBox, parent=LinkMatchingButton, widget=HorizontalBox_2, slot {class=ButtonSlot}}
  TextBlock_1 {
   widgetClassPath=&TextBlock
   parent=HorizontalBox_2
   widget=TextBlock_1
   properties {
    justification=Center
    text=Link Matching Inputs
    colorAndOpacity {colorUseRule=UseColor_Specified, specifiedColor=<rgba 0.5885 0.5885 0.5885 1>}
    font {
     bForceMonospaced=false
     bMaterialIsStencil=false
     fontMaterial=null
     fontObject=&Roboto
     letterSpacing=0
     monospacedWidth=1
     size=10.5
     skewAmount=0
     typefaceFontName=Bold
     outlineSettings {
      bApplyOutlineToDropShadows=false
      bMiteredCorners=false
      bSeparateFillAlpha=false
      outlineMaterial=null
      outlineSize=0
      outlineColor=<rgba 0 0 0 1>
     }
    }
    margin=<ltrb 0 1 0 0>
   }
   slot {class=HorizontalBoxSlot, properties {padding=<ltrb 4 0 0 0>}}
  }
  UnlinkAllButton {
   widgetClassPath=&EditorUtilityButton
   parent=HorizontalBox_104
   widget=UnlinkAllButton
   properties {
    toolTipText=Unlink all inputs.
    widgetStyle {
     clickedSlateSound {resourceObject=null}
     disabled {
      bIsDynamicallyLoaded=false
      drawAs=RoundedBox
      imageType=NoImage
      mirroring=NoMirror
      resourceName=None
      resourceObject=null
      tiling=NoTile
      imageSize=<0 0>
      margin=<ltrb 0 0 0 0>
      outlineSettings {
       bUseBrushTransparency=false
       roundingType=FixedRadius
       width=1
       color {colorUseRule=UseColor_ColorTable, specifiedColor=<rgba 0 0 0 0>}
       cornerRadii=<4 4 4 4>
      }
      tintColor {colorUseRule=UseColor_Specified, specifiedColor=<rgba 0.0395 0.0395 0.0395 1>}
      uVRegion {bIsValid=false, max=<0 0>, min=<0 0>}
     }
     disabledForeground {colorUseRule=UseColor_ColorTable, specifiedColor=<rgba 0 0 0 0>}
     hovered {
      bIsDynamicallyLoaded=false
      drawAs=RoundedBox
      imageType=NoImage
      mirroring=NoMirror
      resourceName=None
      resourceObject=null
      tiling=NoTile
      imageSize=<32 32>
      margin=<ltrb 0 0 0 0>
      outlineSettings {
       bUseBrushTransparency=false
       roundingType=FixedRadius
       width=1
       color {colorUseRule=UseColor_ColorTable, specifiedColor=<rgba 0 0 0 0>}
       cornerRadii=<4 4 4 4>
      }
      tintColor {colorUseRule=UseColor_Specified, specifiedColor=<rgba 0.0953 0.0953 0.0953 1>}
      uVRegion {bIsValid=false, max=<0 0>, min=<0 0>}
     }
     hoveredForeground {colorUseRule=UseColor_ColorTable, specifiedColor=<rgba 0 0 0 0>}
     hoveredSlateSound {resourceObject=null}
     normal {
      bIsDynamicallyLoaded=false
      drawAs=RoundedBox
      imageType=NoImage
      mirroring=NoMirror
      resourceName=None
      resourceObject=null
      tiling=NoTile
      imageSize=<32 32>
      margin=<ltrb 0 0 0 0>
      outlineSettings {
       bUseBrushTransparency=false
       roundingType=FixedRadius
       width=1
       color {colorUseRule=UseColor_ColorTable, specifiedColor=<rgba 0 0 0 0>}
       cornerRadii=<4 4 4 4>
      }
      tintColor {colorUseRule=UseColor_Specified, specifiedColor=<rgba 0.0395 0.0395 0.0395 1>}
      uVRegion {bIsValid=false, max=<0 0>, min=<0 0>}
     }
     normalForeground {colorUseRule=UseColor_ColorTable, specifiedColor=<rgba 0 0 0 0>}
     normalPadding=<ltrb 4 1.5 4 1.5>
     pressed {
      bIsDynamicallyLoaded=false
      drawAs=RoundedBox
      imageType=NoImage
      mirroring=NoMirror
      resourceName=None
      resourceObject=null
      tiling=NoTile
      imageSize=<32 32>
      margin=<ltrb 0 0 0 0>
      outlineSettings {
       bUseBrushTransparency=false
       roundingType=FixedRadius
       width=1
       color {colorUseRule=UseColor_ColorTable, specifiedColor=<rgba 0 0 0 0>}
       cornerRadii=<4 4 4 4>
      }
      tintColor {colorUseRule=UseColor_Specified, specifiedColor=<rgba 0.0284 0.0284 0.0284 1>}
      uVRegion {bIsValid=false, max=<0 0>, min=<0 0>}
     }
     pressedForeground {colorUseRule=UseColor_ColorTable, specifiedColor=<rgba 0 0 0 0>}
     pressedPadding=<ltrb 4 2.5 4 0.5>
     pressedSlateSound {resourceObject=null}
    }
   }
   slot {class=HorizontalBoxSlot, properties {padding=<ltrb 4 8 4 8>}}
  }
  HorizontalBox_3 {widgetClassPath=&HorizontalBox, parent=UnlinkAllButton, widget=HorizontalBox_3, slot {class=ButtonSlot}}
  TextBlock_2 {
   widgetClassPath=&TextBlock
   parent=HorizontalBox_3
   widget=TextBlock_2
   properties {
    justification=Center
    text=Unlink All Inputs
    colorAndOpacity {colorUseRule=UseColor_Specified, specifiedColor=<rgba 0.5885 0.5885 0.5885 1>}
    font {
     bForceMonospaced=false
     bMaterialIsStencil=false
     fontMaterial=null
     fontObject=&Roboto
     letterSpacing=0
     monospacedWidth=1
     size=10.5
     skewAmount=0
     typefaceFontName=Bold
     outlineSettings {
      bApplyOutlineToDropShadows=false
      bMiteredCorners=false
      bSeparateFillAlpha=false
      outlineMaterial=null
      outlineSize=0
      outlineColor=<rgba 0 0 0 1>
     }
    }
    margin=<ltrb 0 1 0 0>
   }
   slot {class=HorizontalBoxSlot, properties {padding=<ltrb 4 0 0 0>}}
  }
  HorizontalBox_7 {
   widgetClassPath=&HorizontalBox
   parent=Controls
   widget=HorizontalBox_7
   slot {class=HorizontalBoxSlot, properties {size {sizeRule=Fill, value=1}}}
  }
  VerticalBox_5 {
   widgetClassPath=&VerticalBox
   parent=HorizontalBox_7
   widget=VerticalBox_5
   slot {class=HorizontalBoxSlot, properties {horizontalAlignment=HAlign_Right, verticalAlignment=VAlign_Center, padding=<ltrb 0 0 4 0>}}
  }
  TextBlock_0 {
   widgetClassPath=&TextBlock
   parent=VerticalBox_5
   widget=TextBlock_0
   properties {
    text=Asset Name:
    colorAndOpacity {colorUseRule=UseColor_Specified, specifiedColor=<rgba 0.5885 0.5885 0.5885 1>}
    font {
     bForceMonospaced=false
     bMaterialIsStencil=false
     fontMaterial=null
     fontObject=&Roboto
     letterSpacing=0
     monospacedWidth=1
     size=9
     skewAmount=0
     typefaceFontName=Bold
     outlineSettings {
      bApplyOutlineToDropShadows=false
      bMiteredCorners=false
      bSeparateFillAlpha=false
      outlineMaterial=null
      outlineSize=0
      outlineColor=<rgba 0 0 0 1>
     }
    }
   }
   slot {class=VerticalBoxSlot, properties {padding=<ltrb 16 0 0 0>}}
  }
  VerticalBox_6 {
   widgetClassPath=&VerticalBox
   parent=HorizontalBox_7
   widget=VerticalBox_6
   slot {class=HorizontalBoxSlot, properties {verticalAlignment=VAlign_Center, size {sizeRule=Fill, value=1}}}
  }
  AssetNameTextBox {
   widgetClassPath=&EditorUtilityEditableTextBox
   bIsVariable=true
   parent=VerticalBox_6
   widget=AssetNameTextBox
   properties {
    selectAllTextWhenFocused=true
    widgetStyle {
     backgroundColor {colorUseRule=UseColor_Specified, specifiedColor=<rgba 1 1 1 1>}
     backgroundImageFocused {
      bIsDynamicallyLoaded=false
      drawAs=RoundedBox
      imageType=NoImage
      mirroring=NoMirror
      resourceName=None
      resourceObject=null
      tiling=NoTile
      imageSize=<0 0>
      margin=<ltrb 0 0 0 0>
      outlineSettings {
       bUseBrushTransparency=false
       roundingType=FixedRadius
       width=1
       color {colorUseRule=UseColor_ColorTable, specifiedColor=<rgba 0 0 0 0>}
       cornerRadii=<4 4 4 4>
      }
      tintColor {colorUseRule=UseColor_Specified, specifiedColor=<rgba 0.0048 0.0048 0.0048 1>}
      uVRegion {bIsValid=false, max=<0 0>, min=<0 0>}
     }
     backgroundImageHovered {
      bIsDynamicallyLoaded=false
      drawAs=RoundedBox
      imageType=NoImage
      mirroring=NoMirror
      resourceName=None
      resourceObject=null
      tiling=NoTile
      imageSize=<0 0>
      margin=<ltrb 0 0 0 0>
      outlineSettings {
       bUseBrushTransparency=false
       roundingType=FixedRadius
       width=1
       color {colorUseRule=UseColor_ColorTable, specifiedColor=<rgba 0 0 0 0>}
       cornerRadii=<4 4 4 4>
      }
      tintColor {colorUseRule=UseColor_Specified, specifiedColor=<rgba 0.0048 0.0048 0.0048 1>}
      uVRegion {bIsValid=false, max=<0 0>, min=<0 0>}
     }
     backgroundImageNormal {
      bIsDynamicallyLoaded=false
      drawAs=RoundedBox
      imageType=NoImage
      mirroring=NoMirror
      resourceName=None
      resourceObject=null
      tiling=NoTile
      imageSize=<0 0>
      margin=<ltrb 0 0 0 0>
      outlineSettings {
       bUseBrushTransparency=false
       roundingType=FixedRadius
       width=1
       color {colorUseRule=UseColor_ColorTable, specifiedColor=<rgba 0 0 0 0>}
       cornerRadii=<4 4 4 4>
      }
      tintColor {colorUseRule=UseColor_Specified, specifiedColor=<rgba 0.0048 0.0048 0.0048 1>}
      uVRegion {bIsValid=false, max=<0 0>, min=<0 0>}
     }
     backgroundImageReadOnly {
      bIsDynamicallyLoaded=false
      drawAs=RoundedBox
      imageType=NoImage
      mirroring=NoMirror
      resourceName=None
      resourceObject=null
      tiling=NoTile
      imageSize=<0 0>
      margin=<ltrb 0 0 0 0>
      outlineSettings {
       bUseBrushTransparency=false
       roundingType=FixedRadius
       width=1
       color {colorUseRule=UseColor_ColorTable, specifiedColor=<rgba 0 0 0 0>}
       cornerRadii=<4 4 4 4>
      }
      tintColor {colorUseRule=UseColor_Specified, specifiedColor=<rgba 0.0284 0.0284 0.0284 1>}
      uVRegion {bIsValid=false, max=<0 0>, min=<0 0>}
     }
     focusedForegroundColor {colorUseRule=UseColor_Specified, specifiedColor=<rgba 1 1 1 1>}
     foregroundColor {colorUseRule=UseColor_Specified, specifiedColor=<rgba 0.5271 0.5271 0.5271 1>}
     hScrollBarPadding=<ltrb 0 0 0 0>
     padding=<ltrb 8 3 8 3>
     readOnlyForegroundColor {colorUseRule=UseColor_Foreground, specifiedColor=<rgba 1 0 1 1>}
     scrollBarStyle {
      thickness=8
      draggedThumbImage {
       bIsDynamicallyLoaded=false
       drawAs=RoundedBox
       imageType=NoImage
       mirroring=NoMirror
       resourceName=None
       resourceObject=null
       tiling=NoTile
       imageSize=<0 0>
       margin=<ltrb 0 0 0 0>
       outlineSettings {
        bUseBrushTransparency=false
        roundingType=FixedRadius
        width=0
        color {colorUseRule=UseColor_Specified, specifiedColor=<rgba 0 0 0 0>}
        cornerRadii=<4 4 4 4>
       }
       tintColor {colorUseRule=UseColor_Specified, specifiedColor=<rgba 0.2159 0.2159 0.2159 1>}
       uVRegion {bIsValid=false, max=<0 0>, min=<0 0>}
      }
      horizontalBackgroundImage {
       bIsDynamicallyLoaded=false
       drawAs=NoDrawType
       imageType=NoImage
       mirroring=NoMirror
       resourceName=None
       resourceObject=null
       tiling=NoTile
       imageSize=<0 0>
       margin=<ltrb 0 0 0 0>
       outlineSettings {
        bUseBrushTransparency=false
        roundingType=HalfHeightRadius
        width=0
        color {colorUseRule=UseColor_Specified, specifiedColor=<rgba 0 0 0 0>}
        cornerRadii=<0 0 0 0>
       }
       tintColor {colorUseRule=UseColor_Specified, specifiedColor=<rgba 1 1 1 1>}
       uVRegion {bIsValid=false, max=<0 0>, min=<0 0>}
      }
      horizontalBottomSlotImage {
       bIsDynamicallyLoaded=false
       drawAs=NoDrawType
       imageType=NoImage
       mirroring=NoMirror
       resourceName=None
       resourceObject=null
       tiling=NoTile
       imageSize=<0 0>
       margin=<ltrb 0 0 0 0>
       outlineSettings {
        bUseBrushTransparency=false
        roundingType=HalfHeightRadius
        width=0
        color {colorUseRule=UseColor_Specified, specifiedColor=<rgba 0 0 0 0>}
        cornerRadii=<0 0 0 0>
       }
       tintColor {colorUseRule=UseColor_Specified, specifiedColor=<rgba 1 1 1 1>}
       uVRegion {bIsValid=false, max=<0 0>, min=<0 0>}
      }
      horizontalTopSlotImage {
       bIsDynamicallyLoaded=false
       drawAs=NoDrawType
       imageType=NoImage
       mirroring=NoMirror
       resourceName=None
       resourceObject=null
       tiling=NoTile
       imageSize=<0 0>
       margin=<ltrb 0 0 0 0>
       outlineSettings {
        bUseBrushTransparency=false
        roundingType=HalfHeightRadius
        width=0
        color {colorUseRule=UseColor_Specified, specifiedColor=<rgba 0 0 0 0>}
        cornerRadii=<0 0 0 0>
       }
       tintColor {colorUseRule=UseColor_Specified, specifiedColor=<rgba 1 1 1 1>}
       uVRegion {bIsValid=false, max=<0 0>, min=<0 0>}
      }
      hoveredThumbImage {
       bIsDynamicallyLoaded=false
       drawAs=RoundedBox
       imageType=NoImage
       mirroring=NoMirror
       resourceName=None
       resourceObject=null
       tiling=NoTile
       imageSize=<0 0>
       margin=<ltrb 0 0 0 0>
       outlineSettings {
        bUseBrushTransparency=false
        roundingType=FixedRadius
        width=0
        color {colorUseRule=UseColor_Specified, specifiedColor=<rgba 0 0 0 0>}
        cornerRadii=<4 4 4 4>
       }
       tintColor {colorUseRule=UseColor_Specified, specifiedColor=<rgba 0.2159 0.2159 0.2159 1>}
       uVRegion {bIsValid=false, max=<0 0>, min=<0 0>}
      }
      normalThumbImage {
       bIsDynamicallyLoaded=false
       drawAs=RoundedBox
       imageType=NoImage
       mirroring=NoMirror
       resourceName=None
       resourceObject=null
       tiling=NoTile
       imageSize=<0 0>
       margin=<ltrb 0 0 0 0>
       outlineSettings {
        bUseBrushTransparency=false
        roundingType=FixedRadius
        width=0
        color {colorUseRule=UseColor_Specified, specifiedColor=<rgba 0 0 0 0>}
        cornerRadii=<4 4 4 4>
       }
       tintColor {colorUseRule=UseColor_Specified, specifiedColor=<rgba 0.0953 0.0953 0.0953 1>}
       uVRegion {bIsValid=false, max=<0 0>, min=<0 0>}
      }
      verticalBackgroundImage {
       bIsDynamicallyLoaded=false
       drawAs=NoDrawType
       imageType=NoImage
       mirroring=NoMirror
       resourceName=None
       resourceObject=null
       tiling=NoTile
       imageSize=<0 0>
       margin=<ltrb 0 0 0 0>
       outlineSettings {
        bUseBrushTransparency=false
        roundingType=HalfHeightRadius
        width=0
        color {colorUseRule=UseColor_Specified, specifiedColor=<rgba 0 0 0 0>}
        cornerRadii=<0 0 0 0>
       }
       tintColor {colorUseRule=UseColor_Specified, specifiedColor=<rgba 1 1 1 1>}
       uVRegion {bIsValid=false, max=<0 0>, min=<0 0>}
      }
      verticalBottomSlotImage {
       bIsDynamicallyLoaded=false
       drawAs=NoDrawType
       imageType=NoImage
       mirroring=NoMirror
       resourceName=None
       resourceObject=null
       tiling=NoTile
       imageSize=<0 0>
       margin=<ltrb 0 0 0 0>
       outlineSettings {
        bUseBrushTransparency=false
        roundingType=HalfHeightRadius
        width=0
        color {colorUseRule=UseColor_Specified, specifiedColor=<rgba 0 0 0 0>}
        cornerRadii=<0 0 0 0>
       }
       tintColor {colorUseRule=UseColor_Specified, specifiedColor=<rgba 1 1 1 1>}
       uVRegion {bIsValid=false, max=<0 0>, min=<0 0>}
      }
      verticalTopSlotImage {
       bIsDynamicallyLoaded=false
       drawAs=NoDrawType
       imageType=NoImage
       mirroring=NoMirror
       resourceName=None
       resourceObject=null
       tiling=NoTile
       imageSize=<0 0>
       margin=<ltrb 0 0 0 0>
       outlineSettings {
        bUseBrushTransparency=false
        roundingType=HalfHeightRadius
        width=0
        color {colorUseRule=UseColor_Specified, specifiedColor=<rgba 0 0 0 0>}
        cornerRadii=<0 0 0 0>
       }
       tintColor {colorUseRule=UseColor_Specified, specifiedColor=<rgba 1 1 1 1>}
       uVRegion {bIsValid=false, max=<0 0>, min=<0 0>}
      }
     }
     textStyle {
      overflowPolicy=Clip
      transformPolicy=None
      colorAndOpacity {colorUseRule=UseColor_Foreground, specifiedColor=<rgba 1 0 1 1>}
      font {
       bForceMonospaced=false
       bMaterialIsStencil=false
       fontMaterial=null
       fontObject=null
       letterSpacing=0
       monospacedWidth=1
       size=10
       skewAmount=0
       typefaceFontName=Regular
       outlineSettings {
        bApplyOutlineToDropShadows=false
        bMiteredCorners=false
        bSeparateFillAlpha=false
        outlineMaterial=null
        outlineSize=0
        outlineColor=<rgba 0 0 0 1>
       }
      }
      highlightColor {colorUseRule=UseColor_ColorTable, specifiedColor=<rgba 0 0 0 0>}
      highlightShape {
       bIsDynamicallyLoaded=false
       drawAs=Image
       imageType=NoImage
       mirroring=NoMirror
       resourceName=None
       resourceObject=null
       tiling=NoTile
       imageSize=<0 0>
       margin=<ltrb 0 0 0 0>
       outlineSettings {
        bUseBrushTransparency=false
        roundingType=HalfHeightRadius
        width=0
        color {colorUseRule=UseColor_Specified, specifiedColor=<rgba 0 0 0 0>}
        cornerRadii=<0 0 0 0>
       }
       tintColor {colorUseRule=UseColor_ColorTable, specifiedColor=<rgba 0 0 0 0>}
       uVRegion {bIsValid=false, max=<0 0>, min=<0 0>}
      }
      selectedBackgroundColor {colorUseRule=UseColor_ColorTable, specifiedColor=<rgba 0 0 0 0>}
      shadowColorAndOpacity=<rgba 0 0 0 1>
      shadowOffset=<0 0>
      strikeBrush {
       bIsDynamicallyLoaded=false
       drawAs=Image
       imageType=NoImage
       mirroring=NoMirror
       resourceName=None
       resourceObject=null
       tiling=NoTile
       imageSize=<32 32>
       margin=<ltrb 0 0 0 0>
       outlineSettings {
        bUseBrushTransparency=false
        roundingType=HalfHeightRadius
        width=0
        color {colorUseRule=UseColor_Specified, specifiedColor=<rgba 0 0 0 0>}
        cornerRadii=<0 0 0 0>
       }
       tintColor {colorUseRule=UseColor_Specified, specifiedColor=<rgba 1 1 1 1>}
       uVRegion {bIsValid=false, max=<0 0>, min=<0 0>}
      }
      underlineBrush {
       bIsDynamicallyLoaded=false
       drawAs=Image
       imageType=NoImage
       mirroring=NoMirror
       resourceName=None
       resourceObject=null
       tiling=NoTile
       imageSize=<32 32>
       margin=<ltrb 0 0 0 0>
       outlineSettings {
        bUseBrushTransparency=false
        roundingType=HalfHeightRadius
        width=0
        color {colorUseRule=UseColor_Specified, specifiedColor=<rgba 0 0 0 0>}
        cornerRadii=<0 0 0 0>
       }
       tintColor {colorUseRule=UseColor_Specified, specifiedColor=<rgba 1 1 1 1>}
       uVRegion {bIsValid=false, max=<0 0>, min=<0 0>}
      }
     }
     vScrollBarPadding=<ltrb 0 0 0 0>
    }
   }
   slot {class=VerticalBoxSlot}
  }
  HorizontalBox {
   widgetClassPath=&HorizontalBox
   parent=Controls
   widget=HorizontalBox
   slot {class=HorizontalBoxSlot, properties {horizontalAlignment=HAlign_Right, verticalAlignment=VAlign_Center}}
  }
  MigratePresetButton {
   widgetClassPath=&EditorUtilityButton
   parent=HorizontalBox
   widget=MigratePresetButton
   properties {
    toolTipText=Creates a new preset based on the new parent using the inputs linked from the original preset.
    widgetStyle {
     clickedSlateSound {resourceObject=null}
     disabled {
      bIsDynamicallyLoaded=false
      drawAs=RoundedBox
      imageType=NoImage
      mirroring=NoMirror
      resourceName=None
      resourceObject=null
      tiling=NoTile
      imageSize=<0 0>
      margin=<ltrb 0 0 0 0>
      outlineSettings {
       bUseBrushTransparency=false
       roundingType=FixedRadius
       width=1
       color {colorUseRule=UseColor_ColorTable, specifiedColor=<rgba 0 0 0 0>}
       cornerRadii=<4 4 4 4>
      }
      tintColor {colorUseRule=UseColor_Specified, specifiedColor=<rgba 0.0395 0.0395 0.0395 1>}
      uVRegion {bIsValid=false, max=<0 0>, min=<0 0>}
     }
     disabledForeground {colorUseRule=UseColor_ColorTable, specifiedColor=<rgba 0 0 0 0>}
     hovered {
      bIsDynamicallyLoaded=false
      drawAs=RoundedBox
      imageType=NoImage
      mirroring=NoMirror
      resourceName=None
      resourceObject=null
      tiling=NoTile
      imageSize=<32 32>
      margin=<ltrb 0 0 0 0>
      outlineSettings {
       bUseBrushTransparency=false
       roundingType=FixedRadius
       width=1
       color {colorUseRule=UseColor_ColorTable, specifiedColor=<rgba 0 0 0 0>}
       cornerRadii=<4 4 4 4>
      }
      tintColor {colorUseRule=UseColor_Specified, specifiedColor=<rgba 0.0044 0.2384 1 1>}
      uVRegion {bIsValid=false, max=<0 0>, min=<0 0>}
     }
     hoveredForeground {colorUseRule=UseColor_ColorTable, specifiedColor=<rgba 0 0 0 0>}
     hoveredSlateSound {resourceObject=null}
     normal {
      bIsDynamicallyLoaded=false
      drawAs=RoundedBox
      imageType=NoImage
      mirroring=NoMirror
      resourceName=None
      resourceObject=null
      tiling=NoTile
      imageSize=<32 32>
      margin=<ltrb 0 0 0 0>
      outlineSettings {
       bUseBrushTransparency=false
       roundingType=FixedRadius
       width=1
       color {colorUseRule=UseColor_ColorTable, specifiedColor=<rgba 0 0 0 0>}
       cornerRadii=<4 4 4 4>
      }
      tintColor {colorUseRule=UseColor_Specified, specifiedColor=<rgba 0 0.162 0.7454 1>}
      uVRegion {bIsValid=false, max=<0 0>, min=<0 0>}
     }
     normalForeground {colorUseRule=UseColor_ColorTable, specifiedColor=<rgba 0 0 0 0>}
     normalPadding=<ltrb 12 1.5 12 1.5>
     pressed {
      bIsDynamicallyLoaded=false
      drawAs=RoundedBox
      imageType=NoImage
      mirroring=NoMirror
      resourceName=None
      resourceObject=null
      tiling=NoTile
      imageSize=<32 32>
      margin=<ltrb 0 0 0 0>
      outlineSettings {
       bUseBrushTransparency=false
       roundingType=FixedRadius
       width=1
       color {colorUseRule=UseColor_ColorTable, specifiedColor=<rgba 0 0 0 0>}
       cornerRadii=<4 4 4 4>
      }
      tintColor {colorUseRule=UseColor_Specified, specifiedColor=<rgba 0 0.0802 0.3515 1>}
      uVRegion {bIsValid=false, max=<0 0>, min=<0 0>}
     }
     pressedForeground {colorUseRule=UseColor_ColorTable, specifiedColor=<rgba 0 0 0 0>}
     pressedPadding=<ltrb 12 2.5 12 0.5>
     pressedSlateSound {resourceObject=null}
    }
   }
   slot {class=HorizontalBoxSlot, properties {padding=<ltrb 8 8 8 8>}}
  }
  TextBlock_224 {
   widgetClassPath=&TextBlock
   parent=MigratePresetButton
   widget=TextBlock_224
   properties {
    justification=Center
    text=Migrate Preset
    font {
     bForceMonospaced=false
     bMaterialIsStencil=false
     fontMaterial=null
     fontObject=&Roboto
     letterSpacing=0
     monospacedWidth=1
     size=10.5
     skewAmount=0
     typefaceFontName=Bold
     outlineSettings {
      bApplyOutlineToDropShadows=false
      bMiteredCorners=false
      bSeparateFillAlpha=false
      outlineMaterial=null
      outlineSize=0
      outlineColor=<rgba 0 0 0 1>
     }
    }
    margin=<ltrb 0 1 0 0>
   }
   slot {class=ButtonSlot}
  }
  MetaSoundPresetMigrator_HelpWidget {
   widgetClassPath=/Game/Sounds/Fort_Audio_Blueprints/Utilities/EditorUtilityWidgets/Metasounds/Tools/MetaSoundPresetMigratorData/MetaSoundPresetMigrator_HelpWidget
   bIsVariable=true
   parent=Overlay_3
   widget=MetaSoundPresetMigrator_HelpWidget
   properties {visibility=Hidden}
   slot {class=OverlaySlot, properties {horizontalAlignment=HAlign_Fill, verticalAlignment=VAlign_Fill, padding=<ltrb 128 128 128 128>}}
  }
 ]
}
